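# Unlit, two-sided translucent material for the PeacefulPoem Pumped character's eye trail, usable on skeletal and Niagara meshes.
# Opacity combines a panning fire texture and smoke noise under a UV gradient mask, with optional subtractive erosion, then view-angle and depth fades.
# Emissive blends two colors along V; particle color and dynamic parameters can drive UV offset, pan speed and alpha.
# World position offset layers noise wiggle, tapered bends, local-axis pushes and camera offset.
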Material /BRCosmetics/Characters/Player/Male/Large/Bodies/M_LRG_PeacefulPoem_Pumped/FX/M_PeacefulPoemPumped_EyeTrail
graph {
 properties {
  bUsedWithMeshDeformer=true
  bUsedWithNiagaraMeshParticles=true
  bUsedWithSkeletalMesh=true
  blendMode=BLEND_Translucent
  pixelDepthOffsetMode=PDOM_Legacy
  shadingModel=MSM_Unlit
  twoSided=true
  usedShadingModels=Unlit
 }
 nodes [
  n1 = ComponentMask(None=TextureCoordinate()) {g=true}
  n2 = ScalarParameter() {defaultValue=0.9, parameterName=Gradient MidPoint}
  n3 = Divide(A=@n1, B=@n2)
  n4 = Divide(A=OneMinus(None=@n1), B=OneMinus(None=@n2))
  n5 = Power(
   Base=Multiply(A=Saturate(None=@n3), B=Saturate(None=@n4))
   Exp=ScalarParameter() {defaultValue=1, parameterName=Eyetrail Gradient Length Exponent}
  )
  n6 = ComponentMask(None=TextureCoordinate()) {r=true}
  n7 = OneMinus(None=@n6)
  n8 = Power(Base=Multiply(A=Multiply(A=@n7, B=@n6)) {constB=4}, Exp=ScalarParameter() {defaultValue=4, parameterName=Eyetrail Gradient Width Exponent})
  n9 = Multiply(A=Saturate(None=Multiply(A=@n5, B=@n8)), B=ScalarParameter() {defaultValue=1, parameterName=Gradient Power})
  n10 = Constant()
  n11 = MaterialFunctionCall(Float2=TextureCoordinate()) {materialFunction=&BreakOutFloat2Components}
  n12 = Multiply(
   A=Power(Base=@n11.G, Exp=ScalarParameter() {defaultValue=5, group=Alpha, parameterName=UV Bias, sortPriority=6})
   B=ScalarParameter() {defaultValue=1, group=Alpha, parameterName=UV Bias Amount, sortPriority=6}
  )
  n13 = Add(
   A=LinearInterpolate(A=Multiply(B=@n12) {constA=0.25}, B=@n11.G, Alpha=ScalarParameter() {parameterName=UseStraightUVs})
   B=Add(A=ParticleColor().G, B=DynamicParameter() {paramNames [UV Offset Y, AlphaNoise_1_Speed, AlphaNoise_2_Speed, AlphaNoise_3_Speed]}.UV Offset Y)
  )
  n14 = ComponentMask(None=VectorParameter() {group=Alpha, parameterName=TextureTileUV, sortPriority=5, defaultValue=<rgba 1 1 0 0>}.RGB) {g=true, r=true}
  n15 = MaterialFunctionCall(X=Constant(), Y=VertexColor().G) {materialFunction=&MakeFloat2}
  n16 = Multiply(
   A=TextureCoordinate()
   B=LinearInterpolate(A=@n14, B=Add(A=@n14, B=@n15.Result), Alpha=ScalarParameter() {parameterName=UseVertexColorForUOffset})
  )
  n17 = ScalarParameter() {defaultValue=0.5, group=Alpha, parameterName=Alpha_Speed, sortPriority=6}
  n18 = DynamicParameter() {paramNames [UV Offset Y, AlphaNoise_1_Speed, AlphaNoise_2_Speed, AlphaNoise_3_Speed]}
  n19 = Multiply(
   A=Time()
   B=LinearInterpolate(A=@n17, B=Multiply(A=@n18.AlphaNoise_1_Speed, B=@n17), Alpha=ScalarParameter() {parameterName=UseDynParmForPanSpeed})
  )
  n20 = AppendVector(
   A=ScalarParameter() {group=Alpha, parameterName=AlphaPanAmountX, sortPriority=7}
   B=ScalarParameter() {defaultValue=1, group=Alpha, parameterName=AlphaPanAmountY, sortPriority=8}
  )
  n21 = TextureSampleParameter2D(UVs=Add(
    A=MaterialFunctionCall(X=@n10, Y=@n13) {materialFunction=&MakeFloat2}.Result
    B=Panner(Coordinate=@n16, Time=@n19, Speed=@n20) {speedX=-0.1, speedY=1}
   )) {
   parameterName=Alpha Texture Two
   samplerType=SAMPLERTYPE_LinearColor
   texture=/Game/Characters/Player/Male/Large/Bodies/M_LRG_PeacefulPoem_Pumped/FX/T_FX_PeacefulPoemPumped_Fire
  }
  n22 = Add(
   A=Multiply(A=@n21.R, B=ScalarParameter() {defaultValue=1, parameterName=CoreMult})
   B=Multiply(A=@n21.G, B=ScalarParameter() {defaultValue=0.2, parameterName=BaseMult})
  )
  n23 = Multiply(A=@n21.B, B=ScalarParameter() {defaultValue=0.2, parameterName=GlowMult})
  n24 = Constant()
  n25 = MaterialFunctionCall(Float2=TextureCoordinate()) {materialFunction=&BreakOutFloat2Components}
  n26 = Multiply(
   A=Power(Base=@n25.G, Exp=ScalarParameter() {defaultValue=5, group=Alpha, parameterName=UV Bias, sortPriority=6})
   B=ScalarParameter() {defaultValue=1, group=Alpha, parameterName=UV Bias Amount, sortPriority=6}
  )
  n27 = Add(
   A=LinearInterpolate(A=Multiply(B=@n26) {constA=0.25}, B=@n25.G, Alpha=ScalarParameter() {parameterName=UseStraightUVs})
   B=Add(A=ParticleColor().G, B=DynamicParameter() {paramNames [UV Offset Y, AlphaNoise_1_Speed, AlphaNoise_2_Speed, AlphaNoise_3_Speed]}.UV Offset Y)
  )
  n28 = Multiply(
   A=TextureCoordinate()
   B=ComponentMask(None=VectorParameter() {group=Alpha, parameterName=TextureTileUV, sortPriority=5, defaultValue=<rgba 1 1 0 0>}.RGB) {g=true, r=true}
  )
  n29 = ScalarParameter() {defaultValue=0.5, group=Alpha, parameterName=Noise_Alpha_Speed, sortPriority=6}
  n30 = DynamicParameter() {paramNames [UV Offset Y, AlphaNoise_1_Speed, AlphaNoise_2_Speed, AlphaNoise_3_Speed]}
  n31 = Multiply(
   A=Time()
   B=LinearInterpolate(A=@n29, B=Multiply(A=@n30.AlphaNoise_2_Speed, B=@n29), Alpha=ScalarParameter() {parameterName=UseDynParmForPanSpeed})
  )
  n32 = AppendVector(
   A=ScalarParameter() {group=Alpha, parameterName=Noise_AlphaPanAmountX, sortPriority=7}
   B=ScalarParameter() {defaultValue=1, group=Alpha, parameterName=Noise_AlphaPanAmountY, sortPriority=8}
  )
  n33 = TextureSampleParameter2D(UVs=Add(
    A=MaterialFunctionCall(X=@n24, Y=@n27) {materialFunction=&MakeFloat2}.Result
    B=Panner(Coordinate=@n28, Time=@n31, Speed=@n32) {speedX=-0.1, speedY=1}
   )) {
   group=Alpha
   parameterName=Alpha Texture One
   samplerSource=SSM_Wrap_WorldGroupSettings
   sortPriority=3
   texture=/Game/Effects/Fort_Effects/Textures/Tile/T_Inky_Smoke_Tile2
  }
  n34 = ScalarParameter() {defaultValue=1, parameterName=NoiseMult}
  n35 = Multiply(A=@n33.R, B=@n34)
  n36 = Add(A=Saturate(None=Add(A=@n22, B=@n23)), B=@n35)
  n37 = Multiply(A=@n9) {constB=10}
  n38 = Constant()
  n39 = MaterialFunctionCall(Float2=TextureCoordinate()) {materialFunction=&BreakOutFloat2Components}
  n40 = Multiply(
   A=Power(Base=@n39.G, Exp=ScalarParameter() {defaultValue=5, group=Alpha, parameterName=UV Bias, sortPriority=6})
   B=ScalarParameter() {defaultValue=1, group=Alpha, parameterName=UV Bias Amount, sortPriority=6}
  )
  n41 = Add(
   A=LinearInterpolate(A=Multiply(B=@n40) {constA=0.25}, B=@n39.G, Alpha=ScalarParameter() {parameterName=UseStraightUVs})
   B=Add(A=ParticleColor().G, B=DynamicParameter() {paramNames [UV Offset Y, AlphaNoise_1_Speed, AlphaNoise_2_Speed, AlphaNoise_3_Speed]}.UV Offset Y)
  )
  n42 = Multiply(
   A=TextureCoordinate()
   B=ComponentMask(None=VectorParameter() {group=Alpha, parameterName=TextureTileUV, sortPriority=5, defaultValue=<rgba 1 1 0 0>}.RGB) {g=true, r=true}
  )
  n43 = ScalarParameter() {defaultValue=0.5, group=Alpha, parameterName=Noise_Alpha_Speed_2, sortPriority=6}
  n44 = DynamicParameter() {paramNames [UV Offset Y, AlphaNoise_1_Speed, AlphaNoise_2_Speed, AlphaNoise_3_Speed]}
  n45 = Multiply(
   A=Time()
   B=LinearInterpolate(A=@n43, B=Multiply(A=@n44.AlphaNoise_3_Speed, B=@n43), Alpha=ScalarParameter() {parameterName=UseDynParmForPanSpeed})
  )
  n46 = AppendVector(
   A=ScalarParameter() {group=Alpha, parameterName=Noise_AlphaPanAmountX, sortPriority=7}
   B=ScalarParameter() {defaultValue=1, group=Alpha, parameterName=Noise_AlphaPanAmountY, sortPriority=8}
  )
  n47 = TextureSampleParameter2D(UVs=Add(
    A=MaterialFunctionCall(X=@n38, Y=@n41) {materialFunction=&MakeFloat2}.Result
    B=Panner(Coordinate=@n42, Time=@n45, Speed=@n46) {speedX=-0.1, speedY=1}
   )) {
   group=Alpha
   parameterName=Alpha Texture One
   samplerSource=SSM_Wrap_WorldGroupSettings
   sortPriority=3
   texture=&T_Inky_Smoke_Tile2
  }
  n48 = Add(A=Multiply(A=@n34, B=@n47.R), B=@n35)
  n49 = Multiply(A=Saturate(None=Power(Base=@n37)), B=Subtract(A=@n9, B=OneMinus(None=@n48)))
  n50 = OneMinus(None=ComponentMask(None=TextureCoordinate()) {g=true})
  n51 = Saturate(None=Power(Base=Multiply(A=@n50) {constB=10}))
  n52 = Multiply(
   A=LinearInterpolate(A=Multiply(A=@n9, B=@n36), B=Multiply(A=@n49, B=@n51), Alpha=ScalarParameter() {parameterName=UseSubtract})
   B=MaterialFunctionCall("FadeSensitivity (optional)"=ScalarParameter() {defaultValue=0.2, group=Alpha, parameterName=FadeViewAngle, sortPriority=2}) {materialFunction=&FadeBasedOnViewAngle}.Result
  )
  n53 = DepthFade(Opacity=@n52, FadeDistance=ScalarParameter() {defaultValue=2, group=Alpha, parameterName=DepthFade, sortPriority=1})
  n54 = Multiply(
   A=ShadingPathSwitch(Default=@n53, Deferred=@n53, Mobile=@n52) {desc=Migrated from FeatureLevelSwitch}
   B=ScalarParameter() {defaultValue=1, parameterName=MasterAlphaMult}
  )
  LinearInterpolate(A=@n54, B=Multiply(A=@n54, B=ParticleColor().A), Alpha=ScalarParameter() {parameterName=UseParticleAlpha})
  n55 = VectorParameter() {group=Color, parameterName=Color_01, sortPriority=1, defaultValue=<rgba 2 0 1.5 1>}
  n56 = VectorParameter() {group=Color, parameterName=Color_02, sortPriority=1, defaultValue=<rgba 0.1 1 2 1>}
  n57 = OneMinus(None=MaterialFunctionCall(Float2=TextureCoordinate()) {materialFunction=&BreakOutFloat2Components}.G)
  n58 = ScalarParameter() {defaultValue=1, parameterName=ColorRamp_mult}
  n59 = Saturate(None=Power(Base=Multiply(A=@n57, B=@n58), Exp=ScalarParameter() {defaultValue=2, parameterName=ColorRamp_pow}))
  MaterialFunctionCall(Emissive=Multiply(A=LinearInterpolate(A=@n55.RGB, B=@n56.RGB, Alpha=@n59), B=ScalarParameter() {defaultValue=1, parameterName=ColorMult})) {materialFunction=&MF_FX_EmissiveAdjustments}
  n60 = MaterialFunctionCall() {materialFunction=&LinearGradient}
  n61 = Multiply(
   A=Power(Base=OneMinus(None=@n60.VGradient), Exp=ScalarParameter() {defaultValue=4, group=WPO_Shape, parameterName=Bend_Up_Contrast, sortPriority=1}) {constExponent=4}
   B=Transform(None=Constant3Vector() {constant=<rgba 0 0 1 0>}) {transformSourceType=TRANSFORMSOURCE_Local}
  )
  n62 = ScalarParameter() {group=WPO_Shape, parameterName=Bend_Up}
  n63 = OneMinus(None=MaterialFunctionCall() {materialFunction=&LinearGradient}.VGradient)
  n64 = ScalarParameter() {defaultValue=1, group=WPO, parameterName=Wiggle_Contrast, sortPriority=1}
  n65 = Multiply(
   A=TextureCoordinate() {uTiling=0.02, vTiling=0.5}
   B=ComponentMask(None=VectorParameter() {group=WPO, parameterName=Wiggle UV, sortPriority=4, defaultValue=<rgba 1 1 0 0>}.RGB) {g=true, r=true}
  )
  n66 = ParticleColor()
  n67 = Time()
  n68 = ScalarParameter() {defaultValue=0.2, group=WPO, parameterName=Wiggle_Speed, sortPriority=5}
  n69 = TextureSampleParameter2D(UVs=Panner(Coordinate=Add(A=@n65, B=@n66.R), Time=Multiply(A=@n67, B=@n68)) {speedX=0.05, speedY=1}) {
   group=WPO
   parameterName=Wiggle Texture
   samplerSource=SSM_Wrap_WorldGroupSettings
   sortPriority=1
   texture=&T_NoisePack_01
  }
  n70 = AppendVector(A=@n69.RGB, B=@n69.RGB)
  n71 = ScalarParameter() {defaultValue=4, group=WPO, parameterName=WiggleAmount, sortPriority=3}
  n72 = Transform(None=Multiply(A=Power(Base=@n63, Exp=@n64) {constExponent=4}, B=Multiply(A=@n70, B=@n71))) {transformSourceType=TRANSFORMSOURCE_Local}
  n73 = Power(
   Base=OneMinus(None=MaterialFunctionCall() {materialFunction=&LinearGradient}.VGradient)
   Exp=ScalarParameter() {defaultValue=4, group=WPO_Shape, parameterName=Bend_Foward_Contrast, sortPriority=3}
  ) {constExponent=4}
  n74 = Transform(None=Constant3Vector() {constant=<rgba 0 1 0 0>}) {transformSourceType=TRANSFORMSOURCE_Local}
  n75 = Add(
   A=Add(
    A=StaticComponentMaskParameter(None=@n72) {defaultB=true, defaultG=true, defaultR=true, group=WPO, parameterName=Wiggle Mask, sortPriority=2}
    B=Multiply(A=@n61, B=@n62)
   )
   B=Multiply(A=Multiply(A=@n73, B=@n74), B=ScalarParameter() {group=WPO_Shape, parameterName=Bend_Foward, sortPriority=2})
  )
  n76 = Transform(None=Constant3Vector() {constant=<rgba 1 0 0 0>}) {transformSourceType=TRANSFORMSOURCE_Local}
  n77 = ScalarParameter() {group=WPO_Positioning, parameterName=Push_Mesh_Left/Right}
  n78 = Transform(None=Constant3Vector() {constant=<rgba 0 1 0 0>}) {transformSourceType=TRANSFORMSOURCE_Local}
  n79 = ScalarParameter() {group=WPO_Positioning, parameterName=Push_Mesh_Forward}
  n80 = Constant3Vector() {constant=<rgba 0 0 1 0>}
  n81 = Add(
   A=Add(A=Multiply(A=@n76, B=@n77), B=Multiply(A=@n78, B=@n79))
   B=Multiply(
    A=Transform(None=@n80) {transformSourceType=TRANSFORMSOURCE_Local}
    B=ScalarParameter() {group=WPO_Positioning, parameterName=Push_Mesh_Up, sortPriority=1}
   )
  )
  n82 = ScalarParameter() {group=WPO, parameterName=Camera Offset}
  n83 = WorldPosition()
  NamedRerouteDeclaration(None=Add(
    A=Add(A=@n75, B=@n81)
    B=MaterialFunctionCall(Offset Amount=@n82, World Position=@n83.XYZ) {materialFunction=/Engine/Functions/Engine_MaterialFunctions02/WorldPositionOffset/CameraOffset}.Camera Offset
   )) {name=WPO, nodeColor=<rgba 1 0.5391 0 1>}
  NamedRerouteUsage()
 ]
}
