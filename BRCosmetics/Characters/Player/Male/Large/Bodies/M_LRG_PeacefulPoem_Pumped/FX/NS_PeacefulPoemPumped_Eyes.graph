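# Niagara eye effect for the PeacefulPoem Pumped character body: glowing purple eyes plus a mesh-based fire element.
# EyeGlow bursts two persistent particles, offset per eye through a partition-driven lerp, with sprite, light and component renderers; light offsets follow skeletal-mesh sampled rotation.
# Fire burst-spawns an oriented, persistent mesh particle.
# User.bisHeadHidden switches both emitters' visibility tags and disables EyeGlow's particle light when the head is hidden.
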NiagaraSystem /BRCosmetics/Characters/Player/Male/Large/Bodies/M_LRG_PeacefulPoem_Pumped/FX/NS_PeacefulPoemPumped_Eyes
dependencies {usedDataInterfaces [], usedDynamicInputs [], usedModules [&SystemState]}
emitters {
 EyeGlow {
  emitterUpdateScript [
   &EmitterState(Life Cycle Mode=::NewEnumerator1, Loop Behavior=::NewEnumerator1, Loop Duration Mode=::NewEnumerator1, UseLoopDelay=false)
   &SpawnBurst_Instantaneous(Spawn Count=2, Use Spawn Probability=false, UseLoopCountLimit=true)
  ]
  particleSpawnScript [
   &PartitionParticles()
   &InitializeParticle(
    Color=<rgba 0.5 0.1 0.8 1>
    Position Offset=&Lerp_Vector(A=<3.8 4 3.75>, B=<3.8 4 -3.75>, Alpha=&MakeFloatFromInt(Int=$Particles.PartitionParticles.Partition))
    Sprite Rotation Mode=::NewEnumerator2
    Sprite Size Mode=::NewEnumerator3
    Uniform Sprite Size=4
    UsePositionOffset=true
    Write Material Random=true
   )
  ]
  particleUpdateScript [
   &ParticleState(
    Kill Particles When Lifetime Has Elapsed=false
    Let Infinitely Lived Particles Die When Emitter Deactivates=true
    Loop Particles Lifetime=true
   )
   &ScaleColor(Scale Alpha=20, Scale RGB=&VectorFromFloat(Value=3.5), ScaleA=true, ScaleRGB=true)
   &DynamicMaterialParameters(Module.Index0Param0Write=true, Module.Index0Param1Write=true, Module.Index0Param2Write=true, Module.Index0Param3Write=true)
   &FN_Light_Attributes(
    Color=<rgba 0.5 0.1 0.8 1>
    Create Color Attribute=true
    Create Exponent Attribute=true
    Create Position Attribute=true
    Create Radius Attribute=true
    Enable Particle Light Enable=true
    Exponent=1
    Light Enabled:NiagaraBool=&Boolean_NotOperation(Bool=$User.bisHeadHidden)
    Position=&AddVectorToPosition(
     Position=$Particles.Position
     Vector=&RotateVectorByQuaternion(Vector To Rotate=<0 4 -1>, Quaternion=$Particles.SkeletalMeshLocation.SampledRotation)
    )
    Radius=3
    Scale Color=25
   )
   &FN_Gen5_Light(
    Attenutation=3
    Color:LinearColor=<rgba 0.5 0.1 0.8 1>
    Intensity=20
    Position Offset:Vector3f=&RotateVectorByQuaternion(Vector To Rotate=<0 1.2 -1>, Quaternion=$Particles.SkeletalMeshLocation.SampledRotation)
   )
   /SetVariables(Particles.VisibilityTag=&MakeCustomIntFromBool(Bool=$User.bisHeadHidden, True INT=1, False INT=0)) {isSetParameters=true, module=SetVariables_#1}
  ]
  rendererClasses [
   &NiagaraSpriteRendererProperties
   &NiagaraLightRendererProperties
   &NiagaraComponentRendererProperties
  ]
  renderers [
   {rendererClass=&NiagaraSpriteRendererProperties}
   {rendererClass=&NiagaraLightRendererProperties, rendererIndex=1}
   {rendererClass=&NiagaraComponentRendererProperties, rendererIndex=2}
  ]
 }
 Fire {
  emitterUpdateScript [
   &EmitterState(Life Cycle Mode=::NewEnumerator1, Loop Behavior=::NewEnumerator1, Loop Duration Mode=::NewEnumerator1, UseLoopDelay=false)
   &SpawnBurst_Instantaneous(Use Spawn Probability=false, UseLoopCountLimit=false)
  ]
  particleSpawnScript [
   &InitializeParticle(Lifetime=6, Mesh Scale Mode=::NewEnumerator3, Position Offset=<-0.25 0 0>, UsePositionOffset=true, Write Material Random=true)
   &Color(Color=&MakeLinearColorFromVectorAndFloat("Vector (RGB)"=<2 2 2>, "Float (Alpha)"=1))
   &InitialMeshOrientation(
    Mesh Axis To Orient=<0 1 0>
    Mesh Orientation Mode=::NewEnumerator2
    Orientation Vector=<0 1 0>
    Rotation=<0 0.25 0>
    Use Rotation Vector=true
   )
  ]
  particleUpdateScript [
   &ParticleState(Kill Particles When Lifetime Has Elapsed=false, Let Infinitely Lived Particles Die When Emitter Deactivates=true)
   /SetVariables(Particles.VisibilityTag=&MakeCustomIntFromBool(Bool=$User.bisHeadHidden, True INT=1, False INT=0)) {isSetParameters=true, module=SetVariables_#1}
  ]
  rendererClasses [&NiagaraMeshRendererProperties]
  renderers [{rendererClass=&NiagaraMeshRendererProperties, rendererIndex=0}]
 }
}
summary {
 systemName=NS_PeacefulPoemPumped_Eyes
 emitters [
  {
   emitterName=EyeGlow
   rendererClasses [
    &NiagaraSpriteRendererProperties
    &NiagaraLightRendererProperties
    &NiagaraComponentRendererProperties
   ]
  }
  {emitterName=Fire, rendererClasses [&NiagaraMeshRendererProperties]}
 ]
 userVariables [
  User.bisHeadHidden {type=NiagaraBool}
  User.RibbonIDOffset {type=NiagaraInt32}
  User.SkeletalMesh {
   type=NiagaraDataInterfaceSkeletalMesh
   defaultValue {
    struct=NiagaraExt_VariableValue_DataInterface
    value {dataInterface=&self:NiagaraDataInterfaceSkeletalMesh_2}
   }
  }
 ]
}
