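# Parent material for the PeacefulPoem Pumped character cape that adds an animated, tinted emissive noise effect, purple by default.
# It edits base color and emissive between the shared 2023 CharacterParent PreFX and PostFX functions.
# A custom UV texture's red channel darkens base color, masks the glow and drives one noise coordinate.
# Its green channel blends between tiled and remapped UVs for a custom panner, and between two glow treatments.
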Material /BRCosmetics/Characters/Player/Male/Large/Bodies/M_LRG_PeacefulPoem_Pumped/Materials/M_PeacefulPoemPumped_Cape_Parent
graph {
 properties {
  bUseMaterialAttributes=true
  bUsedWithClothing=true
  bUsedWithMeshDeformer=true
  bUsedWithMorphTargets=true
  bUsedWithNanite=true
  bUsedWithSkeletalMesh=true
  pixelDepthOffsetMode=PDOM_Legacy
  previewMesh=/BRCosmetics/Accessories/FORT_Capes/Cape_M_LRG_PeacefulPoem_PumpedUniversal/Meshes/M_LRG_PeacefulPoem_PumpedUniversal_Cape
  shadingModel=MSM_FromMaterialExpression
  usedShadingModels=Default Lit | Subsurface | Clear Coat | Subsurface Profile | Cloth
 }
 nodes [
  n1 = MaterialFunctionCall() {materialFunction=&MF_CharacterParent_2023_PreFX}
  MaterialFunctionCall(Material=NamedRerouteUsage(), M=@n1.M, Pre Skinned Local Position=@n1.M, Distance Blend=@n1.M, Fresnel=@n1.M) {materialFunction=&MF_CharacterParent_2023_PostFX}
  NamedRerouteDeclaration(None=@n1.Result) {name=InMaterialAttributes, nodeColor=<rgba 1 0.7031 0 1>}
  n2 = GetMaterialAttributes(None=NamedRerouteUsage()) {attributeGetTypes [#1, #2]}
  NamedRerouteDeclaration(None=SetMaterialAttributes!(MaterialAttributes=@n2.MaterialAttributes, Emissive Color=NamedRerouteUsage(), Base Color=NamedRerouteUsage()) {attributeSetTypes [#1, #2]}) {name=OutMaterialAttributes, nodeColor=<rgba 0.4062 1 0 1>}
  NamedRerouteDeclaration(None=@n2.Emissive Color) {name=inEmissive, nodeColor=<rgba 1 0.5625 0 1>}
  NamedRerouteDeclaration(None=@n2.Base Color) {name=inBaseColor, nodeColor=<rgba 0.4531 1 0 1>}
  n3 = NamedRerouteUsage()
  n4 = TextureSampleParameter2D() {
   parameterName=CustomUVTexture
   texture=&/T_PeacefulPoemPumped_UVs_Universal_D
  }
  NamedRerouteDeclaration(None=LinearInterpolate(A=@n3, B=Multiply(A=@n3) {constB=0.1}, Alpha=Power(Base=@n4.R))) {name=outBaseColor, nodeColor=<rgba 1 0.8203 0 1>}
  n5 = TextureCoordinate()
  n6 = Constant2Vector() {g=1, r=2}
  n7 = VectorParameter() {parameterName=NoiseTiling, defaultValue=<rgba 3 0.3 0 1>}
  n8 = Multiply(A=ComponentMask(None=TextureCoordinate()) {g=true}, B=@n7.R) {constB=3}
  n9 = Multiply(A=@n4.R, B=@n7.G) {constB=0.3}
  n10 = MaterialFunctionCall!(
   UV=LinearInterpolate(A=Multiply(A=@n5, B=@n6), B=AppendVector(A=@n8, B=@n9), Alpha=@n4.G)
   Speed=ComponentMask(None=VectorParameter() {parameterName=Speed, defaultValue=<rgba 0 -0.025 0 1>}.RGB) {g=true, r=true}
   Mask=@n4.R
  ) {materialFunction=&/MF_PeaceulPoemPumpedPanner}
  n11 = LinearInterpolate(Alpha=@n10.RawNoise) {constA=0.5, constB=2}
  n12 = LinearInterpolate(A=Multiply(A=Multiply(A=@n11, B=@n4.R)) {constB=0.5}, B=@n10.Result, Alpha=@n4.G)
  n13 = VectorParameter() {parameterName=EmissiveColor, defaultValue=<rgba 0.2623 0.0033 0.7758 1>}
  n14 = Add(A=Ceil(None=@n4.R), B=OneMinus(None=@n4.G))
  NamedRerouteDeclaration(None=LinearInterpolate(A=NamedRerouteUsage(), B=Multiply(A=@n12, B=@n13.RGB), Alpha=Saturate(None=@n14))) {name=outEmissive, nodeColor=<rgba 1 0 0.5625 1>}
 ]
}
