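# Material function for the Peaceful Poem Pumped male player body that outputs a panning noise texture, a warp texture and raw noise.
# The mask texture's red channel pans at the input Speed, with UVs nudged by a centered green channel, likely from the tiled warp sample.
# That noise is remapped and combined with multiply and power terms of a rerouted value, apparently the Mask input.
# A parameter-tiled sample supplies the warp output; an appended-UV branch is unconnected.
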MaterialFunction /Game/Characters/Player/Male/Large/Bodies/M_LRG_PeacefulPoem_Pumped/Materials/MF_PeaceulPoemPumpedPanner
graph {nodes [
  n1 = TextureCoordinate()
  AppendVector(A=Multiply(A=ComponentMask(None=@n1) {g=true}) {constB=3}, B=Multiply(A=NamedRerouteUsage()) {constB=0.3})
  n2 = FunctionInput() {inputName=UV, inputType=FunctionInput_Vector2}
  n3 = NamedRerouteUsage()
  n4 = Multiply(A=Subtract(A=ComponentMask(None=@n3) {g=true}) {constB=0.5}) {constB=0.05}
  n5 = TextureSample(UVs=Panner(
    Coordinate=Add(A=@n2, B=@n4)
    Speed=FunctionInput() {inputName=Speed, inputType=FunctionInput_Vector2, sortPriority=1, previewValue=<-0.05 0 0 1>}
   ) {speedY=-0.03}) {
   samplerSource=SSM_Wrap_WorldGroupSettings
   samplerType=SAMPLERTYPE_Masks
   texture=&/T_PeacefulPoem_Pumped_Mask
  }
  n6 = NamedRerouteUsage()
  n7 = Multiply(A=LinearInterpolate(Alpha=@n5.R) {constA=0.25}, B=@n6)
  n8 = Multiply(A=Power(Base=@n6)) {constB=0.75}
  FunctionOutput(None=NamedRerouteDeclaration(None=Add(A=@n7, B=@n8)) {name=PanningTexture, nodeColor=<rgba 0.6406 0 1 1>})
  n9 = TextureSample(UVs=Multiply(A=TextureCoordinate(), B=ScalarParameter() {defaultValue=15, parameterName=WarpTextureTiling}) {constB=15}) {
   samplerSource=SSM_Wrap_WorldGroupSettings
   samplerType=SAMPLERTYPE_Masks
   texture=&T_PeacefulPoem_Pumped_Mask
  }
  FunctionOutput(None=NamedRerouteDeclaration(None=@n9.RGB) {name=DefaultSampledTexture, nodeColor=<rgba 0.875 0 1 1>}) {outputName=WarpTexture, sortPriority=1}
  FunctionOutput(None=@n5.R) {outputName=RawNoise}
  NamedRerouteDeclaration(None=FunctionInput() {inputName=Mask, inputType=FunctionInput_Scalar, sortPriority=2}) {name=Mask, nodeColor=<rgba 1 0.6094 0 1>}
 ]}
