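# Parent character material for the PeacefulPoem Pumped body, adding a tinted (default purple) emissive effect between the shared 2023 CharacterParent PreFX and PostFX functions.
# Camera-offset pre-skinned position builds radial UVs that fake an inner depth glow in the chest.
# Outer glow, UV-texture masks for spikes and claws, panned spike noise, a box-bounded cavity mask and sine-based flicker provide further effect layers.
# Summed layers, scaled by EmissiveColor, set the emissive attribute.
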Material /BRCosmetics/Characters/Player/Male/Large/Bodies/M_LRG_PeacefulPoem_Pumped/Materials/M_PeacefulPoemPumped_Parent
graph {
 properties {
  bUseMaterialAttributes=true
  bUsedWithMeshDeformer=true
  bUsedWithMorphTargets=true
  bUsedWithNanite=true
  bUsedWithSkeletalMesh=true
  pixelDepthOffsetMode=PDOM_Legacy
  shadingModel=MSM_FromMaterialExpression
  usedShadingModels=Default Lit | Subsurface | Clear Coat | Subsurface Profile | Cloth
 }
 nodes [
  n1 = MaterialFunctionCall() {materialFunction=&MF_CharacterParent_2023_PreFX}
  MaterialFunctionCall(Material=NamedRerouteUsage(), M=@n1.M, Pre Skinned Local Position=@n1.M, Distance Blend=@n1.M, Fresnel=@n1.M) {materialFunction=&MF_CharacterParent_2023_PostFX}
  NamedRerouteDeclaration(None=@n1.Result) {name=InMaterialAttributes, nodeColor=<rgba 1 0.7969 0 1>}
  NamedRerouteDeclaration(None=@n1.Tech Art Mask) {name=FXMask, nodeColor=<rgba 0.3594 1 0 1>}
  n2 = Subtract(A=NamedRerouteUsage(), B=ComponentMask(None=NamedRerouteUsage()) {r=true})
  n3 = NamedRerouteUsage()
  n4 = Ceil(None=@n3)
  n5 = Desaturation(None=Saturate(None=Subtract(A=@n2, B=@n4)))
  NamedRerouteDeclaration(None=Multiply(A=Multiply(A=@n5) {constB=7}, B=NamedRerouteUsage())) {name=OutterGlow, nodeColor=<rgba 0.4062 0 1 1>}
  n6 = Subtract(A=ComponentMask(None=NamedRerouteUsage()) {r=true}, B=Step(X=NamedRerouteUsage()) {constY=0.01})
  n7 = Multiply(A=NamedRerouteUsage(), B=Constant2Vector() {g=1, r=1.75})
  n8 = ComponentMask(None=NamedRerouteUsage()) {g=true}
  n9 = TextureCoordinate()
  n10 = TextureSample!(
   UVs=Add(A=Panner(Coordinate=@n7) {speedX=-0.075, speedY=0.01}, B=Multiply(A=@n8) {constB=0.15})
   Tex
   "DDX(UVs)"=DDX(Value=@n9)
   "DDY(UVs)"=DDY(Value=@n9)
   Apply View MipBias
  ) {
   mipValueMode=TMVM_Derivative
   samplerSource=SSM_Wrap_WorldGroupSettings
   samplerType=SAMPLERTYPE_Masks
   texture=/Game/Characters/Player/Male/Large/Bodies/M_LRG_PeacefulPoem_Pumped/Materials/T_PeacefulPoem_Pumped_Mask
  }
  n11 = LinearInterpolate(Alpha=@n10.R) {constA=0.3, constB=1.25}
  n12 = NamedRerouteUsage()
  n13 = OneMinus(None=Power(Base=Saturate(None=@n12)) {constExponent=0.25})
  n14 = Multiply(A=@n13, B=@n13)
  n15 = Multiply(A=@n14, B=Multiply(A=NamedRerouteUsage()) {constB=2})
  n16 = Max(A=NamedRerouteUsage()) {constB=0.8}
  n17 = Add(A=Multiply(A=@n11, B=Multiply(A=@n15) {constB=5}), B=Multiply(A=@n14, B=Multiply(A=@n16) {constB=50}) {constB=10})
  NamedRerouteDeclaration(None=Multiply(A=Saturate(None=@n6), B=Add(A=@n11, B=@n17))) {name=InnerGlow, nodeColor=<rgba 1 0.8906 0 1>}
  n18 = TextureSampleParameter2D() {
   parameterName=CustomUVTexture
   texture=/Game/Characters/Player/Male/Large/Bodies/M_LRG_PeacefulPoem_Pumped/Materials/T_PeacefulPoemPumped_UVs_Body_D
  }
  NamedRerouteDeclaration(None=@n18.R) {name=SpikesMask, nodeColor=<rgba 0 0.8125 1 1>}
  n19 = VectorParameter() {parameterName=SpikeNoiseTiling, defaultValue=<rgba 3 0.3 0 1>}
  n20 = Multiply(A=ComponentMask(None=TextureCoordinate()) {g=true}, B=@n19.R) {constB=3}
  n21 = NamedRerouteUsage()
  n22 = Multiply(A=@n21, B=@n19.G) {constB=0.3}
  n23 = Constant2Vector() {g=-0.03}
  n24 = Constant2Vector() {g=0.075}
  n25 = NamedRerouteUsage()
  NamedRerouteDeclaration(None=MaterialFunctionCall!(UV=AppendVector(A=@n20, B=@n22), Speed=LinearInterpolate(A=@n23, B=@n24, Alpha=@n25), Mask=@n21) {materialFunction=/Game/Characters/Player/Male/Large/Bodies/M_LRG_PeacefulPoem_Pumped/Materials/MF_PeaceulPoemPumpedPanner}.Result) {name=SpikesTexture, nodeColor=<rgba 0.6406 0 1 1>}
  n26 = Distance(A=NamedRerouteUsage(), B=NamedRerouteUsage())
  n27 = ScalarParameter() {defaultValue=1, parameterName=Inner Size}
  n28 = VertexInterpolator(VS=PreSkinnedPosition())
  n29 = NamedRerouteDeclaration(None=Add(A=Divide(A=@n28.PS) {constB=10}, B=VectorParameter() {parameterName=PositionOffsetParam, defaultValue=<rgba 0 0 -12 1>}.RGB)) {name=ChestUV, nodeColor=<rgba 1 0 0.4453 1>}
  n30 = Multiply(A=Normalize(VectorInput=CameraVectorWS()), B=ScalarParameter() {defaultValue=1.5, parameterName=Inner Depth}) {constB=2}
  n31 = Multiply(A=NamedRerouteDeclaration(None=Transform(None=@n30) {transformSourceType=TRANSFORMSOURCE_World, transformType=TRANSFORM_Local}) {name=CameraDirection, nodeColor=<rgba 0 0.6719 1 1>}) {constB=-1}
  n32 = MaterialFunctionCall(Vector or UVs=ComponentMask(None=Add(A=@n29, B=@n31)) {b=true, r=true}) {materialFunction=&VectorToRadialValue}
  NamedRerouteDeclaration(None=AppendVector(A=Divide(A=@n26, B=@n27), B=@n32.Vector Converted to Angle)) {name=DepthRadialUV, nodeColor=<rgba 1 0.2812 0 1>}
  NamedRerouteDeclaration(None=@n18.G) {name=ClawsMask, nodeColor=<rgba 0.677 0 1 1>}
  n33 = GetMaterialAttributes(None=NamedRerouteUsage()) {attributeGetTypes [#1]}
  n34 = Add(A=Add(A=NamedRerouteUsage(), B=NamedRerouteUsage()), B=NamedRerouteUsage())
  n35 = VectorParameter() {parameterName=EmissiveColor, defaultValue=<rgba 0.2623 0.0033 0.7758 1>}
  NamedRerouteDeclaration(None=SetMaterialAttributes(MaterialAttributes=@n33.MaterialAttributes, Emissive Color=Multiply(A=@n34, B=@n35.RGB)) {attributeSetTypes [#1]}) {name=OutMaterialAttributes, nodeColor=<rgba 0 1 0.8594 1>}
  NamedRerouteDeclaration(None=@n32.Linear Distance) {name=InnerDot, nodeColor=<rgba 0 0.1562 1 1>}
  n36 = TextureSample!(UVs=Multiply(A=TextureCoordinate()) {constB=15}, Tex, Apply View MipBias) {
   samplerSource=SSM_Wrap_WorldGroupSettings
   samplerType=SAMPLERTYPE_Masks
   texture=&T_PeacefulPoem_Pumped_Mask
  }
  NamedRerouteDeclaration(None=@n36.RGB) {name=DefaultSampledTexture, nodeColor=<rgba 0.875 0 1 1>}
  n37 = PreSkinnedPosition()
  n38 = OneMinus(None=MaterialFunctionCall!(
    A=VertexInterpolator(VS=@n37).PS
    B=Constant3Vector() {constant=<rgba 0 0 90 0>}
    Bounds=Constant3Vector() {constant=<rgba 80 30 100 0>}
    Edge Falloff=Constant() {r=10}
   ) {materialFunction=/Engine/Functions/Engine_MaterialFunctions01/Gradient/BoxMask-3D}.Result)
  n39 = Power(Base=MaterialFunctionCall(Color=NamedRerouteUsage()) {materialFunction=&SplitComponents}.R)
  NamedRerouteDeclaration(None=Saturate(None=Multiply(A=@n38, B=@n39))) {name=CavitiesMask, nodeColor=<rgba 1 0.8906 0 1>}
  NamedRerouteDeclaration(None=@n33.Emissive Color) {name=InEmissive, nodeColor=<rgba 1 0.9141 0 1>}
  n40 = Time() {period=40100}
  n41 = Multiply(A=@n40) {constB=0.1}
  n42 = Multiply(A=Add(A=@n40) {constB=0.5}) {constB=0.2}
  n43 = Abs(None=Add(A=Sine(None=@n41), B=Sine(None=@n42) {period=0.5}))
  NamedRerouteDeclaration(None=LinearInterpolate(Alpha=Saturate(None=@n43)) {constA=0.5}) {name=Flicker, nodeColor=<rgba 0.875 0 1 1>}
  NamedRerouteDeclaration(None=@n1.Pre skinned Local Position) {name=PreSkinnedPosition, nodeColor=<rgba 0 0.4141 1 1>}
  Step(X=@n3) {constY=0.001}
  TextureSample() {texture=/Game/Characters/Player/Male/Large/Bodies/M_LRG_PeacefulPoem_Pumped/Textures/T_PeacefulPoemPumped_Body_E}
  TextureSample() {
   samplerType=SAMPLERTYPE_Masks
   texture=/Game/Characters/Player/Male/Large/Bodies/M_LRG_PeacefulPoem_Pumped/Textures/T_PeacefulPoemPumped_Body_FX
  }
 ]
}
