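# Physics asset for the armor mesh of the large male PeacefulPoem Pumped player body cosmetic, set up for its left and right skirt armor plates.
# Capsules cover the pelvis, thighs and a mirrored three-bone chain per plate, starting from a tiny base body.
# Each base rotates freely against the pelvis, the upper plate is tightly limited to its base, and the lower plate hinges on a single limited swing axis.
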PhysicsAsset /BRCosmetics/Characters/Player/Male/Large/Bodies/M_LRG_PeacefulPoem_Pumped/Meshes/M_LRG_PeacefulPoem_Pumped_Armor_Physics
bodies {
 dyn_skirt_armor_01_l [{
   shapeType=Capsule
   length=13.775
   radius=10.256
   shapeName=dyn_skirt_armor_01_r_capsule
   center=<14.5145 -7.097 0.1816>
   rotation=<rot -0.0135 -88.6252 90.4042>
  }]
 dyn_skirt_armor_01_r [{
   shapeType=Capsule
   length=13.775
   radius=10.256
   shapeName=dyn_skirt_armor_01_r_capsule
   center=<-14.5145 7.097 -0.1816>
   rotation=<rot -0.0135 -88.6252 -89.5958>
  }]
 dyn_skirt_armor_l [{
   shapeType=Capsule
   length=8.8094
   radius=9.7019
   shapeName=dyn_skirt_armor_r_capsule
   center=<10.2866 -6.1751 2.1001>
   rotation=<rot -4.0286 89.8871 91.7228>
  }]
 dyn_skirt_armor_l_base [{shapeType=Capsule, length=0.505, radius=0.505, shapeName=dyn_skirt_armor_r_base_capsule, rotation=<rot 0 -90 180>}]
 dyn_skirt_armor_r [{
   shapeType=Capsule
   length=7.3406
   radius=9.3269
   shapeName=dyn_skirt_armor_r_capsule
   center=<-8.2848 6.6521 -2.0737>
   rotation=<rot -4.0286 89.8871 -88.2772>
  }]
 dyn_skirt_armor_r_base [{shapeType=Capsule, length=0.505, radius=0.505, shapeName=dyn_skirt_armor_r_base_capsule, rotation=<rot 0 -90 0>}]
 pelvis [{shapeType=Capsule, length=2.9614, radius=15.7891, center=<-0.6221 -1.4164 0>}]
 thigh_l [{shapeType=Capsule, length=29.986, radius=10.3542, center=<-18.3568 -0.2426 3.8127>, rotation=<rot -6.571 91.2126 -86.6877>}]
 thigh_r [{shapeType=Capsule, length=29.986, radius=10.3542, center=<18.3568 0.2426 -3.8127>, rotation=<rot -6.571 91.2126 93.3123>}]
}
constraints [
 {bone1Name=thigh_l, bone2Name=pelvis, swing1LimitDegrees=45, swing2LimitDegrees=45, twistLimitDegrees=45}
 {bone1Name=dyn_skirt_armor_r, bone2Name=dyn_skirt_armor_r_base, swing1LimitDegrees=4, swing2LimitDegrees=4, twistLimitDegrees=4}
 {
  bone1Name=dyn_skirt_armor_r_base
  bone2Name=pelvis
  swing1LimitDegrees=45
  swing1Motion=Free
  swing2LimitDegrees=45
  swing2Motion=Free
  twistLimitDegrees=45
  twistMotion=Free
 }
 {
  bone1Name=dyn_skirt_armor_l_base
  bone2Name=pelvis
  swing1LimitDegrees=45
  swing1Motion=Free
  swing2LimitDegrees=45
  swing2Motion=Free
  twistLimitDegrees=45
  twistMotion=Free
 }
 {bone1Name=dyn_skirt_armor_l, bone2Name=dyn_skirt_armor_l_base, swing1LimitDegrees=4, swing2LimitDegrees=4, twistLimitDegrees=4}
 {
  bone1Name=dyn_skirt_armor_01_l
  bone2Name=dyn_skirt_armor_l
  swing1LimitDegrees=20
  swing2LimitDegrees=45
  swing2Motion=Locked
  twistLimitDegrees=45
  twistMotion=Locked
 }
 {bone1Name=thigh_r, bone2Name=pelvis, swing1LimitDegrees=45, swing2LimitDegrees=45, twistLimitDegrees=45}
 {
  bone1Name=dyn_skirt_armor_01_r
  bone2Name=dyn_skirt_armor_r
  swing1LimitDegrees=20
  swing2LimitDegrees=45
  swing2Motion=Locked
  twistLimitDegrees=45
  twistMotion=Locked
 }
]
