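# Control Rig for the PeacefulPoem Pumped large male body that procedurally deforms the outfit's shoulder pads, chest cloth, skirt, knee pads and built-in pickaxe bones.
# Separate functions drive weighted bone offsets from deformRig anim attributes and spherical pose readers on clavicles, upper arms and thighs.
# Weighted parent constraints anchor cloth and skirt bases, while parachute and pickaxe booleans adjust or gate some corrections.
# A skirt setup function also builds helper bones under a pelvis copy.
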ControlRigBlueprint /BRCosmetics/Characters/Player/Male/Large/Bodies/M_LRG_PeacefulPoem_Pumped/Meshes/M_LRG_PeacefulPoem_Pumped_CtrlRig
graphs [
 New Function_ContainedGraph {nodes [
   n1 = RigVMFunctionEntryNode()
   RigVMFunction_Sequence = &RigVMFunction_Sequence(ExecuteContext=@n1.ExecuteContext)
   &RigUnit_ModifyTransforms(
    ItemToModify="((Item=(Type=Bone,Name=\"deform_shoulderpad_l_base\"),Transform=(Rotation=(X=0,Y=0,Z=0,W=1),Translation=(X=0,Y=0,Z=0),Scale3D=(X=1,Y=1.1,Z=1))),(Item=(Type=Bone,Name=\"deform_shoulderpad_r_base\"),Transform=(Rotation=(X=0,Y=0,Z=0,W=1),Translation=(X=0,Y=0,Z=0) ...(truncated)"
    Weight=1
    WeightMinimum=0
    ExecutePin=@RigVMFunction_Sequence.A
   )
   > RigVMFunction_Sequence_1 = RigVMAggregateNode()
   n2 = RigVMDispatchNode(BoneName=root, Default=0, Name=deformRig_l_upperarm_up_alpha) {template="DISPATCH_RigDispatch_GetAnimAttribute(in Name,in BoneName,in Default,out Value,out Found)"}
   &RigUnit_ModifyTransforms(
    ItemToModify=((Item=(Type=Bone,Name="deform_shoulderpad_l"),Transform=(Rotation=(X=0.1913,Y=-0.1216,Z=0.0948,W=0.9694),Translation=(X=2.2369,Y=-2.5953,Z=2.1447),Scale3D=(X=0.9,Y=0.9,Z=0.9))))
    WeightMinimum=0
    Weight=@n2.Value
    ExecutePin=@RigVMFunction_Sequence_1.A
   )
   RigVMFunction_MathBoolToFloat = &RigVMFunction_MathBoolToFloat(Value=RigVMVariableNode() {variable=IsParachuteOpen}.Value)
   Multiply = &RigVMFunction_MathDoubleMul(B=@n2.Value, A=@RigVMFunction_MathBoolToFloat.Result)
   > &RigUnit_ModifyTransforms(
    ItemToModify=((Item=(Type=Bone,Name="deform_shoulderpad_l"),Transform=(Rotation=(X=-0.3714,Y=0.0835,Z=-0.0238,W=0.9244),Translation=(X=3.5,Y=0,Z=-2.2797),Scale3D=(X=0.9,Y=0.9,Z=0.9))))
    WeightMinimum=0
    Weight=@Multiply.Result
   )
   n3 = RigVMDispatchNode(BoneName=root, Default=0, Name=deformRig_r_upperarm_up_alpha) {template="DISPATCH_RigDispatch_GetAnimAttribute(in Name,in BoneName,in Default,out Value,out Found)"}
   &RigUnit_ModifyTransforms(
    ItemToModify=((Item=(Type=Bone,Name="deform_shoulderpad_r"),Transform=(Rotation=(X=0.2902,Y=0.0256,Z=-0.0316,W=0.9561),Translation=(X=-0.8389,Y=2.4681,Z=-2.1118),Scale3D=(X=0.9,Y=0.9,Z=0.9))))
    WeightMinimum=0
    Weight=@n3.Value
    ExecutePin=@RigVMFunction_Sequence_1.B
   )
   > &RigUnit_ModifyTransforms(
    ItemToModify=((Item=(Type=Bone,Name="deform_shoulderpad_r"),Transform=(Rotation=(X=-0.3839,Y=0.0439,Z=-0.0075,W=0.9223),Translation=(X=-4.02,Y=-1.6305,Z=-0.4766),Scale3D=(X=0.9,Y=0.9,Z=0.9))))
    WeightMinimum=0
    Weight=&RigVMFunction_MathDoubleMul(B=@n3.Value, A=@RigVMFunction_MathBoolToFloat.Result).Result
   )
   &RigUnit_ModifyTransforms(
    ItemToModify=((Item=(Type=Bone,Name="deform_shoulderpad_l"),Transform=(Rotation=(X=0.0485,Y=0.164,Z=-0.0108,W=0.9852),Translation=(X=0.0783,Y=-1.4996,Z=2.0461),Scale3D=(X=1,Y=1,Z=1))))
    WeightMinimum=0
    Weight=RigVMDispatchNode(BoneName=root, Default=0, Name=deformRig_l_upperarm_fr_alpha) {template="DISPATCH_RigDispatch_GetAnimAttribute(in Name,in BoneName,in Default,out Value,out Found)"}.Value
    ExecutePin=@RigVMFunction_Sequence_1.C
   )
   &RigUnit_ModifyTransforms(
    ItemToModify=((Item=(Type=Bone,Name="deform_shoulderpad_r"),Transform=(Rotation=(X=0.0492,Y=0.1648,Z=0.0046,W=0.9851),Translation=(X=-0.9097,Y=2.5712,Z=-1.8879),Scale3D=(X=1,Y=1,Z=1))))
    WeightMinimum=0
    Weight=RigVMDispatchNode(BoneName=root, Default=0, Name=deformRig_r_upperarm_fr_alpha) {template="DISPATCH_RigDispatch_GetAnimAttribute(in Name,in BoneName,in Default,out Value,out Found)"}.Value
    ExecutePin=@RigVMFunction_Sequence_1.D
   )
   n4 = &RigUnit_SphericalPoseReader(
    ActiveRegionScaleFactors=(PositiveWidth=1,NegativeWidth=1,PositiveHeight=1,NegativeHeight=1)
    ActiveRegionSize=0.05
    Debug=(bDrawDebug=false,bDraw2D=False,bDrawLocalAxes=False,DebugScale=25,DebugSegments=20,DebugThickness=0.25)
    DriverAxis=(X=1,Y=0,Z=0)
    DriverItem=(Type=Bone,Name="clavicle_l")
    FalloffRegionScaleFactors=(PositiveWidth=1,NegativeWidth=1,PositiveHeight=1,NegativeHeight=1)
    FalloffSize=0.2
    FlipHeightScaling=False
    FlipWidthScaling=False
    OptionalParentItem=(Type=Bone,Name="None")
    RotationOffset=(X=0,Y=-30,Z=0)
    ExecutePin=@RigVMFunction_Sequence_1.E
   )
   > &RigUnit_ModifyTransforms(
    ItemToModify=((Item=(Type=Bone,Name="deform_shoulderpad_l"),Transform=(Rotation=(X=0.0079,Y=-0.0822,Z=-0.0948,W=0.9921),Translation=(X=5.297,Y=-0.0898,Z=-4.1484),Scale3D=(X=1,Y=1,Z=1))))
    WeightMinimum=0
    Weight=@n4.OutputParam
   )
   > RigVMVariableNode(Value=@n4.OutputParam) {variable=l_clav_spr}
   n5 = &RigUnit_SphericalPoseReader(
    ActiveRegionScaleFactors=(PositiveWidth=1,NegativeWidth=1,PositiveHeight=1,NegativeHeight=1)
    ActiveRegionSize=0.05
    Debug=(bDrawDebug=false,bDraw2D=False,bDrawLocalAxes=False,DebugScale=25,DebugSegments=20,DebugThickness=0.25)
    DriverAxis=(X=-1,Y=0,Z=0)
    DriverItem=(Type=Bone,Name="clavicle_r")
    FalloffRegionScaleFactors=(PositiveWidth=1,NegativeWidth=1,PositiveHeight=1,NegativeHeight=1)
    FalloffSize=0.2
    FlipHeightScaling=False
    FlipWidthScaling=False
    OptionalParentItem=(Type=Bone,Name="None")
    RotationOffset=(X=180,Y=-30,Z=0)
    ExecutePin=@RigVMFunction_Sequence_1.F
   )
   > &RigUnit_ModifyTransforms(
    ItemToModify=((Item=(Type=Bone,Name="deform_shoulderpad_r"),Transform=(Rotation=(X=-0.0149,Y=-0.2089,Z=-0.013,W=0.9777),Translation=(X=-5.297,Y=0,Z=4.1484),Scale3D=(X=1,Y=1,Z=1))))
    WeightMinimum=0
    Weight=@n5.OutputParam
   )
   > RigVMVariableNode(Value=@n5.OutputParam) {variable=r_clav_spr}
   n6 = &RigUnit_SphericalPoseReader(
    ActiveRegionScaleFactors=(PositiveWidth=1,NegativeWidth=1,PositiveHeight=1,NegativeHeight=1)
    ActiveRegionSize=0.1
    Debug=(bDrawDebug=false,bDraw2D=False,bDrawLocalAxes=False,DebugScale=25,DebugSegments=20,DebugThickness=0.25)
    DriverAxis=(X=1,Y=0,Z=0)
    DriverItem=(Type=Bone,Name="upperarm_l")
    FalloffRegionScaleFactors=(PositiveWidth=1,NegativeWidth=1,PositiveHeight=1,NegativeHeight=1)
    FalloffSize=0.2
    FlipHeightScaling=False
    FlipWidthScaling=False
    OptionalParentItem=(Type=Bone,Name="spine_05")
    RotationOffset=(X=0,Y=-150,Z=0)
    ExecutePin=@RigVMFunction_Sequence_1.G
   )
   > &RigUnit_ModifyTransforms(
    ItemToModify=((Item=(Type=Bone,Name="deform_shoulderpad_l"),Transform=(Rotation=(X=0,Y=0,Z=0,W=1),Translation=(X=0,Y=-1.6713,Z=0),Scale3D=(X=1,Y=1,Z=1))))
    WeightMinimum=0
    Weight=@n6.OutputParam
   )
   n7 = &RigUnit_SphericalPoseReader(
    ActiveRegionScaleFactors=(PositiveWidth=1,NegativeWidth=1,PositiveHeight=1,NegativeHeight=1)
    ActiveRegionSize=0.1
    Debug=(bDrawDebug=false,bDraw2D=False,bDrawLocalAxes=False,DebugScale=25,DebugSegments=20,DebugThickness=0.25)
    DriverAxis=(X=-1,Y=0,Z=0)
    DriverItem=(Type=Bone,Name="upperarm_r")
    FalloffRegionScaleFactors=(PositiveWidth=1,NegativeWidth=1,PositiveHeight=1,NegativeHeight=1)
    FalloffSize=0.2
    FlipHeightScaling=False
    FlipWidthScaling=False
    OptionalParentItem=(Type=Bone,Name="spine_05")
    RotationOffset=(X=0,Y=50,Z=0)
    ExecutePin=@RigVMFunction_Sequence_1.H
   )
   > &RigUnit_ModifyTransforms(
    ItemToModify=((Item=(Type=Bone,Name="deform_shoulderpad_r"),Transform=(Rotation=(X=0,Y=0,Z=0,W=1),Translation=(X=0,Y=1.6713,Z=0),Scale3D=(X=1,Y=1,Z=1))))
    WeightMinimum=0
    Weight=@n7.OutputParam
   )
   RigVMFunctionReturnNode(
    ExecuteContext=@RigVMFunction_Sequence.B
    L Clavicle SPR=RigVMVariableNode() {variable=l_clav_spr}.Value
    R Clavicle SPR=RigVMVariableNode() {variable=r_clav_spr}.Value
   )
  ]}
 New Function_ContainedGraph {nodes [
   n1 = RigVMFunctionEntryNode()
   > &RigUnit_ParentConstraint(
    AdvancedSettings=(RotationOrderForFilter=XZY)
    Child=(Type=Bone,Name="dyn_pec_cloth_l_01")
    Parents="((Item=(Type=Bone,Name=\"upperarm_twist_01_l\"),Weight=0.0129),(Item=(Type=Bone,Name=\"deform_upperarm_up_l\"),Weight=0.0171),(Item=(Type=Bone,Name=\"deform_upperarm_fr_l\"),Weight=0.0256),(Item=(Type=Bone,Name=\"spine_03\"),Weight=0.0536),(Item=(Type=Bone,Name=\"d ...(truncated)"
    Weight=1
   )
   > &RigUnit_ParentConstraint(
    AdvancedSettings=(RotationOrderForFilter=XZY)
    Child=(Type=Bone,Name="dyn_pec_cloth_r_01")
    Parents="((Item=(Type=Bone,Name=\"deform_upperarm_up_r\"),Weight=0.016),(Item=(Type=Bone,Name=\"upperarm_twist_01_r\"),Weight=0.018),(Item=(Type=Bone,Name=\"deform_upperarm_fr_r\"),Weight=0.0307),(Item=(Type=Bone,Name=\"spine_03\"),Weight=0.0591),(Item=(Type=Bone,Name=\"def ...(truncated)"
    Weight=1
   )
   > &RigUnit_ParentConstraint(
    AdvancedSettings=(RotationOrderForFilter=XZY)
    Child=(Type=Bone,Name="dyn_pec_cloth_l_02")
    Parents="((Item=(Type=Bone,Name=\"spine_02\"),Weight=0.0284),(Item=(Type=Bone,Name=\"upperarm_twist_01_l\"),Weight=0.0546),(Item=(Type=Bone,Name=\"deform_upperarm_fr_l\"),Weight=0.0749),(Item=(Type=Bone,Name=\"spine_03\"),Weight=0.0802),(Item=(Type=Bone,Name=\"spine_04\"),We ...(truncated)"
    Weight=0
   )
   > &RigUnit_ParentConstraint(
    AdvancedSettings=(RotationOrderForFilter=XZY)
    Child=(Type=Bone,Name="dyn_pec_cloth_r_02")
    Parents="((Item=(Type=Bone,Name=\"spine_02\"),Weight=0.0284),(Item=(Type=Bone,Name=\"upperarm_twist_01_r\"),Weight=0.0546),(Item=(Type=Bone,Name=\"deform_upperarm_fr_r\"),Weight=0.0749),(Item=(Type=Bone,Name=\"spine_03\"),Weight=0.0802),(Item=(Type=Bone,Name=\"spine_04\"),We ...(truncated)"
    Weight=0
   )
   > &RigUnit_ModifyTransforms(
    ItemToModify=((Item=(Type=Bone,Name="dyn_pec_cloth_l_01"),Transform=(Rotation=(X=0,Y=0,Z=0,W=1),Translation=(X=0,Y=0,Z=1),Scale3D=(X=1,Y=1,Z=1))))
    Weight=1
    WeightMinimum=0
   )
   > &RigUnit_ModifyTransforms(
    ItemToModify=((Item=(Type=Bone,Name="dyn_pec_cloth_r_01"),Transform=(Rotation=(X=0,Y=0,Z=0,W=1),Translation=(X=0,Y=0,Z=-1),Scale3D=(X=1,Y=1,Z=1))))
    Weight=1
    WeightMinimum=0
   )
   > &RigUnit_ModifyTransforms(
    ItemToModify=((Item=(Type=Bone,Name="dyn_pec_cloth_l_01"),Transform=(Rotation=(X=0,Y=0.331,Z=0,W=0.9436),Translation=(X=0,Y=0,Z=1.2),Scale3D=(X=1,Y=1,Z=1))))
    WeightMinimum=0
    Weight=RigVMDispatchNode(BoneName=root, Default=0, Name=deformRig_l_upperarm_up_alpha) {template="DISPATCH_RigDispatch_GetAnimAttribute(in Name,in BoneName,in Default,out Value,out Found)"}.Value
   )
   > &RigUnit_ModifyTransforms(
    ItemToModify=((Item=(Type=Bone,Name="dyn_pec_cloth_r_01"),Transform=(Rotation=(X=0,Y=0.2627,Z=0,W=0.9649),Translation=(X=0,Y=0,Z=-1.2),Scale3D=(X=1,Y=1,Z=1))))
    WeightMinimum=0
    Weight=RigVMDispatchNode(BoneName=root, Default=0, Name=deformRig_r_upperarm_up_alpha) {template="DISPATCH_RigDispatch_GetAnimAttribute(in Name,in BoneName,in Default,out Value,out Found)"}.Value
   )
   > &RigUnit_ModifyTransforms(
    ItemToModify=((Item=(Type=Bone,Name="dyn_pec_cloth_l_01"),Transform=(Rotation=(X=-0.0174,Y=-0.0903,Z=0.188,W=0.9779),Translation=(X=0,Y=0,Z=0),Scale3D=(X=1,Y=1,Z=1))))
    WeightMinimum=0
    Weight=RigVMDispatchNode(BoneName=root, Default=0, Name=deformRig_l_upperarm_fr_alpha) {template="DISPATCH_RigDispatch_GetAnimAttribute(in Name,in BoneName,in Default,out Value,out Found)"}.Value
   )
   > &RigUnit_ModifyTransforms(
    ItemToModify=((Item=(Type=Bone,Name="dyn_pec_cloth_r_01"),Transform=(Rotation=(X=-0.0174,Y=-0.0903,Z=0.188,W=0.9779),Translation=(X=0,Y=0,Z=0),Scale3D=(X=1,Y=1,Z=1))))
    WeightMinimum=0
    Weight=RigVMDispatchNode(BoneName=root, Default=0, Name=deformRig_r_upperarm_fr_alpha) {template="DISPATCH_RigDispatch_GetAnimAttribute(in Name,in BoneName,in Default,out Value,out Found)"}.Value
   )
   > &RigUnit_ModifyTransforms(
    ItemToModify=((Item=(Type=Bone,Name="dyn_pec_cloth_l_01"),Transform=(Rotation=(X=-0.0697,Y=-0.0435,Z=-0.003,W=0.9966),Translation=(X=0,Y=0,Z=0),Scale3D=(X=1,Y=1,Z=1))))
    WeightMinimum=0
    Weight=@n1.L Clavicle SPR
   )
   > &RigUnit_ModifyTransforms(
    ItemToModify=((Item=(Type=Bone,Name="dyn_pec_cloth_r_01"),Transform=(Rotation=(X=-0.0697,Y=-0.0354,Z=-0.0036,W=0.9969),Translation=(X=0,Y=0,Z=0),Scale3D=(X=1,Y=1,Z=1))))
    WeightMinimum=0
    Weight=@n1.R Clavicle SPR
   )
   > &RigUnit_ModifyTransforms(
    ItemToModify=((Item=(Type=Bone,Name="dyn_pec_cloth_l_01"),Transform=(Rotation=(X=0.0028,Y=-0.1758,Z=0.1661,W=0.9703),Translation=(X=0,Y=0,Z=1),Scale3D=(X=1,Y=1,Z=1))))
    WeightMinimum=0
    Weight=RigVMDispatchNode(BoneName=root, Default=0, Name=deformRig_l_pec_alpha) {template="DISPATCH_RigDispatch_GetAnimAttribute(in Name,in BoneName,in Default,out Value,out Found)"}.Value
   )
   > &RigUnit_ModifyTransforms(
    ItemToModify=((Item=(Type=Bone,Name="dyn_pec_cloth_r_01"),Transform=(Rotation=(X=-0.1718,Y=-0.2028,Z=0.1318,W=0.955),Translation=(X=0,Y=0,Z=-1.2),Scale3D=(X=1,Y=1,Z=1))))
    WeightMinimum=0
    Weight=RigVMDispatchNode(BoneName=root, Default=0, Name=deformRig_R_pec_alpha) {template="DISPATCH_RigDispatch_GetAnimAttribute(in Name,in BoneName,in Default,out Value,out Found)"}.Value
   )
   > RigVMFunctionReturnNode()
   Get Delta Ref = RigVMFunctionReferenceNode(ExecuteContext=(), Item=(Type=Bone,Name="clavicle_l"))
   DISPATCH_RigVMDispatch_BreakStruct = RigVMDispatchNode(Struct=@"Get Delta Ref".Result.Rotation) {template="DISPATCH_RigVMDispatch_BreakStruct(in Struct,out Elements)"}
   Remap = &RigVMFunction_MathDoubleRemap(SourceMaximum=1, SourceMinimum=0, TargetMaximum=1, TargetMinimum=0, bClamp=False, Value=@DISPATCH_RigVMDispatch_BreakStruct.Elements.Pitch)
   > &RigUnit_ModifyTransforms(
    ItemToModify=((Item=(Type=Bone,Name="dyn_pec_cloth_r_01"),Transform=(Rotation=(X=0,Y=0,Z=0,W=1),Translation=(X=0,Y=0,Z=0),Scale3D=(X=1,Y=1,Z=1))))
    WeightMinimum=0
    Weight=@Remap.Result
   )
   RigVMVariableNode() {variable=IsParachuteOpen}
  ]}
 New Function_ContainedGraph {nodes [
   n1 = RigVMFunctionEntryNode()
   n2 = &RigVMFunction_Sequence(ExecuteContext=@n1.ExecuteContext)
   &RigUnit_ParentConstraint(
    AdvancedSettings=(RotationOrderForFilter=XZY)
    Child=(Type=Bone,Name="dyn_skirt_front_r_01_base")
    Parents="((Item=(Type=Bone,Name=\"deform_groin_r\"),Weight=0.0042),(Item=(Type=Bone,Name=\"spine_04\"),Weight=0.0141),(Item=(Type=Bone,Name=\"spine_03\"),Weight=0.1413),(Item=(Type=Bone,Name=\"spine_01\"),Weight=0.171),(Item=(Type=Bone,Name=\"spine_02\"),Weight=0.2204),(Item ...(truncated)"
    Weight=1
    ExecutePin=@n2.A
   )
   > &RigUnit_ParentConstraint(
    AdvancedSettings=(RotationOrderForFilter=XZY)
    Child=(Type=Bone,Name="dyn_skirt_side_r_01_base")
    Parents="((Item=(Type=Bone,Name=\"thigh_twist_01_r\"),Weight=0.002),(Item=(Type=Bone,Name=\"spine_04\"),Weight=0.0022),(Item=(Type=Bone,Name=\"spine_03\"),Weight=0.0976),(Item=(Type=Bone,Name=\"spine_01\"),Weight=0.1314),(Item=(Type=Bone,Name=\"spine_02\"),Weight=0.1665),(It ...(truncated)"
    Weight=1
   )
   > &RigUnit_ParentConstraint(
    AdvancedSettings=(RotationOrderForFilter=XZY)
    Child=(Type=Bone,Name="dyn_skirt_rear_r_01_base")
    Parents=((Item=(Type=Bone,Name="spine_04"),Weight=0.0074),(Item=(Type=Bone,Name="spine_03"),Weight=0.083),(Item=(Type=Bone,Name="spine_01"),Weight=0.1097),(Item=(Type=Bone,Name="spine_02"),Weight=0.117),(Item=(Type=Bone,Name="pelvis"),Weight=0.6829))
    Weight=1
   )
   > &RigUnit_ParentConstraint(
    AdvancedSettings=(RotationOrderForFilter=XZY)
    Child=(Type=Bone,Name="dyn_skirt_rear_l_01_base")
    Parents=((Item=(Type=Bone,Name="spine_04"),Weight=0.0055),(Item=(Type=Bone,Name="spine_03"),Weight=0.077),(Item=(Type=Bone,Name="spine_01"),Weight=0.104),(Item=(Type=Bone,Name="spine_02"),Weight=0.1177),(Item=(Type=Bone,Name="pelvis"),Weight=0.6957))
    Weight=1
   )
   > &RigUnit_ParentConstraint(
    AdvancedSettings=(RotationOrderForFilter=XZY)
    Child=(Type=Bone,Name="dyn_skirt_side_l_01_base")
    Parents="((Item=(Type=Bone,Name=\"spine_04\"),Weight=0.0023),(Item=(Type=Bone,Name=\"thigh_twist_01_l\"),Weight=0.0023),(Item=(Type=Bone,Name=\"spine_03\"),Weight=0.0994),(Item=(Type=Bone,Name=\"spine_01\"),Weight=0.1334),(Item=(Type=Bone,Name=\"spine_02\"),Weight=0.1693),(I ...(truncated)"
    Weight=1
   )
   > &RigUnit_ParentConstraint(
    AdvancedSettings=(RotationOrderForFilter=XZY)
    Child=(Type=Bone,Name="dyn_skirt_front_l_01_base")
    Parents="((Item=(Type=Bone,Name=\"deform_groin_l\"),Weight=0.0047),(Item=(Type=Bone,Name=\"spine_04\"),Weight=0.0128),(Item=(Type=Bone,Name=\"spine_03\"),Weight=0.1281),(Item=(Type=Bone,Name=\"spine_01\"),Weight=0.1613),(Item=(Type=Bone,Name=\"spine_02\"),Weight=0.2035),(Ite ...(truncated)"
    Weight=1
   )
   > &RigUnit_ParentConstraint(
    AdvancedSettings=(RotationOrderForFilter=XZY)
    Child=(Type=Bone,Name="dyn_skirt_armor_l_base")
    Parents="((Item=(Type=Bone,Name=\"thigh_twist_01_l\"),Weight=0.0002),(Item=(Type=Bone,Name=\"deform_lat_l\"),Weight=0.0032),(Item=(Type=Bone,Name=\"clavicle_l\"),Weight=0.0096),(Item=(Type=Bone,Name=\"spine_04\"),Weight=0.0235),(Item=(Type=Bone,Name=\"spine_03\"),Weight=0.18 ...(truncated)"
    Weight=1
   )
   > &RigUnit_ParentConstraint(
    AdvancedSettings=(RotationOrderForFilter=XZY)
    Child=(Type=Bone,Name="dyn_skirt_armor_r_base")
    Parents="((Item=(Type=Bone,Name=\"upperarm_twist_01_r\"),Weight=0.0006),(Item=(Type=Bone,Name=\"deform_lat_r\"),Weight=0.0044),(Item=(Type=Bone,Name=\"clavicle_r\"),Weight=0.012),(Item=(Type=Bone,Name=\"spine_04\"),Weight=0.0279),(Item=(Type=Bone,Name=\"spine_01\"),Weight=0. ...(truncated)"
    Weight=1
   )
   > &RigUnit_ModifyTransforms(
    ItemToModify=((Item=(Type=Bone,Name="dyn_skirt_armor_l_base"),Transform=(Rotation=(X=0,Y=0,Z=0,W=1),Translation=(X=-0.4503,Y=0,Z=-0.9472),Scale3D=(X=1,Y=1,Z=1))))
    Weight=1
    WeightMinimum=0
   )
   > RigVMFunction_Sequence_1 = RigVMAggregateNode()
   n3 = &RigUnit_SphericalPoseReader(
    ActiveRegionScaleFactors=(PositiveWidth=3.5,NegativeWidth=0.5,PositiveHeight=3,NegativeHeight=2)
    ActiveRegionSize=0.1
    Debug=(bDrawDebug=false,DebugScale=25,DebugSegments=20,DebugThickness=0.25)
    DriverAxis=(X=-1,Y=0,Z=0)
    DriverItem=(Type=Bone,Name="thigh_l")
    FalloffRegionScaleFactors=(PositiveWidth=1,NegativeWidth=1,PositiveHeight=1,NegativeHeight=1)
    FalloffSize=0.65
    FlipHeightScaling=true
    FlipWidthScaling=false
    OptionalParentItem=(Type=Bone,Name="spine_01")
    RotationOffset=(X=90,Y=0,Z=-30)
    ExecutePin=@RigVMFunction_Sequence_1.B
   )
   RigVMFunction_AnimEvalRichCurve = &RigVMFunction_AnimEvalRichCurve(
    Curve=(EditorCurveData=(Keys=((),(InterpMode=RCIM_Cubic,Time=0.399,Value=0.3349,ArriveTangent=1,LeaveTangent=1),(Time=1,Value=1))))
    SourceMinimum=0
    TargetMinimum=0
    Value=@n3.OutputParam
   )
   > &RigUnit_ModifyTransforms(
    ItemToModify="((Item=(Type=Bone,Name=\"dyn_skirt_front_l_01_base\"),Transform=(Rotation=(X=0.2282,Y=0.2763,Z=-0.8736,W=0.3293),Translation=(X=4.0716,Y=0.3052,Z=-1.1891),Scale3D=(X=1,Y=1,Z=1))),(Item=(Type=Bone,Name=\"dyn_skirt_side_l_01_base\"),Transform=(Rotation=(X=0.1813 ...(truncated)"
    WeightMinimum=0
    Weight=@RigVMFunction_AnimEvalRichCurve.Result
   )
   > n4 = &RigUnit_SphericalPoseReader(
    ActiveRegionScaleFactors=(PositiveWidth=2,NegativeWidth=1,PositiveHeight=1,NegativeHeight=2)
    ActiveRegionSize=0.1
    Debug=(bDrawDebug=false,bDraw2D=False,bDrawLocalAxes=False,DebugScale=25,DebugSegments=20,DebugThickness=0.25)
    DriverAxis=(X=-1,Y=0,Z=0)
    DriverItem=(Type=Bone,Name="thigh_l")
    FalloffRegionScaleFactors=(PositiveWidth=1,NegativeWidth=1,PositiveHeight=1,NegativeHeight=1)
    FalloffSize=0.43
    FlipHeightScaling=true
    FlipWidthScaling=false
    OptionalParentItem=(Type=Bone,Name="pelvis")
    RotationOffset=(X=180,Y=0,Z=0)
   )
   > &RigUnit_ModifyTransforms(
    ItemToModify="((Item=(Type=Bone,Name=\"dyn_skirt_front_l_01_base\"),Transform=(Rotation=(X=0,Y=0,Z=0.1685,W=0.9857),Translation=(X=0,Y=0,Z=0),Scale3D=(X=1,Y=1,Z=1))),(Item=(Type=Bone,Name=\"dyn_skirt_side_l_01_base\"),Transform=(Rotation=(X=-0.1173,Y=-0.1583,Z=-0.5837,W=0.7 ...(truncated)"
    WeightMinimum=0
    Weight=@n4.OutputParam
   )
   > n5 = &RigUnit_SphericalPoseReader(
    ActiveRegionScaleFactors=(PositiveWidth=2,NegativeWidth=1,PositiveHeight=1,NegativeHeight=1)
    ActiveRegionSize=0.1
    Debug=(bDrawDebug=false,bDraw2D=False,bDrawLocalAxes=False,DebugScale=25,DebugSegments=20,DebugThickness=0.25)
    DriverAxis=(X=-1,Y=0,Z=0)
    DriverItem=(Type=Bone,Name="thigh_l")
    FalloffRegionScaleFactors=(PositiveWidth=1,NegativeWidth=1,PositiveHeight=1,NegativeHeight=1)
    FalloffSize=0.3
    FlipHeightScaling=true
    FlipWidthScaling=false
    OptionalParentItem=(Type=Bone,Name="spine_01")
    RotationOffset=(X=140,Y=0,Z=10)
   )
   > &RigUnit_ModifyTransforms(
    ItemToModify="((Item=(Type=Bone,Name=\"dyn_skirt_front_l_01_base\"),Transform=(Rotation=(X=0.1023,Y=-0.5622,Z=-0.0453,W=0.8194),Translation=(X=1.2535,Y=0,Z=-1.4882),Scale3D=(X=1,Y=2.375,Z=1.1676))),(Item=(Type=Bone,Name=\"dyn_skirt_side_l_01_base\"),Transform=(Rotation=(X=0 ...(truncated)"
    WeightMinimum=0
    Weight=@n5.OutputParam
   )
   > n6 = &RigUnit_SphericalPoseReader(
    ActiveRegionScaleFactors=(PositiveWidth=1,NegativeWidth=1,PositiveHeight=1,NegativeHeight=1)
    ActiveRegionSize=0.1
    Debug=(bDrawDebug=false,bDraw2D=False,bDrawLocalAxes=False,DebugScale=25,DebugSegments=20,DebugThickness=0.25)
    DriverAxis=(X=-1,Y=0,Z=0)
    DriverItem=(Type=Bone,Name="thigh_l")
    FalloffRegionScaleFactors=(PositiveWidth=1,NegativeWidth=1,PositiveHeight=1,NegativeHeight=1)
    FalloffSize=0.5
    FlipHeightScaling=False
    FlipWidthScaling=False
    OptionalParentItem=(Type=Bone,Name="None")
    RotationOffset=(X=270,Y=0,Z=0)
   )
   > &RigUnit_ModifyTransforms(
    ItemToModify="((Item=(Type=Bone,Name=\"dyn_skirt_rear_l_01_base\"),Transform=(Rotation=(X=-0.0469,Y=-0.0815,Z=-0.5644,W=0.8201),Translation=(X=0,Y=0,Z=0),Scale3D=(X=1,Y=1,Z=1))),(Item=(Type=Bone,Name=\"dyn_skirt_front_l_01_base\"),Transform=(Rotation=(X=0,Y=0,Z=0,W=1),Trans ...(truncated)"
    WeightMinimum=0
    Weight=@n6.OutputParam
   )
   > n7 = &RigUnit_SphericalPoseReader(
    ActiveRegionScaleFactors=(PositiveWidth=1,NegativeWidth=2,PositiveHeight=1,NegativeHeight=2)
    ActiveRegionSize=0.12
    Debug=(bDrawDebug=false,DebugScale=25,DebugSegments=20,DebugThickness=0.25)
    DriverAxis=(X=-1,Y=0,Z=0)
    DriverItem=(Type=Bone,Name="thigh_l")
    FalloffRegionScaleFactors=(PositiveWidth=1,NegativeWidth=1,PositiveHeight=1,NegativeHeight=1)
    FalloffSize=0.35
    FlipHeightScaling=false
    FlipWidthScaling=true
    OptionalParentItem=(Type=Bone,Name="spine_01")
    RotationOffset=(X=50,Y=0,Z=20)
   )
   > &RigUnit_ModifyTransforms(
    ItemToModify="((Item=(Type=Bone,Name=\"dyn_skirt_front_l_01_base\"),Transform=(Rotation=(X=-0.0885,Y=0.3573,Z=-0.0296,W=0.9293),Translation=(X=-2.0955,Y=-1.9633,Z=-1.7936),Scale3D=(X=1,Y=1,Z=1))),(Item=(Type=Bone,Name=\"dyn_skirt_side_l_01_base\"),Transform=(Rotation=(X=0,Y ...(truncated)"
    WeightMinimum=0
    Weight=@n7.OutputParam
   )
   n8 = &RigUnit_SphericalPoseReader(
    ActiveRegionScaleFactors=(PositiveWidth=3.5,NegativeWidth=0.5,PositiveHeight=2,NegativeHeight=3)
    ActiveRegionSize=0.1
    Debug=(bDrawDebug=false,DebugScale=25,DebugSegments=20,DebugThickness=0.25)
    DriverAxis=(X=1,Y=0,Z=0)
    DriverItem=(Type=Bone,Name="thigh_r")
    FalloffRegionScaleFactors=(PositiveWidth=1,NegativeWidth=1,PositiveHeight=1,NegativeHeight=0.5)
    FalloffSize=0.83
    FlipHeightScaling=false
    FlipWidthScaling=true
    OptionalParentItem=(Type=Bone,Name="spine_01")
    RotationOffset=(X=-90,Y=-10,Z=-50)
    ExecutePin=@RigVMFunction_Sequence_1.A
   )
   n9 = &RigVMFunction_AnimEvalRichCurve(
    Curve=(EditorCurveData=(Keys=((),(InterpMode=RCIM_Cubic,Time=0.663,Value=0.4604,ArriveTangent=1,LeaveTangent=1),(Time=1,Value=1))))
    SourceMinimum=0
    TargetMinimum=0
    Value=@n8.OutputParam
   )
   Add = &RigVMFunction_MathDoubleAdd(A=@n9.Result, B=@n3.OutputParam)
   Remap = &RigVMFunction_MathDoubleRemap(SourceMaximum=2, SourceMinimum=0.8, TargetMaximum=1, TargetMinimum=0, Value=@Add.Result)
   &RigUnit_ModifyTransforms(
    ItemToModify="((Item=(Type=Bone,Name=\"dyn_skirt_front_r_01_base\"),Transform=(Rotation=(X=-0.023,Y=0.0572,Z=0.1121,W=0.9918),Translation=(X=0,Y=0,Z=0),Scale3D=(X=1,Y=1,Z=1))),(Item=(Type=Bone,Name=\"dyn_skirt_front_l_01_base\"),Transform=(Rotation=(X=-0.1628,Y=-0.3536,Z=0. ...(truncated)"
    WeightMinimum=0
    Weight=@Remap.Result
    ExecutePin=@RigVMFunction_Sequence_1.C
   )
   RigVMFunctionReturnNode(ExecuteContext=@n2.B)
   n10 = &RigUnit_SphericalPoseReader(
    ActiveRegionScaleFactors=(PositiveWidth=2,NegativeWidth=2,PositiveHeight=1,NegativeHeight=2)
    ActiveRegionSize=0.07
    Debug=(bDrawDebug=false,DebugScale=25,DebugSegments=20,DebugThickness=0.25)
    DriverAxis=(X=1,Y=0,Z=0)
    DriverItem=(Type=Bone,Name="thigh_r")
    FalloffRegionScaleFactors=(PositiveWidth=1,NegativeWidth=1,PositiveHeight=1,NegativeHeight=1)
    FalloffSize=0.25
    FlipHeightScaling=false
    FlipWidthScaling=true
    OptionalParentItem=(Type=Bone,Name="spine_01")
    RotationOffset=(X=-140,Y=0,Z=0)
    ExecutePin=@n14.ExecutePin
   )
   > &RigUnit_ModifyTransforms(
    ItemToModify="((Item=(Type=Bone,Name=\"dyn_skirt_front_r_01_base\"),Transform=(Rotation=(X=-0.0793,Y=-0.4214,Z=-0.0124,W=0.9033),Translation=(X=0,Y=0,Z=0),Scale3D=(X=1,Y=1,Z=1.5053))),(Item=(Type=Bone,Name=\"dyn_skirt_side_r_01_base\"),Transform=(Rotation=(X=0,Y=0,Z=0.238,W ...(truncated)"
    WeightMinimum=0
    Weight=@n10.OutputParam
   )
   &RigUnit_ModifyTransforms(
    ExecutePin=()
    ItemToModify="((Item=(Type=Bone,Name=\"dyn_skirt_front_r_01_base\"),Transform=(Rotation=(X=-0.1013,Y=-0.2098,Z=-0.8307,W=0.5057),Translation=(X=12.6608,Y=-0.1363,Z=-2.0906),Scale3D=(X=1,Y=1,Z=1.5053))),(Item=(Type=Bone,Name=\"dyn_skirt_side_r_01_base\"),Transform=(Rotation= ...(truncated)"
    Weight=1
    WeightMinimum=0
   )
   &RigUnit_ModifyTransforms(
    ItemToModify="((Item=(Type=Bone,Name=\"dyn_skirt_front_r_01_base\"),Transform=(Rotation=(X=-0.0654,Y=-0.2236,Z=-0.9025,W=0.3622),Translation=(X=14.6935,Y=0.9029,Z=-1.4917),Scale3D=(X=1,Y=1,Z=1.5053))),(Item=(Type=Bone,Name=\"dyn_skirt_side_r_01_base\"),Transform=(Rotation=( ...(truncated)"
    WeightMinimum=0
    Weight=@n9.Result
    ExecutePin=@n8.ExecutePin
   )
   > n11 = &RigUnit_SphericalPoseReader(
    ActiveRegionScaleFactors=(PositiveWidth=2,NegativeWidth=1,PositiveHeight=1,NegativeHeight=2)
    ActiveRegionSize=0.1
    Debug=(bDrawDebug=false,bDraw2D=False,bDrawLocalAxes=False,DebugScale=25,DebugSegments=20,DebugThickness=0.25)
    DriverAxis=(X=1,Y=0,Z=0)
    DriverItem=(Type=Bone,Name="thigh_r")
    FalloffRegionScaleFactors=(PositiveWidth=1,NegativeWidth=1,PositiveHeight=1,NegativeHeight=1)
    FalloffSize=0.43
    FlipHeightScaling=true
    FlipWidthScaling=true
    OptionalParentItem=(Type=Bone,Name="pelvis")
    RotationOffset=(X=0,Y=0,Z=0)
   )
   > &RigUnit_ModifyTransforms(
    ItemToModify="((Item=(Type=Bone,Name=\"dyn_skirt_front_r_01_base\"),Transform=(Rotation=(X=-0.203,Y=0.3509,Z=-0.2001,W=0.892),Translation=(X=0,Y=0,Z=0),Scale3D=(X=1,Y=1,Z=1))),(Item=(Type=Bone,Name=\"dyn_skirt_side_r_01_base\"),Transform=(Rotation=(X=-0.1096,Y=0.0798,Z=-0.5 ...(truncated)"
    WeightMinimum=0
    Weight=@n11.OutputParam
   )
   > n12 = &RigUnit_SphericalPoseReader(
    ActiveRegionScaleFactors=(PositiveWidth=2,NegativeWidth=1,PositiveHeight=1,NegativeHeight=2)
    ActiveRegionSize=0.1
    Debug=(bDrawDebug=true,bDraw2D=False,bDrawLocalAxes=False,DebugScale=25,DebugSegments=20,DebugThickness=0.25)
    DriverAxis=(X=1,Y=0,Z=0)
    DriverItem=(Type=Bone,Name="thigh_r")
    FalloffRegionScaleFactors=(PositiveWidth=1,NegativeWidth=1,PositiveHeight=1,NegativeHeight=1)
    FalloffSize=0.3
    FlipHeightScaling=true
    FlipWidthScaling=true
    OptionalParentItem=(Type=Bone,Name="spine_01")
    RotationOffset=(X=-45,Y=0,Z=10)
   )
   > &RigUnit_ModifyTransforms(
    ItemToModify="((Item=(Type=Bone,Name=\"dyn_skirt_front_r_01_base\"),Transform=(Rotation=(X=-0.02,Y=0.25,Z=-0.0036,W=0.968),Translation=(X=0,Y=0,Z=0),Scale3D=(X=1,Y=3.2303,Z=1.1956))),(Item=(Type=Bone,Name=\"dyn_skirt_side_r_01_base\"),Transform=(Rotation=(X=-0.0539,Y=-0.122 ...(truncated)"
    WeightMinimum=0
    Weight=@n12.OutputParam
   )
   > n13 = &RigUnit_SphericalPoseReader(
    ActiveRegionScaleFactors=(PositiveWidth=1,NegativeWidth=1,PositiveHeight=1,NegativeHeight=1)
    ActiveRegionSize=0.1
    Debug=(bDrawDebug=false,bDraw2D=False,bDrawLocalAxes=False,DebugScale=25,DebugSegments=20,DebugThickness=0.25)
    DriverAxis=(X=1,Y=0,Z=0)
    DriverItem=(Type=Bone,Name="thigh_r")
    FalloffRegionScaleFactors=(PositiveWidth=1,NegativeWidth=1,PositiveHeight=1,NegativeHeight=1)
    FalloffSize=0.5
    FlipHeightScaling=False
    FlipWidthScaling=False
    OptionalParentItem=(Type=Bone,Name="None")
    RotationOffset=(X=90,Y=0,Z=0)
   )
   > n14 = &RigUnit_ModifyTransforms(
    ItemToModify="((Item=(Type=Bone,Name=\"dyn_skirt_rear_r_01_base\"),Transform=(Rotation=(X=-0.0571,Y=-0.0321,Z=-0.4957,W=0.866),Translation=(X=0,Y=0,Z=0),Scale3D=(X=1,Y=1,Z=1))),(Item=(Type=Bone,Name=\"dyn_skirt_side_r_01_base\"),Transform=(Rotation=(X=0,Y=0,Z=0,W=1),Transla ...(truncated)"
    WeightMinimum=0
    Weight=@n13.OutputParam
   )
  ]}
 New Function_ContainedGraph {nodes [
   n1 = RigVMVariableNode() {variable=NewLocalVar}
   DISPATCH_RigVMDispatch_ArrayIterator = RigVMDispatchNode(ExecuteContext=(), Array=@n1.Value) {template="DISPATCH_RigVMDispatch_ArrayIterator(in Array,out Element,out Index,out Count,out Ratio)"}
   RigUnit_ItemToName = /Script/ControlRig.RigUnit_ItemToName(Value=@DISPATCH_RigVMDispatch_ArrayIterator.Element) {is=struct RigVMUnitNode, template="Cast::Execute(in Value,out Result)"}
   > SetControlColor = /Script/ControlRig.RigUnit_SetControlColor(Color=(R=1,G=0,B=0,A=1), Control=@RigUnit_ItemToName.Result) {is=struct RigVMUnitNode, template="SetControlColor::Execute(in Control,in Color)"}
   n2 = RigVMDispatchNode(ExecuteContext=(), Array=@n1.Value) {template="DISPATCH_RigVMDispatch_ArrayIterator(in Array,out Element,out Index,out Count,out Ratio)"}
   Greater = /Script/RigVM.RigVMFunction_MathIntGreater(B=1, A=@n2.Index) {is=struct RigVMUnitNode, template="Greater::Execute(in A,in B,out Result)"}
   RigVMFunction_ControlFlowBranch = &RigVMFunction_ControlFlowBranch(ExecuteContext=@n2.ExecuteContext, Condition=@Greater.Result)
   RigVMFunction_DebugLineNoSpace = /Script/RigVM.RigVMFunction_DebugLineNoSpace(
    A=(X=0,Y=0,Z=0)
    B=(X=0,Y=0,Z=0)
    Color=(R=1,G=0,B=0,A=1)
    DebugDrawSettings=(DepthPriority=SDPG_Foreground,Lifetime=-1)
    Thickness=0
    WorldOffset=(Rotation=(X=0,Y=0,Z=0,W=1),Translation=(X=0,Y=0,Z=0),Scale3D=(X=1,Y=1,Z=1))
    bEnabled=True
    ExecutePin=@RigVMFunction_ControlFlowBranch.True
   ) {is=struct RigVMUnitNode}
   n3 = RigVMFunctionEntryNode()
   > RigVMFunction_Sequence = RigVMAggregateNode()
   n4 = /Script/ControlRig.RigUnit_HierarchyAddBone(
    Name=cr_def_pelvis_skirtrig
    Parent=(Type=Bone,Name="None")
    Space=LocalSpace
    Transform=&RigUnit_GetTransform(bInitial=true, Item=@n3.pelvis).Transform
    ExecutePin=@RigVMFunction_Sequence.A
   ) {is=struct RigVMUnitNode}
   > n5 = &RigUnit_HierarchyAddBone(
    Name=cr_def_front_mid_skirtrig
    Space=LocalSpace
    Transform=(Rotation=(X=0,Y=0,Z=0,W=1),Translation=(X=-30,Y=20,Z=0),Scale3D=(X=1,Y=1,Z=1))
    Parent=@n4.Item
   )
   > n6 = &RigUnit_HierarchyAddBone(
    Name=cr_def_front_thigh_l_skirtrig
    Space=LocalSpace
    Transform=(Rotation=(X=0,Y=0,Z=0,W=1),Translation=(X=-30,Y=15,Z=-20),Scale3D=(X=1,Y=1,Z=1))
    Parent=@n4.Item
   )
   > n7 = &RigUnit_HierarchyAddBone(
    Name=cr_def_side_thigh_l_skirtrig
    Space=LocalSpace
    Transform=(Rotation=(X=0,Y=0,Z=0,W=1),Translation=(X=-30,Y=0,Z=-25),Scale3D=(X=1,Y=1,Z=1))
    Parent=@n4.Item
   )
   > n8 = &RigUnit_HierarchyAddBone(
    Name=cr_def_rear_thigh_l_skirtrig
    Space=LocalSpace
    Transform=(Rotation=(X=0,Y=0,Z=0,W=1),Translation=(X=-30,Y=-15,Z=-20),Scale3D=(X=1,Y=1,Z=1))
    Parent=@n4.Item
   )
   > n9 = &RigUnit_HierarchyAddBone(
    Name=cr_def_rear_mid_skirtrig
    Space=LocalSpace
    Transform=(Rotation=(X=0,Y=0,Z=0,W=1),Translation=(X=-30,Y=-20,Z=0),Scale3D=(X=1,Y=1,Z=1))
    Parent=@n4.Item
   )
   > n10 = &RigUnit_HierarchyAddBone(
    Name=cr_def_front_thigh_r_skirtrig
    Space=LocalSpace
    Transform=(Rotation=(X=0,Y=0,Z=0,W=1),Translation=(X=-30,Y=15,Z=20),Scale3D=(X=1,Y=1,Z=1))
    Parent=@n4.Item
   )
   > n11 = &RigUnit_HierarchyAddBone(
    Name=cr_def_side_thigh_r_skirtrig
    Space=LocalSpace
    Transform=(Rotation=(X=0,Y=0,Z=0,W=1),Translation=(X=-30,Y=0,Z=25),Scale3D=(X=1,Y=1,Z=1))
    Parent=@n4.Item
   )
   > n12 = &RigUnit_HierarchyAddBone(
    Name=cr_def_rear_thigh_r_skirtrig
    Space=LocalSpace
    Transform=(Rotation=(X=0,Y=0,Z=0,W=1),Translation=(X=-30,Y=-15,Z=20),Scale3D=(X=1,Y=1,Z=1))
    Parent=@n4.Item
   )
   > RigVMVariableNode(
    Value=((Type=None,Name="None"),(Type=None,Name="None"),(Type=None,Name="None"),(Type=None,Name="None"),(Type=None,Name="None"),(Type=None,Name="None"),(Type=None,Name="None"),(Type=None,Name="None"),(Type=None,Name="None"))
    Value.0=@n4.Item
    Value.1=@n5.Item
    Value.2=@n6.Item
    Value.6=@n10.Item
    Value.3=@n7.Item
    Value.7=@n11.Item
    Value.4=@n8.Item
    Value.8=@n12.Item
    Value.5=@n9.Item
   ) {variable=NewLocalVar}
   RigVMFunctionReturnNode(ExecuteContext=@RigVMFunction_Sequence.D, Argument=RigVMVariableNode() {variable=NewLocalVar}.Value)
  ]}
 New Function_ContainedGraph {nodes [
   RigVMFunctionEntryNode()
   > &RigUnit_ModifyTransforms(
    ItemToModify=((Item=(Type=Bone,Name="deform_kneepad_base_l"),Transform=(Rotation=(X=0,Y=0,Z=-0.342,W=0.9397),Translation=(X=0,Y=0,Z=0),Scale3D=(X=1,Y=1,Z=1))))
    WeightMinimum=0
    Weight=RigVMDispatchNode(BoneName=root, Default=0, Name=deformRig_l_knee_alpha) {template="DISPATCH_RigDispatch_GetAnimAttribute(in Name,in BoneName,in Default,out Value,out Found)"}.Value
   )
   > &RigUnit_ModifyTransforms(
    ItemToModify=((Item=(Type=Bone,Name="deform_kneepad_base_r"),Transform=(Rotation=(X=0,Y=0,Z=-0.342,W=0.9397),Translation=(X=0,Y=0,Z=0),Scale3D=(X=1,Y=1,Z=1))))
    WeightMinimum=0
    Weight=RigVMDispatchNode(BoneName=root, Default=0, Name=deformRig_r_knee_alpha) {template="DISPATCH_RigDispatch_GetAnimAttribute(in Name,in BoneName,in Default,out Value,out Found)"}.Value
   )
   > RigVMFunctionReturnNode()
  ]}
 New Function_ContainedGraph {nodes [
   n1 = RigVMFunctionEntryNode()
   RigVMFunction_Sequence = &RigVMFunction_Sequence(ExecuteContext=@n1.ExecuteContext)
   RigUnit_GetTransform = &RigUnit_GetTransform(Item=(Type=Bone,Name="lowerarm_twist_02_l"), Space=LocalSpace, bInitial=False)
   RigUnit_SetRotation = &RigUnit_SetRotation(
    Item=(Type=Bone,Name="deform_pickaxe_main_l")
    Weight=1
    bInitial=False
    Value=@RigUnit_GetTransform.Transform.Rotation
    ExecutePin=@RigVMFunction_Sequence.A
   )
   n2 = &RigUnit_GetTransform(Item=(Type=Bone,Name="lowerarm_twist_02_r"), Space=LocalSpace, bInitial=False)
   > n3 = &RigUnit_SetRotation(Item=(Type=Bone,Name="deform_pickaxe_main_r"), Weight=1, bInitial=False, Value=@n2.Transform.Rotation)
   n4 = &RigVMFunction_Sequence(ExecuteContext=@n3.ExecutePin)
   RigVMFunction_MathBoolNot = &RigVMFunction_MathBoolNot(Value=RigVMVariableNode() {variable=DisablePickaxeElbows}.Value)
   n5 = &RigVMFunction_ControlFlowBranch(Condition=@RigVMFunction_MathBoolNot.Result, ExecuteContext=@n4.B)
   n6 = &RigVMFunction_Sequence(ExecuteContext=@n5.True)
   RigVMFunction_MathQuaternionToEuler = &RigVMFunction_MathQuaternionToEuler(RotationOrder=ZYX, Value=@n2.Transform.Rotation)
   Remap = &RigVMFunction_MathDoubleRemap(SourceMaximum=40, SourceMinimum=0, TargetMaximum=0, TargetMinimum=1, Value=@RigVMFunction_MathQuaternionToEuler.Result.X)
   Multiply = &RigVMFunction_MathDoubleMul(
    B=@Remap.Result
    A=&RigVMFunction_MathDoubleRemap(
     SourceMaximum=0.7
     SourceMinimum=0
     TargetMaximum=1
     TargetMinimum=0
     Value=RigVMDispatchNode(BoneName=root, Default=0, Name=deformRig_r_elbow_alpha) {template="DISPATCH_RigDispatch_GetAnimAttribute(in Name,in BoneName,in Default,out Value,out Found)"}.Value
    ).Result
   )
   &RigUnit_ModifyTransforms(
    ItemToModify=((Item=(Type=Bone,Name="deform_pickaxe_elbow_in_r"),Transform=(Rotation=(X=0.4617,Y=0,Z=0,W=0.887),Translation=(X=0,Y=0,Z=0),Scale3D=(X=1,Y=1,Z=1))))
    WeightMinimum=0
    Weight=@Multiply.Result
    ExecutePin=@n6.B
   )
   n7 = &RigVMFunction_MathQuaternionToEuler(RotationOrder=ZYX, Value=@RigUnit_GetTransform.Transform.Rotation)
   n8 = RigVMDispatchNode(BoneName=root, Default=0, Name=deformRig_l_elbow_alpha) {template="DISPATCH_RigDispatch_GetAnimAttribute(in Name,in BoneName,in Default,out Value,out Found)"}
   &RigUnit_ModifyTransforms(
    ItemToModify=((Item=(Type=Bone,Name="deform_pickaxe_elbow_in_l"),Transform=(Rotation=(X=0.4617,Y=0,Z=0,W=0.887),Translation=(X=0,Y=0,Z=0),Scale3D=(X=1,Y=1,Z=1))))
    WeightMinimum=0
    Weight=&RigVMFunction_MathDoubleMul(
     B=&RigVMFunction_MathDoubleRemap(SourceMaximum=40, SourceMinimum=0, TargetMaximum=0, TargetMinimum=1, Value=@n7.Result.X).Result
     A=&RigVMFunction_MathDoubleRemap(SourceMaximum=0.7, SourceMinimum=0, TargetMaximum=1, TargetMinimum=0, Value=@n8.Value).Result
    ).Result
    ExecutePin=@n6.A
   )
   RigVMFunction_ControlFlowBranch = &RigVMFunction_ControlFlowBranch(Condition=RigVMVariableNode() {variable=IsBuiltInPickaxeEquipped}.Value, ExecuteContext=@n4.A)
   n9 = &RigVMFunction_Sequence(ExecuteContext=@RigVMFunction_ControlFlowBranch.False)
   &RigUnit_ModifyTransforms(
    ItemToModify="((Item=(Type=Bone,Name=\"deform_pickaxe_blade_a_02_l\"),Transform=(Rotation=(X=0,Y=-0.0872,Z=0,W=0.9962),Translation=(X=0,Y=0,Z=0),Scale3D=(X=1,Y=1,Z=1))),(Item=(Type=Bone,Name=\"deform_pickaxe_blade_b_02_l\"),Transform=(Rotation=(X=0,Y=-0.0872,Z=0,W=0.9962),T ...(truncated)"
    WeightMinimum=0
    Weight=RigVMDispatchNode(BoneName=root, Default=0, Name=deformRig_l_wrist_up_alpha) {template="DISPATCH_RigDispatch_GetAnimAttribute(in Name,in BoneName,in Default,out Value,out Found)"}.Value
    ExecutePin=@n9.A
   )
   &RigUnit_ModifyTransforms(
    ItemToModify="((Item=(Type=Bone,Name=\"deform_pickaxe_blade_a_02_r\"),Transform=(Rotation=(X=0,Y=-0.0872,Z=0,W=0.9962),Translation=(X=0,Y=0,Z=0),Scale3D=(X=1,Y=1,Z=1))),(Item=(Type=Bone,Name=\"deform_pickaxe_blade_b_02_r\"),Transform=(Rotation=(X=0,Y=-0.0872,Z=0,W=0.9962),T ...(truncated)"
    WeightMinimum=0
    Weight=RigVMDispatchNode(BoneName=root, Default=0, Name=deformRig_r_wrist_up_alpha) {template="DISPATCH_RigDispatch_GetAnimAttribute(in Name,in BoneName,in Default,out Value,out Found)"}.Value
    ExecutePin=@n9.B
   )
   RigVMFunctionReturnNode(ExecuteContext=@RigVMFunction_Sequence.B)
  ]}
 RigVMFunctionLibrary {nodes [
   Deform_Builtin_Pickaxe = RigVMCollapseNode(ExecuteContext=())
   Deform_Chestcloth = RigVMCollapseNode(ExecuteContext=())
   Deform_Knees = RigVMCollapseNode(ExecuteContext=())
   Deform_Shoulderpads = RigVMCollapseNode(ExecuteContext=())
   Deform_Skirt = RigVMCollapseNode(ExecuteContext=())
   Setup_Skirt = RigVMCollapseNode(ExecuteContext=(), pelvis=(Type=Bone,Name="None"))
  ]}
 RigVMFunction_Sequence_1_1_ContainedGraph {nodes [
   n1 = RigVMFunctionEntryNode()
   n2 = &RigVMFunction_Sequence(ExecuteContext=@n1.ExecuteContext)
   n3 = &RigVMFunction_Sequence(ExecuteContext=@n2.B)
   n4 = &RigVMFunction_Sequence(ExecuteContext=@n3.B)
   n5 = &RigVMFunction_Sequence(ExecuteContext=@n4.B)
   n6 = &RigVMFunction_Sequence(ExecuteContext=@n5.B)
   n7 = &RigVMFunction_Sequence(ExecuteContext=@n6.B)
   n8 = &RigVMFunction_Sequence(ExecuteContext=@n7.B)
   n9 = &RigVMFunction_Sequence(ExecuteContext=@n8.B)
   n10 = &RigVMFunction_Sequence(ExecuteContext=@n9.B)
   RigVMFunctionReturnNode(A=@n2.A, B=@n3.A, C=@n4.A, D=@n5.A, E=@n6.A, F=@n7.A, G=@n8.A, H=@n9.A, I=@n10.A, J=@n10.B)
  ]}
 RigVMFunction_Sequence_1_1_ContainedGraph {nodes [
   n1 = RigVMFunctionEntryNode()
   n2 = &RigVMFunction_Sequence(ExecuteContext=@n1.ExecuteContext)
   n3 = &RigVMFunction_Sequence(ExecuteContext=@n2.B)
   RigVMFunctionReturnNode(A=@n2.A, B=@n3.A, C=@n3.B)
  ]}
 RigVMFunction_Sequence_1_ContainedGraph {nodes [
   n1 = RigVMFunctionEntryNode()
   RigVMFunction_Sequence = &RigVMFunction_Sequence(ExecuteContext=@n1.ExecuteContext)
   n2 = &RigVMFunction_Sequence(ExecuteContext=@RigVMFunction_Sequence.B)
   n3 = &RigVMFunction_Sequence(ExecuteContext=@n2.B)
   RigVMFunctionReturnNode(A=@RigVMFunction_Sequence.A, B=@n2.A, C=@n3.A, D=@n3.B)
  ]}
 RigVMModel {nodes [
   RigUnit_BeginExecution = &RigUnit_BeginExecution()
   FN_DeformRig_FNC = RigVMFunctionReferenceNode(
    DeformRig_ Curve="(deformRig_l_upperarm_up_alpha_72_#1=True,deformRig_r_upperarm_up_alpha_73_#2=True,deformRig_l_upperarm_fr_alpha_74_#3=True,deformRig_r_upperarm_fr_alpha_75_CB51B80C4 ...(truncated)"
    DeformRig_Bool="(upperarm_1_#4=True,elbow_3_#5=True,wrist_7_#6=True,groin_8_#7=True,glute_17_#8=True,knee_9_40C408324CA261 ...(truncated)"
    DeformRig_Debug=(upperarm_debug_18_#4=False,wrist_debug_20_#6=False,lat_debug_26_#9=False,groinGlute_debug_27_#7=false)
    DeformRig_Value="(upperarm_up_18_#10=(Rotation=(X=0,Y=0,Z=0,W=1),Translation=(X=0,Y=0,Z=8),Scale3D=(X=0.7,Y=1,Z=1)),upperarm_fwd_fr_19_#11=(Rotation=(X=0,Y=0,Z=0,W=1),Translation=(X=0,Y=10,Z=0),Scale3D=(X=0.5,Y=1,Z= ...(truncated)"
    ExecuteContext=@RigUnit_BeginExecution.ExecutePin
   )
   > Deform_Shoulderpads = RigVMFunctionReferenceNode()
   > Deform_Chestcloth = RigVMFunctionReferenceNode(L Clavicle SPR=@Deform_Shoulderpads.L Clavicle SPR, R Clavicle SPR=@Deform_Shoulderpads.R Clavicle SPR)
   > Deform_Skirt = RigVMFunctionReferenceNode()
   > Deform_Knees = RigVMFunctionReferenceNode()
   > Deform_Builtin_Pickaxe = RigVMFunctionReferenceNode()
   RigUnit_GetTransform = &RigUnit_GetTransform(Item=(Type=Bone,Name="head"), Space=LocalSpace, bInitial=False)
   Interpolate = /Script/RigVM.RigVMFunction_MathQuaternionSlerp(B=(X=0,Y=0,Z=0,W=1), T=0.3, A=@RigUnit_GetTransform.Transform.Rotation) {is=struct RigVMUnitNode, template="Interpolate::Execute(in A,in B,in T,out Result)"}
   RigUnit_SetRotation = &RigUnit_SetRotation(ExecutePin=(), Item=(Type=Bone,Name="head"), Weight=1, bInitial=False, Value=@Interpolate.Result)
  ]}
]
hierarchy [
 root {loc=<0 0 0>, rot=<0 0 0 1>}
 pelvis {parent=root, loc=<0 -0.7321 92.4218>, rot=<0.0127 0.707 -0.0127 -0.707>}
 spine_01 {parent=pelvis, loc=<5.6002 -0.5128 0>, rot=<0 0 0.034 -0.9994>}
 spine_02 {parent=spine_01, loc=<8.0753 0.0139 0>, rot=<0 0 0.0238 -0.9997>}
 spine_03 {parent=spine_02, loc=<7.0393 0 0>, rot=<0 0 0.0149 -0.9999>}
 spine_04 {parent=spine_03, loc=<8.1024 0 0>, rot=<0 0 -0.0132 -0.9999>}
 spine_05 {parent=spine_04, loc=<10.0146 0 0>, rot=<0 0 -0.0015 -1>}
 attach_cape {parent=spine_05, loc=<9.1509 -13.3783 0>, rot=<0.0616 -0.7044 -0.0616 -0.7044>}
 attach_backpack {parent=spine_05, loc=<-0.123 -15.1423 0>, rot=<-0.0093 -0.707 0.0093 -0.707>}
 neck_01 {parent=spine_05, loc=<10.815 1.9196 0>, rot=<0 0 -0.1864 -0.9825>}
 neck_02 {parent=neck_01, loc=<4.7347 0.3045 0>, rot=<-0.0227 0.0123 0.0443 -0.9987>}
 head {parent=neck_02, loc=<5.7632 0.5552 0>, rot=<0 0 0.115 -0.9934>}
 clavicle_l {parent=spine_05, loc=<2.5856 -0.2866 -4.3777>, rot=<0.0465 -0.7289 -0.009 -0.6829>}
 upperarm_l {parent=clavicle_l, loc=<18.2264 0.1342 -0.001>, rot=<0.0253 -0.2693 0.0013 -0.9627>}
 lowerarm_l {parent=upperarm_l, loc=<29.0071 0.1686 -1.239>, rot=<0.0391 -0.0099 -0.2453 -0.9686>}
 hand_l {parent=lowerarm_l, loc=<25.8607 0.222 1.0274>, rot=<-0.7193 -0.0988 0.1193 -0.6773>}
 thumb_01_l {parent=hand_l, loc=<1.3118 -0.3191 -2.459>, rot=<0.6546 -0.4042 0.0664 -0.6354>}
 thumb_02_l {parent=thumb_01_l, loc=<5.9336 -0.0448 0.0401>, rot=<-0.0037 0.0213 0.298 -0.9543>}
 thumb_03_l {parent=thumb_02_l, loc=<2.2223 0.3682 -0.0907>, rot=<0.0118 -0.013 0.0066 -0.9998>}
 index_metacarpal_l {parent=hand_l, loc=<1.6535 0.0183 -2.96>, rot=<0.0158 -0.0124 -0.0173 -0.9996>}
 index_01_l {parent=index_metacarpal_l, loc=<8.3008 0.0357 -0.0524>, rot=<-0.0044 0.0485 0.0924 -0.9945>}
 index_02_l {parent=index_01_l, loc=<3.3587 -0.3004 0.0452>, rot=<-0.002 0.0027 0.3225 -0.9466>}
 index_03_l {parent=index_02_l, loc=<2.5578 -0.123 -0.0252>, rot=<-0.002 0.0067 0.2558 -0.9667>}
 middle_metacarpal_l {parent=hand_l, loc=<2.1856 -0.1725 -0.5984>, rot=<0.006 0.0168 -0.0329 -0.9993>}
 middle_01_l {parent=middle_metacarpal_l, loc=<7.3275 -0.2447 0.0129>, rot=<-0.0208 0.0042 0.0858 -0.9961>}
 middle_02_l {parent=middle_01_l, loc=<4.0123 0.1401 0.1175>, rot=<-0.0218 0.0258 0.3512 -0.9357>}
 middle_03_l {parent=middle_02_l, loc=<2.6077 0.0534 0.1373>, rot=<0.0082 0.0188 0.1708 -0.9851>}
 ring_metacarpal_l {parent=hand_l, loc=<2.2074 -0.3625 1.3638>, rot=<-0.0284 0.0722 -0.0458 -0.9959>}
 ring_01_l {parent=ring_metacarpal_l, loc=<7.0295 -0.6197 -0.1446>, rot=<-0.0489 -0.0252 0.1892 -0.9804>}
 ring_02_l {parent=ring_01_l, loc=<3.6174 0.4647 -0.1027>, rot=<0 0 0.2806 -0.9598>}
 ring_03_l {parent=ring_02_l, loc=<2.4152 0.114 0.0199>, rot=<0 0 0.1189 -0.9929>}
 pinky_metacarpal_l {parent=hand_l, loc=<1.7616 -0.9753 3.3814>, rot=<-0.0397 0.0814 -0.069 -0.9935>}
 pinky_01_l {parent=pinky_metacarpal_l, loc=<7.1709 -0.4831 -0.1094>, rot=<-0.0654 -0.0035 0.2242 -0.9723>}
 pinky_02_l {parent=pinky_01_l, loc=<2.8274 0.0893 -0.1043>, rot=<0.0037 0.0011 0.2854 -0.9584>}
 pinky_03_l {parent=pinky_02_l, loc=<2.2046 0.0307 -0.0078>, rot=<-0.0002 0.0098 0.1445 -0.9895>}
 deform_wrist_dn_base_l {parent=hand_l, loc=<0 -3.4173 0>, rot=<0 0 0 -1>}
 deform_wrist_dn_l {parent=deform_wrist_dn_base_l, loc=<0 0 0>, rot=<0 0 0 1>}
 deform_wrist_up_base_l {parent=hand_l, loc=<0 2.7805 0>, rot=<0 0 0 -1>}
 deform_wrist_up_l {parent=deform_wrist_up_base_l, loc=<0 0 0>, rot=<0 0 0 1>}
 lowerarm_twist_02_l {parent=lowerarm_l, loc=<9.6199 0 0>, rot=<0 0 0 -1>}
 lowerarm_twist_01_l {parent=lowerarm_l, loc=<17.8549 0 0>, rot=<0 0 0 -1>}
 deform_elbow_in_base_l {parent=lowerarm_l, loc=<0.2112 5.645 0.8274>, rot=<0 0 0 1>}
 deform_elbow_in_l {parent=deform_elbow_in_base_l, loc=<0 0 0>, rot=<0 0 0 1>}
 deform_elbow_out_base_l {parent=lowerarm_l, loc=<0 -7.5129 0>, rot=<0 0 0 1>}
 deform_elbow_out_l {parent=deform_elbow_out_base_l, loc=<0 0 0>, rot=<0 0 0 1>}
 deform_pickaxe_main_l {parent=lowerarm_l, loc=<11.7763 0.0218 1.0453>, rot=<0 0 0 1>}
 deform_pickaxe_elbow_in_l {parent=deform_pickaxe_main_l, loc=<-10.3501 4.8499 5.7102>, rot=<-0.3658 0 0 -0.9307>}
 deform_pickaxe_blade_a_01_l {parent=deform_pickaxe_main_l, loc=<12.4448 -3.0432 3.395>, rot=<0.0407 -0.949 -0.1056 0.2942>}
 deform_pickaxe_blade_a_02_l {parent=deform_pickaxe_blade_a_01_l, loc=<20.2902 0 0>, rot=<-0.0006 -0.9336 -0.0179 -0.3578>}
 deform_pickaxe_blade_b_01_l {parent=deform_pickaxe_main_l, loc=<12.3099 2.6021 4.5438>, rot=<0.0407 -0.949 -0.1056 0.2942>}
 deform_pickaxe_blade_b_02_l {parent=deform_pickaxe_blade_b_01_l, loc=<20.2669 0.0293 0.0474>, rot=<-0.0006 -0.9336 -0.0179 -0.3578>}
 upperarm_twist_01_l {parent=upperarm_l, loc=<7.9796 0.0843 -0.6195>, rot=<0 0 0 -1>}
 deform_upperarm_bk_base_l {parent=upperarm_twist_01_l, loc=<-6.8983 -13.3826 -1.3258>, rot=<0 0 0 -1>}
 deform_upperarm_bk_l {parent=deform_upperarm_bk_base_l, loc=<0 0 0>, rot=<0 0 0 1>}
 deform_upperarm_fr_base_l {parent=upperarm_twist_01_l, loc=<-8.3487 10.7117 -2.1455>, rot=<0 0 0 -1>}
 deform_upperarm_fr_l {parent=deform_upperarm_fr_base_l, loc=<0 0 0>, rot=<0 0 0 1>}
 deform_upperarm_up_base_l {parent=upperarm_twist_01_l, loc=<-11.2012 -1.2234 8.7784>, rot=<0 0 0 -1>}
 deform_upperarm_up_l {parent=deform_upperarm_up_base_l, loc=<0 0 0>, rot=<0 0 0 1>}
 upperarm_twist_02_l {parent=upperarm_l, loc=<19.5059 0.0843 -0.6195>, rot=<0 0 0 -1>}
 deform_bicep_base_l {parent=upperarm_twist_02_l, loc=<0 8.0494 0>, rot=<0 0 0 1>}
 deform_bicep_l {parent=deform_bicep_base_l, loc=<0 0 0>, rot=<0 0 0 1>}
 deform_shoulderpad_l_base {parent=clavicle_l, loc=<12.158 0.7104 4.0294>, rot=<0.0008 0.0336 -0.0252 -0.9991>}
 deform_shoulderpad_l {parent=deform_shoulderpad_l_base, loc=<0 0 0>, rot=<0 0 0 1>}
 deform_shoulderpad_rear_spike_l {parent=deform_shoulderpad_l, loc=<12.6281 -7.2117 8.6467>, rot=<-0.089 0.046 0.0612 -0.9931>}
 deform_shoulderpad_front_spike_l {parent=deform_shoulderpad_l, loc=<12.6284 6.8138 8.6254>, rot=<0.0855 0.0464 -0.0609 -0.9934>}
 clavicle_r {parent=spine_05, loc=<2.5855 -0.2866 4.3777>, rot=<0.7289 0.0465 -0.6829 0.009>}
 upperarm_r {parent=clavicle_r, loc=<-18.2264 -0.1342 0.001>, rot=<0.0253 -0.2693 0.0013 -0.9627>}
 lowerarm_r {parent=upperarm_r, loc=<-29.0071 -0.1686 1.239>, rot=<0.0391 -0.0099 -0.2453 -0.9686>}
 hand_r {parent=lowerarm_r, loc=<-25.8607 -0.222 -1.0274>, rot=<-0.7193 -0.0988 0.1193 -0.6773>}
 thumb_01_r {parent=hand_r, loc=<-1.3118 0.3191 2.459>, rot=<0.6546 -0.4042 0.0664 -0.6354>}
 thumb_02_r {parent=thumb_01_r, loc=<-5.9336 0.0448 -0.0401>, rot=<-0.0037 0.0213 0.298 -0.9543>}
 thumb_03_r {parent=thumb_02_r, loc=<-2.2223 -0.3682 0.0907>, rot=<0.0118 -0.013 0.0066 -0.9998>}
 index_metacarpal_r {parent=hand_r, loc=<-1.6535 -0.0183 2.96>, rot=<0.0158 -0.0124 -0.0173 -0.9996>}
 index_01_r {parent=index_metacarpal_r, loc=<-8.3008 -0.0357 0.0524>, rot=<-0.0044 0.0485 0.0924 -0.9945>}
 index_02_r {parent=index_01_r, loc=<-3.3587 0.3004 -0.0452>, rot=<-0.002 0.0027 0.3225 -0.9466>}
 index_03_r {parent=index_02_r, loc=<-2.5578 0.123 0.0252>, rot=<-0.002 0.0067 0.2558 -0.9667>}
 middle_metacarpal_r {parent=hand_r, loc=<-2.1856 0.1725 0.5984>, rot=<0.006 0.0168 -0.0329 -0.9993>}
 middle_01_r {parent=middle_metacarpal_r, loc=<-7.3275 0.2447 -0.0129>, rot=<-0.0208 0.0042 0.0858 -0.9961>}
 middle_02_r {parent=middle_01_r, loc=<-4.0123 -0.1401 -0.1175>, rot=<-0.0218 0.0258 0.3512 -0.9357>}
 middle_03_r {parent=middle_02_r, loc=<-2.6077 -0.0534 -0.1373>, rot=<0.0082 0.0188 0.1708 -0.9851>}
 ring_metacarpal_r {parent=hand_r, loc=<-2.2074 0.3625 -1.3638>, rot=<-0.0284 0.0722 -0.0458 -0.9959>}
 ring_01_r {parent=ring_metacarpal_r, loc=<-7.0295 0.6197 0.1446>, rot=<-0.0489 -0.0252 0.1892 -0.9804>}
 ring_02_r {parent=ring_01_r, loc=<-3.6174 -0.4647 0.1027>, rot=<0 0 0.2806 -0.9598>}
 ring_03_r {parent=ring_02_r, loc=<-2.4152 -0.114 -0.0199>, rot=<0 0 0.1189 -0.9929>}
 pinky_metacarpal_r {parent=hand_r, loc=<-1.7616 0.9753 -3.3814>, rot=<-0.0397 0.0814 -0.069 -0.9935>}
 pinky_01_r {parent=pinky_metacarpal_r, loc=<-7.1709 0.4831 0.1094>, rot=<-0.0654 -0.0035 0.2242 -0.9723>}
 pinky_02_r {parent=pinky_01_r, loc=<-2.8274 -0.0893 0.1043>, rot=<0.0037 0.0011 0.2854 -0.9584>}
 pinky_03_r {parent=pinky_02_r, loc=<-2.2046 -0.0307 0.0078>, rot=<-0.0002 0.0098 0.1445 -0.9895>}
 deform_wrist_dn_base_r {parent=hand_r, loc=<0 3.4172 0>, rot=<0 0 0 -1>}
 deform_wrist_dn_r {parent=deform_wrist_dn_base_r, loc=<0 0 0>, rot=<0 0 0 -1>}
 deform_wrist_up_base_r {parent=hand_r, loc=<-0.0001 -2.7804 0>, rot=<0 0 0 -1>}
 deform_wrist_up_r {parent=deform_wrist_up_base_r, loc=<0 0 0>, rot=<0 0 0 -1>}
 lowerarm_twist_02_r {parent=lowerarm_r, loc=<-9.6199 0 0>, rot=<0 0 0 -1>}
 lowerarm_twist_01_r {parent=lowerarm_r, loc=<-17.8549 0 0>, rot=<0 0 0 -1>}
 deform_elbow_in_base_r {parent=lowerarm_r, loc=<-0.2114 -5.6449 -0.827>, rot=<0 0 0 -1>}
 deform_elbow_in_r {parent=deform_elbow_in_base_r, loc=<0 0 0>, rot=<0 0 0 1>}
 deform_elbow_out_base_r {parent=lowerarm_r, loc=<-0.0001 7.5129 0.0002>, rot=<0 0 0 -1>}
 deform_elbow_out_r {parent=deform_elbow_out_base_r, loc=<0 0 0>, rot=<0 0 0 1>}
 deform_pickaxe_main_r {parent=lowerarm_r, loc=<-11.7765 -0.0217 -1.0451>, rot=<0 0 0 -1>}
 deform_pickaxe_elbow_in_r {parent=deform_pickaxe_main_r, loc=<10.3504 -4.85 -5.7108>, rot=<-0.3658 0 0 -0.9307>}
 deform_pickaxe_blade_a_01_r {parent=deform_pickaxe_main_r, loc=<-12.4447 3.0432 -3.3951>, rot=<0.0407 -0.949 -0.1056 0.2942>}
 deform_pickaxe_blade_a_02_r {parent=deform_pickaxe_blade_a_01_r, loc=<-20.2901 0 -0.0001>, rot=<-0.0006 -0.9336 -0.0179 -0.3578>}
 deform_pickaxe_blade_b_01_r {parent=deform_pickaxe_main_r, loc=<-12.3097 -2.6022 -4.5441>, rot=<0.0407 -0.949 -0.1056 0.2942>}
 deform_pickaxe_blade_b_02_r {parent=deform_pickaxe_blade_b_01_r, loc=<-20.2672 -0.0295 -0.0473>, rot=<-0.0006 -0.9336 -0.0179 -0.3578>}
 upperarm_twist_01_r {parent=upperarm_r, loc=<-7.9796 -0.0843 0.6195>, rot=<0 0 0 -1>}
 deform_upperarm_bk_base_r {parent=upperarm_twist_01_r, loc=<6.8985 13.3826 1.3257>, rot=<0 0 0 1>}
 deform_upperarm_bk_r {parent=deform_upperarm_bk_base_r, loc=<0 0 0>, rot=<0 0 0 1>}
 deform_upperarm_fr_base_r {parent=upperarm_twist_01_r, loc=<8.3485 -10.7117 2.1456>, rot=<0 0 0 1>}
 deform_upperarm_fr_r {parent=deform_upperarm_fr_base_r, loc=<0 0 0>, rot=<0 0 0 1>}
 deform_upperarm_up_base_r {parent=upperarm_twist_01_r, loc=<11.2015 1.2234 -8.7788>, rot=<0 0 0 1>}
 deform_upperarm_up_r {parent=deform_upperarm_up_base_r, loc=<0 0 0>, rot=<0 0 0 1>}
 upperarm_twist_02_r {parent=upperarm_r, loc=<-19.5059 -0.0843 0.6195>, rot=<0 0 0 -1>}
 deform_bicep_base_r {parent=upperarm_twist_02_r, loc=<0.0002 -8.0494 -0.0003>, rot=<0 0 0 1>}
 deform_bicep_r {parent=deform_bicep_base_r, loc=<0 0 0>, rot=<0 0 0 1>}
 deform_shoulderpad_r_base {parent=clavicle_r, loc=<-12.158 -0.7104 -4.0294>, rot=<0.0008 0.0336 -0.0252 -0.9991>}
 deform_shoulderpad_r {parent=deform_shoulderpad_r_base, loc=<0 0 0>, rot=<0 0 0 1>}
 deform_shoulderpad_rear_spike_r {parent=deform_shoulderpad_r, loc=<-12.6281 7.2118 -8.647>, rot=<-0.089 0.046 0.0612 -0.9931>}
 deform_shoulderpad_front_spike_r {parent=deform_shoulderpad_r, loc=<-12.6284 -6.8138 -8.625>, rot=<0.0855 0.0464 -0.0609 -0.9934>}
 dyn_pec_cloth_l_base {parent=spine_05, loc=<3.2877 10.9928 -10.742>, rot=<-0.111 -0.8554 -0.4142 -0.2905>}
 dyn_pec_cloth_l_01 {parent=dyn_pec_cloth_l_base, loc=<0 0 0>, rot=<0 0 0 1>}
 dyn_pec_cloth_l_02 {parent=dyn_pec_cloth_l_01, loc=<7.6704 0 0>, rot=<0.0236 -0.3388 0.0395 -0.9397>}
 dyn_pec_cloth_l_03 {parent=dyn_pec_cloth_l_02, loc=<7.0455 3.2383 -0.5755>, rot=<0.0198 0.002 0.0109 -0.9997>}
 dyn_pec_cloth_l_04 {parent=dyn_pec_cloth_l_02, loc=<6.9386 -2.0392 -0.862>, rot=<0.0198 0.002 0.0109 -0.9997>}
 dyn_pec_cloth_l_05 {parent=dyn_pec_cloth_l_04, loc=<6.6031 1.1547 -0.3341>, rot=<0 0 0 1>}
 dyn_pec_cloth_r_base {parent=spine_05, loc=<3.2877 10.9928 10.742>, rot=<0.8554 -0.111 -0.2905 0.4142>}
 dyn_pec_cloth_r_01 {parent=dyn_pec_cloth_r_base, loc=<0 0 0>, rot=<0 0 0 -1>}
 dyn_pec_cloth_r_02 {parent=dyn_pec_cloth_r_01, loc=<-7.6702 -0.0001 -0.0001>, rot=<0.0236 -0.3388 0.0395 -0.9397>}
 dyn_pec_cloth_r_03 {parent=dyn_pec_cloth_r_02, loc=<-7.0457 -3.2383 0.5755>, rot=<0.0198 0.002 0.0109 -0.9997>}
 dyn_pec_cloth_r_04 {parent=dyn_pec_cloth_r_02, loc=<-6.9392 2.0392 0.862>, rot=<0.0198 0.002 0.0109 -0.9997>}
 dyn_pec_cloth_r_05 {parent=dyn_pec_cloth_r_04, loc=<-6.6023 -1.1545 0.3341>, rot=<0 0 0 -1>}
 deform_pec_aim_base_l {parent=spine_04, loc=<5.0232 13.6719 -2.6508>, rot=<0.0672 -0.5201 0.3228 -0.7879>}
 deform_pec_aim_l {parent=deform_pec_aim_base_l, loc=<0 0 0>, rot=<0 0 0 1>}
 deform_pec_base_l {parent=deform_pec_aim_l, loc=<16.3279 0 0>, rot=<0 0 0 1>}
 deform_pec_l {parent=deform_pec_base_l, loc=<0 0 0>, rot=<0 0 0 1>}
 deform_pec_aim_base_r {parent=spine_04, loc=<5.023 13.6718 2.6508>, rot=<-0.5201 -0.0672 0.7879 0.3228>}
 deform_pec_aim_r {parent=deform_pec_aim_base_r, loc=<0 0 0>, rot=<0 0 0 1>}
 deform_pec_base_r {parent=deform_pec_aim_r, loc=<-16.3279 0 0>, rot=<0 0 0 -1>}
 deform_pec_r {parent=deform_pec_base_r, loc=<0 0 0.0002>, rot=<0 0 0 -1>}
 deform_lat_l_base {parent=spine_03, loc=<2.5566 -10.3716 -17.6918>, rot=<0 0 0 1>}
 deform_lat_l {parent=deform_lat_l_base, loc=<0 0 0>, rot=<0 0 0 1>}
 deform_lat_r_base {parent=spine_03, loc=<2.5568 -10.3716 17.6918>, rot=<0 0 1 0>}
 deform_lat_r {parent=deform_lat_r_base, loc=<0 0 0>, rot=<0 0 0 -1>}
 thigh_l {parent=pelvis, loc=<-1.8716 -0.0109 -11.9936>, rot=<0.0012 0.0572 0.0155 -0.9982>}
 calf_l {parent=thigh_l, loc=<-41.9178 -0.2246 0.0926>, rot=<0.0006 -0.0161 -0.0391 -0.9991>}
 foot_l {parent=calf_l, loc=<-38.7057 1.2438 0.2452>, rot=<0.0975 -0.0474 0.0317 -0.9936>}
 ball_l {parent=foot_l, loc=<-6.1614 13.1469 0.6807>, rot=<0.0001 -0.0001 -0.7063 -0.7079>}
 calf_twist_02_l {parent=calf_l, loc=<-15.9268 -0.0967 0.0397>, rot=<0.0237 0.0095 -0.0035 -0.9997>}
 calf_twist_01_l {parent=calf_l, loc=<-32.9335 -0.0967 0.0397>, rot=<0.0237 0.0095 -0.0035 -0.9997>}
 deform_knee_in_base_l {parent=calf_l, loc=<1.298 -8.0356 -0.0496>, rot=<0 0 0 1>}
 deform_knee_in_l {parent=deform_knee_in_base_l, loc=<0 0 0>, rot=<0 0 0 1>}
 deform_knee_out_base_l {parent=calf_l, loc=<0 6.8739 0>, rot=<0 0 0 -1>}
 deform_knee_out_l {parent=deform_knee_out_base_l, loc=<0 0 0>, rot=<0 0 0 1>}
 deform_kneepad_base_l {parent=calf_l, loc=<-7.0049 4.8036 0.8361>, rot=<0 0 0 -1>}
 deform_kneepad_l {parent=deform_kneepad_base_l, loc=<0 0 0>, rot=<0 0 0 1>}
 thigh_twist_01_l {parent=thigh_l, loc=<-21.6303 0 0>, rot=<0.0013 0 0 -1>}
 deform_groin_base_l {parent=thigh_l, loc=<-2.1908 8.5661 -1.7442>, rot=<0 0 0 -1>}
 deform_groin_l {parent=deform_groin_base_l, loc=<0 0 0>, rot=<0 0 0 1>}
 thigh_r {parent=pelvis, loc=<-1.8716 -0.0109 11.9936>, rot=<0.0572 -0.0012 0.9982 0.0155>}
 calf_r {parent=thigh_r, loc=<41.9178 0.2246 -0.0926>, rot=<0.0006 -0.0161 -0.0391 -0.9991>}
 foot_r {parent=calf_r, loc=<38.7057 -1.2438 -0.2452>, rot=<0.0975 -0.0474 0.0317 -0.9936>}
 ball_r {parent=foot_r, loc=<6.1614 -13.1469 -0.6807>, rot=<0.0001 -0.0001 -0.7063 -0.7079>}
 calf_twist_02_r {parent=calf_r, loc=<15.9268 0.0967 -0.0397>, rot=<0.0237 0.0095 -0.0035 -0.9997>}
 calf_twist_01_r {parent=calf_r, loc=<32.9335 0.0967 -0.0397>, rot=<0.0237 0.0095 -0.0035 -0.9997>}
 deform_knee_in_base_r {parent=calf_r, loc=<-1.2981 8.0356 0.0496>, rot=<0 0 0 -1>}
 deform_knee_in_r {parent=deform_knee_in_base_r, loc=<0 0 0>, rot=<0 0 0 1>}
 deform_knee_out_base_r {parent=calf_r, loc=<0 -6.8739 0>, rot=<0 0 0 -1>}
 deform_knee_out_r {parent=deform_knee_out_base_r, loc=<0 0 0>, rot=<0 0 0 1>}
 deform_kneepad_base_r {parent=calf_r, loc=<7.0049 -4.8036 -0.8361>, rot=<0 0 0 -1>}
 deform_kneepad_r {parent=deform_kneepad_base_r, loc=<0 0 0>, rot=<0 0 0 1>}
 thigh_twist_01_r {parent=thigh_r, loc=<21.6303 0 0>, rot=<0.0013 0 0 -1>}
 deform_groin_base_r {parent=thigh_r, loc=<2.1908 -8.5661 1.7442>, rot=<0 0 0 -1>}
 deform_groin_r {parent=deform_groin_base_r, loc=<0 0 0>, rot=<0 0 0 -1>}
 deform_glute_base_l {parent=pelvis, loc=<-10.1877 -11.6217 -12.899>, rot=<0.0275 0.0501 -0.4472 -0.8926>}
 deform_glute_l {parent=deform_glute_base_l, loc=<0 0 0>, rot=<0 0 0 1>}
 deform_glute_base_r {parent=pelvis, loc=<-10.1877 -11.6218 12.899>, rot=<0.0501 -0.0275 0.8926 -0.4472>}
 deform_glute_r {parent=deform_glute_base_r, loc=<0 0 0>, rot=<0 0 0 1>}
 dyn_skirt_front_r_01_base {parent=pelvis, loc=<8.6312 8.3006 9.5698>, rot=<-0.1769 -0.0533 0.0642 -0.9807>}
 dyn_skirt_front_r_01 {parent=dyn_skirt_front_r_01_base, loc=<0 0 0>, rot=<0 0 0 1>}
 dyn_skirt_front_r_02 {parent=dyn_skirt_front_r_01, loc=<-10 0 0>, rot=<0 0 0 -1>}
 dyn_skirt_front_r_03 {parent=dyn_skirt_front_r_02, loc=<-10 0 0>, rot=<0 0 0 -1>}
 dyn_skirt_side_r_01_base {parent=pelvis, loc=<8.412 2.8525 14.5527>, rot=<-0.7247 -0.0727 0.0872 -0.6797>}
 dyn_skirt_side_r_01 {parent=dyn_skirt_side_r_01_base, loc=<0 0 0>, rot=<0 0 0 1>}
 dyn_skirt_side_r_02 {parent=dyn_skirt_side_r_01, loc=<-10 0 0>, rot=<0 0 0 -1>}
 dyn_skirt_side_r_03 {parent=dyn_skirt_side_r_02, loc=<-10 0 0>, rot=<0 0 0 -1>}
 dyn_skirt_side_r_04 {parent=dyn_skirt_side_r_03, loc=<-9.2092 0.0321 -3.6211>, rot=<0.0692 -0.0481 0.0813 -0.9931>}
 dyn_skirt_side_r_05 {parent=dyn_skirt_side_r_03, loc=<-9.2154 1.4476 0.6175>, rot=<-0.1846 0 0 -0.9828>}
 dyn_skirt_rear_r_01_base {parent=pelvis, loc=<5.0561 -13.8668 5.2432>, rot=<0.9996 0.0128 -0.0109 0.0208>}
 dyn_skirt_rear_r_01 {parent=dyn_skirt_rear_r_01_base, loc=<0 0 0>, rot=<0 0 0 1>}
 dyn_skirt_rear_r_02 {parent=dyn_skirt_rear_r_01, loc=<-10 0 0>, rot=<0 0 0 -1>}
 dyn_skirt_rear_r_03 {parent=dyn_skirt_rear_r_02, loc=<-10 0 0>, rot=<0 0 0 -1>}
 dyn_skirt_rear_r_04_root {parent=dyn_skirt_rear_r_01, loc=<-6.8889 -1.0601 -7.2778>, rot=<0.3867 0.0655 -0.0176 -0.9197>}
 dyn_skirt_rear_r_04 {parent=dyn_skirt_rear_r_04_root, loc=<0 0 0>, rot=<0 0 0 1>}
 dyn_skirt_rear_r_05 {parent=dyn_skirt_rear_r_04, loc=<-8.2605 0 0>, rot=<0 0 0 1>}
 dyn_skirt_rear_l_01_base {parent=pelvis, loc=<4.7409 -14.0598 -8.3132>, rot=<0.9844 -0.0143 0.0346 -0.1717>}
 dyn_skirt_rear_l_01 {parent=dyn_skirt_rear_l_01_base, loc=<0 0 0>, rot=<0 0 0 1>}
 dyn_skirt_rear_l_02 {parent=dyn_skirt_rear_l_01, loc=<-10 0 0>, rot=<0 0 0 -1>}
 dyn_skirt_rear_l_03 {parent=dyn_skirt_rear_l_02, loc=<-10 0 0>, rot=<0 0 0.0552 -0.9985>}
 dyn_skirt_rear_l_04 {parent=dyn_skirt_rear_l_03, loc=<-9.0437 0 0>, rot=<0 0 0 1>}
 dyn_skirt_side_l_01_base {parent=pelvis, loc=<8.2733 -2.1542 -14.2471>, rot=<0.0727 -0.7247 -0.6797 -0.0872>}
 dyn_skirt_side_l_01 {parent=dyn_skirt_side_l_01_base, loc=<0 0 0>, rot=<0 0 0 1>}
 dyn_skirt_side_l_02 {parent=dyn_skirt_side_l_01, loc=<10.0001 0 0>, rot=<0 0 0 -1>}
 dyn_skirt_side_l_03 {parent=dyn_skirt_side_l_02, loc=<9.9999 0 0>, rot=<0 0 0 -1>}
 dyn_skirt_side_l_04 {parent=dyn_skirt_side_l_03, loc=<10.3072 -1.2793 -4.9876>, rot=<-0.1987 -0.0097 0.0478 -0.9789>}
 dyn_skirt_side_l_05 {parent=dyn_skirt_side_l_03, loc=<9.6461 -1.9612 3.0846>, rot=<0.1382 0 0 -0.9904>}
 dyn_skirt_front_l_01_base {parent=pelvis, loc=<9.3664 10.2445 -7.322>, rot=<0.126 0.0305 0.0329 -0.991>}
 dyn_skirt_front_l_01 {parent=dyn_skirt_front_l_01_base, loc=<0 0 0>, rot=<0 0 0 1>}
 dyn_skirt_front_l_02 {parent=dyn_skirt_front_l_01, loc=<-8.6804 0 0>, rot=<0 0 0 -1>}
 dyn_skirt_armor_l_base {parent=pelvis, loc=<13.0777 -2.0234 -14.6809>, rot=<-0.0754 0.6952 0.7075 0.1025>}
 dyn_skirt_armor_l {parent=dyn_skirt_armor_l_base, loc=<0 0 0>, rot=<0 0 0 1>}
 dyn_skirt_armor_01_l {parent=dyn_skirt_armor_l, loc=<5.7402 -0.452 -0.0666>, rot=<0 0 0 -1>}
 dyn_skirt_armor_r_base {parent=pelvis, loc=<13.0781 -2.0234 14.6809>, rot=<-0.6952 -0.0754 0.1025 -0.7075>}
 dyn_skirt_armor_r {parent=dyn_skirt_armor_r_base, loc=<0 0 0>, rot=<0 0 0 1>}
 dyn_skirt_armor_01_r {parent=dyn_skirt_armor_r, loc=<-5.741 0.4518 0.0666>, rot=<0 0 0 1>}
 dyn_skirt_front_angle_l_01_base {parent=pelvis, loc=<9.8319 5.4357 -12.1845>, rot=<0.3874 0.0477 0.1043 -0.9148>}
 dyn_skirt_front_angle_l_01 {parent=dyn_skirt_front_angle_l_01_base, loc=<0 0 0>, rot=<0 0 0 1>}
 dyn_skirt_front_angle_l_02 {parent=dyn_skirt_front_angle_l_01, loc=<-12.1507 0 0>, rot=<0 0 0 -1>}
 ik_hand_root {parent=root, loc=<0 0 0>, rot=<0 0 0 1>}
 ik_hand_gun {parent=ik_hand_root, loc=<-64.462 4.8638 101.1961>, rot=<0.6084 0.1793 0.4613 -0.6205>}
 ik_hand_r {parent=ik_hand_gun, loc=<0 0 0>, rot=<0 0 0 1>}
]
variables [
 DisablePickaxeElbows {type=bool, default=False}
 IsBuiltInPickaxeEquipped {type=bool, default=False}
 IsParachuteOpen {type=bool, default=False}
]
