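# Frontend physics asset for the large male PeacefulPoem Pumped cosmetic body, covering its dangling skirt and armor panels.
# Capsules on pelvis, thighs and calves form the lower body, while capsule chains shape the skirt and boxes shape the armor.
# Tiny base bodies join the pelvis with free rotation, and most segments below use tight angular limits.
# Free-rotation cross-constraints tie the right front chain to the right side chain.
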PhysicsAsset /BRCosmetics/Characters/Player/Male/Large/Bodies/M_LRG_PeacefulPoem_Pumped/Meshes/M_LRG_PeacefulPoem_Pumped_FrontendPhysics
bodies {
 calf_l [{shapeType=Capsule, length=29.4906, radius=9.211, center=<-13.9353 -0.6719 0.9981>, rotation=<rot 0 99.4949 -87.3898>}]
 calf_r [{shapeType=Capsule, length=29.4906, radius=9.211, center=<13.9353 0.6719 -0.9981>, rotation=<rot 0 99.4949 92.6102>}]
 dyn_skirt_armor_01_l [{
   shapeType=Box
   extentX=35.875
   extentY=7.5
   extentZ=23
   shapeName=dyn_skirt_armor_01_l_box
   center=<17.5294 -10.583 0>
   rotation=<rot 0.0438 -0.7112 -0.0118>
  }]
 dyn_skirt_armor_01_r [{
   shapeType=Box
   extentX=35.875
   extentY=7.5
   extentZ=23
   shapeName=dyn_skirt_armor_01_l_box
   center=<-17.5294 10.583 0>
   rotation=<rot 0.0438 -0.7112 179.9882>
  }]
 dyn_skirt_armor_l [{
   shapeType=Box
   extentX=35.8125
   extentY=7.6875
   extentZ=21.6875
   shapeName=dyn_skirt_armor_r_box
   center=<17.9846 -11.1835 -0.5505>
   rotation=<rot 0.0003 0 178.061>
  }]
 dyn_skirt_armor_l_base [{shapeType=Capsule, length=0.505, radius=0.505, shapeName=dyn_skirt_armor_r_base_capsule, rotation=<rot 0 -90 180>}]
 dyn_skirt_armor_r [{
   shapeType=Box
   extentX=35.8125
   extentY=7.6875
   extentZ=21.6875
   shapeName=dyn_skirt_armor_r_box
   center=<-17.9846 11.1835 0.5505>
   rotation=<rot 0.0003 0 -1.939>
  }]
 dyn_skirt_armor_r_base [{shapeType=Capsule, length=0.505, radius=0.505, shapeName=dyn_skirt_armor_r_base_capsule, rotation=<rot 0 -90 0>}]
 dyn_skirt_front_angle_l_01 [{
   shapeType=Capsule
   length=4.8248
   radius=2.8084
   shapeName=dyn_skirt_front_angle_l_01_capsule
   center=<-9.5012 3.0087 0.7082>
   rotation=<rot -11.0063 -90 -41.1818>
  }]
 dyn_skirt_front_angle_l_01_base [{shapeType=Capsule, length=0.505, radius=0.505, shapeName=dyn_skirt_front_angle_l_01_base_capsule, rotation=<rot 0 -90 0>}]
 dyn_skirt_front_angle_l_02 [{
   shapeType=Capsule
   length=6.8077
   radius=3.484
   shapeName=dyn_skirt_front_angle_l_02_capsule
   center=<-5.6895 2.3058 -0.1031>
   rotation=<rot 0 89.7125 -70.1958>
  }]
 dyn_skirt_front_l_01 [{
   shapeType=Capsule
   length=17.3089
   radius=3.5
   shapeName=dyn_skirt_front_l_01_capsule
   center=<-4.7114 5.6758 -0.1905>
   rotation=<rot 3.2 -90 -22.9965>
  }]
 dyn_skirt_front_l_01_base [{shapeType=Capsule, length=0.505, radius=0.505, shapeName=dyn_skirt_front_l_01_base_capsule, rotation=<rot 0 -90 0>}]
 dyn_skirt_front_l_02 [{
   shapeType=Capsule
   length=15.4907
   radius=4.2188
   shapeName=dyn_skirt_front_l_02_capsule
   center=<-2.9474 2.6691 -0.321>
   rotation=<rot -0.9974 -90 -33.4672>
  }]
 dyn_skirt_front_r_01 [{
   shapeType=Capsule
   length=9.9375
   radius=4.2188
   shapeName=dyn_skirt_front_r_01_capsule
   center=<-5.638 5.368 -4.0449>
   rotation=<rot 13.865 -90.0001 0>
  }]
 dyn_skirt_front_r_01_base [{shapeType=Capsule, length=0.505, radius=0.505, shapeName=dyn_skirt_front_r_01_base_capsule, rotation=<rot 0 -90 0>}]
 dyn_skirt_front_r_02 [{
   shapeType=Capsule
   length=19.7188
   radius=5.4062
   shapeName=dyn_skirt_front_r_02_capsule
   center=<-5.6253 5.2123 -5.8081>
   rotation=<rot 23.8974 -85.2159 1.755>
  }]
 dyn_skirt_front_r_03 [{
   shapeType=Capsule
   length=20.5625
   radius=5.4062
   shapeName=dyn_skirt_front_r_03_capsule
   center=<-5.8331 3.4112 -4.5348>
   rotation=<rot 22.9147 -89.8371 -1.2027>
  }]
 dyn_skirt_rear_l_01 [{
   shapeType=Capsule
   length=7
   radius=5
   shapeName=dyn_skirt_rear_l_01_capsule
   center=<-1.702 3.241 -0.1995>
   rotation=<rot 2.3004 -85.99 2.3497>
  }]
 dyn_skirt_rear_l_01_base [{shapeType=Capsule, length=0.505, radius=0.505, shapeName=dyn_skirt_rear_l_01_base_capsule, rotation=<rot 0 -90 0>}]
 dyn_skirt_rear_l_02 [{
   shapeType=Capsule
   length=5.5312
   radius=5
   shapeName=dyn_skirt_rear_l_02_capsule
   center=<-2.8433 1.2149 -1.2879>
   rotation=<rot -14.6396 92.1453 0>
  }]
 dyn_skirt_rear_l_03 [{
   shapeType=Capsule
   length=4.2812
   radius=3.4062
   shapeName=dyn_skirt_rear_l_03_capsule
   center=<-3.2473 1.2441 -0.5732>
   rotation=<rot 10.9049 -90 0>
  }]
 dyn_skirt_rear_l_04 [{
   shapeType=Capsule
   length=9.7812
   radius=2.2188
   shapeName=dyn_skirt_rear_l_04_capsule
   center=<-6.4219 0.5997 0.5699>
   rotation=<rot 0 -97.3763 -91.679>
  }]
 dyn_skirt_rear_r_01 [{shapeType=Capsule, length=7, radius=5, shapeName=dyn_skirt_rear_r_01_capsule, center=<-2.9743 4.7499 0>, rotation=<rot 10.9477 -90 0>}]
 dyn_skirt_rear_r_01_base [{shapeType=Capsule, length=0.505, radius=0.505, shapeName=dyn_skirt_rear_r_01_base_capsule, rotation=<rot 0 -90 0>}]
 dyn_skirt_rear_r_02 [{
   shapeType=Capsule
   length=6.3438
   radius=4.9062
   shapeName=dyn_skirt_rear_r_02_capsule
   center=<-4.6086 0.7903 0.5425>
   rotation=<rot 7.279 -90 0.9789>
  }]
 dyn_skirt_rear_r_03 [{
   shapeType=Capsule
   length=11.4062
   radius=3.8438
   shapeName=dyn_skirt_rear_r_03_capsule
   center=<-8.3362 0.7368 0.3406>
   rotation=<rot 0 -92.9886 -88.287>
  }]
 dyn_skirt_rear_r_04 [{
   shapeType=Capsule
   length=3.3067
   radius=2.8038
   shapeName=dyn_skirt_rear_r_04_capsule
   center=<-3.3481 0.2095 0>
   rotation=<rot 0 -83.882 -87.1121>
  }]
 dyn_skirt_rear_r_05 [{
   shapeType=Capsule
   length=4.6404
   radius=2.9473
   shapeName=dyn_skirt_rear_r_05_capsule
   center=<-2.8299 0.1684 0>
   rotation=<rot 0 -92.7397 -84.0366>
  }]
 dyn_skirt_side_l_01 [{
   shapeType=Capsule
   length=9.4776
   radius=2.5224
   shapeName=dyn_skirt_side_l_01_capsule
   center=<4.4395 -4.2841 0>
   rotation=<rot 5.1186 -90 0>
  }]
 dyn_skirt_side_l_01_base [{shapeType=Capsule, length=0.505, radius=0.505, shapeName=dyn_skirt_side_l_01_base_capsule, rotation=<rot 0 -90 0>}]
 dyn_skirt_side_l_02 [
  {shapeType=Capsule, length=11.8181, radius=3.6506, shapeName=dyn_skirt_side_l_02_capsule, center=<4.568 -3.6483 0>, rotation=<rot 0 -90 0>}
  {
   shapeType=Capsule
   length=4.75
   radius=4.0938
   shapeName=dyn_skirt_side_l_02_capsule
   center=<5.022 2.446 -9.6841>
   rotation=<rot 50.6245 98.4301 96.2985>
  }
 ]
 dyn_skirt_side_l_03 [{shapeType=Capsule, length=14.4822, radius=3.7366, shapeName=dyn_skirt_side_l_03_capsule, center=<3.5881 -3.6988 0>, rotation=<rot 0 -90 0>}]
 dyn_skirt_side_l_04 [{
   shapeType=Capsule
   length=13.5322
   radius=1.189
   shapeName=dyn_skirt_side_l_04_capsule
   center=<6.3266 -0.3866 0.1444>
   rotation=<rot -0.1368 90.2857 -86.7392>
  }]
 dyn_skirt_side_l_05 [{
   shapeType=Capsule
   length=13.8236
   radius=1.2812
   shapeName=dyn_skirt_side_l_05_capsule
   center=<7.181 -0.4445 -0.9417>
   rotation=<rot 0 -94.4731 -85.6935>
  }]
 dyn_skirt_side_r_01 [{
   shapeType=Capsule
   length=11.9062
   radius=3.4062
   shapeName=dyn_skirt_side_l_01_capsule
   center=<-4.4395 4.2841 0>
   rotation=<rot 5.1186 -90 -180>
  }]
 dyn_skirt_side_r_01_base [{shapeType=Capsule, length=0.505, radius=0.505, shapeName=dyn_skirt_side_r_01_base_capsule, rotation=<rot 0 -90 0>}]
 dyn_skirt_side_r_02 [{
   shapeType=Capsule
   length=11.5
   radius=3.9688
   shapeName=dyn_skirt_side_l_02_capsule
   center=<-4.568 4.1367 -0.757>
   rotation=<rot 0 -90 180>
  }]
 dyn_skirt_side_r_03 [{shapeType=Capsule, length=14.5312, radius=3.6875, shapeName=dyn_skirt_side_l_03_capsule, center=<-3.5881 3.6988 0>, rotation=<rot 0 -90 180>}]
 dyn_skirt_side_r_04 [{
   shapeType=Capsule
   length=2.9687
   radius=2.232
   shapeName=dyn_skirt_side_r_04_capsule
   center=<-2.1584 0.465 -0.6723>
   rotation=<rot 0 88.1084 -62.9689>
  }]
 dyn_skirt_side_r_05 [{
   shapeType=Capsule
   length=4.3395
   radius=2.176
   shapeName=dyn_skirt_side_r_05_capsule
   center=<-4.4559 1.0459 -1.0875>
   rotation=<rot 0 84.0599 -71.7447>
  }]
 pelvis [{shapeType=Capsule, length=7.8677, radius=12.1954, center=<-0.6221 -1.4272 0>}]
 thigh_l [{shapeType=Capsule, length=39.1735, radius=10.7292, center=<-15.8299 0.1186 5.2313>, rotation=<rot -6.384 88.5573 -87.1086>}]
 thigh_r [{shapeType=Capsule, length=38.8922, radius=11.0105, center=<15.7488 0.0728 -3.5395>, rotation=<rot -6.384 88.5573 92.8914>}]
}
constraints [
 {bone1Name=calf_l, bone2Name=thigh_l, swing1LimitDegrees=45, swing2LimitDegrees=45, twistLimitDegrees=45}
 {
  bone1Name=dyn_skirt_front_l_01_base
  bone2Name=pelvis
  swing1LimitDegrees=45
  swing1Motion=Free
  swing2LimitDegrees=45
  swing2Motion=Free
  twistLimitDegrees=45
  twistMotion=Free
 }
 {bone1Name=dyn_skirt_front_l_01, bone2Name=dyn_skirt_front_l_01_base, swing1LimitDegrees=8, swing2LimitDegrees=8, twistLimitDegrees=8}
 {bone1Name=dyn_skirt_front_l_02, bone2Name=dyn_skirt_front_l_01, swing1LimitDegrees=8, swing2LimitDegrees=8, twistLimitDegrees=8}
 {
  bone1Name=dyn_skirt_front_r_01_base
  bone2Name=pelvis
  swing1LimitDegrees=45
  swing1Motion=Free
  swing2LimitDegrees=45
  swing2Motion=Free
  twistLimitDegrees=45
  twistMotion=Free
 }
 {bone1Name=dyn_skirt_front_r_01, bone2Name=dyn_skirt_front_r_01_base, swing1LimitDegrees=18, swing2LimitDegrees=4, twistLimitDegrees=2}
 {bone1Name=dyn_skirt_front_r_02, bone2Name=dyn_skirt_front_r_01, swing1LimitDegrees=18, swing2LimitDegrees=4, twistLimitDegrees=12}
 {bone1Name=dyn_skirt_front_r_03, bone2Name=dyn_skirt_front_r_02, swing1LimitDegrees=18, swing2LimitDegrees=4, twistLimitDegrees=12}
 {
  bone1Name=dyn_skirt_rear_l_01_base
  bone2Name=pelvis
  swing1LimitDegrees=45
  swing1Motion=Free
  swing2LimitDegrees=45
  swing2Motion=Free
  twistLimitDegrees=45
  twistMotion=Free
 }
 {bone1Name=dyn_skirt_rear_l_01, bone2Name=dyn_skirt_rear_l_01_base, swing1LimitDegrees=14, swing2LimitDegrees=14, twistLimitDegrees=14}
 {bone1Name=dyn_skirt_rear_l_02, bone2Name=dyn_skirt_rear_l_01, swing1LimitDegrees=28, swing2LimitDegrees=28, twistLimitDegrees=12}
 {bone1Name=dyn_skirt_rear_l_03, bone2Name=dyn_skirt_rear_l_02, swing1LimitDegrees=28, swing2LimitDegrees=28, twistLimitDegrees=12}
 {bone1Name=dyn_skirt_rear_l_04, bone2Name=dyn_skirt_rear_l_03, swing1LimitDegrees=28, swing2LimitDegrees=28, twistLimitDegrees=12}
 {
  bone1Name=dyn_skirt_rear_r_01_base
  bone2Name=pelvis
  swing1LimitDegrees=45
  swing1Motion=Free
  swing2LimitDegrees=45
  swing2Motion=Free
  twistLimitDegrees=45
  twistMotion=Free
 }
 {bone1Name=dyn_skirt_rear_r_01, bone2Name=dyn_skirt_rear_r_01_base, swing1LimitDegrees=14, swing2LimitDegrees=14, twistLimitDegrees=14}
 {bone1Name=dyn_skirt_rear_r_02, bone2Name=dyn_skirt_rear_r_01, swing1LimitDegrees=28, swing2LimitDegrees=28, twistLimitDegrees=12}
 {bone1Name=dyn_skirt_rear_r_03, bone2Name=dyn_skirt_rear_r_02, swing1LimitDegrees=28, swing2LimitDegrees=28, twistLimitDegrees=12}
 {
  bone1Name=dyn_skirt_side_l_01_base
  bone2Name=pelvis
  swing1LimitDegrees=45
  swing1Motion=Free
  swing2LimitDegrees=45
  swing2Motion=Free
  twistLimitDegrees=45
  twistMotion=Free
 }
 {bone1Name=dyn_skirt_side_l_01, bone2Name=dyn_skirt_side_l_01_base, swing1LimitDegrees=12, swing2LimitDegrees=12, twistLimitDegrees=8}
 {bone1Name=dyn_skirt_side_l_02, bone2Name=dyn_skirt_side_l_01, swing1LimitDegrees=12, swing2LimitDegrees=12, twistLimitDegrees=8}
 {bone1Name=dyn_skirt_side_l_03, bone2Name=dyn_skirt_side_l_02, swing1LimitDegrees=12, swing2LimitDegrees=12, twistLimitDegrees=8}
 {bone1Name=dyn_skirt_side_l_04, bone2Name=dyn_skirt_side_l_03, swing1LimitDegrees=14, swing2LimitDegrees=14, twistLimitDegrees=14}
 {bone1Name=dyn_skirt_side_l_05, bone2Name=dyn_skirt_side_l_03, swing1LimitDegrees=14, swing2LimitDegrees=14, twistLimitDegrees=14}
 {
  bone1Name=dyn_skirt_side_r_01_base
  bone2Name=pelvis
  swing1LimitDegrees=45
  swing1Motion=Free
  swing2LimitDegrees=45
  swing2Motion=Free
  twistLimitDegrees=45
  twistMotion=Free
 }
 {bone1Name=calf_r, bone2Name=thigh_r, swing1LimitDegrees=45, swing2LimitDegrees=45, twistLimitDegrees=45}
 {bone1Name=dyn_skirt_rear_r_04, bone2Name=dyn_skirt_rear_r_01, swing1LimitDegrees=12, swing2LimitDegrees=12, twistLimitDegrees=12}
 {bone1Name=dyn_skirt_rear_r_05, bone2Name=dyn_skirt_rear_r_04, swing1LimitDegrees=12, swing2LimitDegrees=12, twistLimitDegrees=12}
 {
  bone1Name=dyn_skirt_armor_r_base
  bone2Name=pelvis
  swing1LimitDegrees=45
  swing1Motion=Free
  swing2LimitDegrees=45
  swing2Motion=Free
  twistLimitDegrees=45
  twistMotion=Free
 }
 {
  bone1Name=dyn_skirt_armor_l_base
  bone2Name=pelvis
  swing1LimitDegrees=45
  swing1Motion=Free
  swing2LimitDegrees=45
  swing2Motion=Free
  twistLimitDegrees=45
  twistMotion=Free
 }
 {bone1Name=thigh_l, bone2Name=pelvis, swing1LimitDegrees=45, swing2LimitDegrees=45, twistLimitDegrees=45}
 {
  bone1Name=dyn_skirt_front_angle_l_01_base
  bone2Name=pelvis
  swing1LimitDegrees=45
  swing1Motion=Free
  swing2LimitDegrees=45
  swing2Motion=Free
  twistLimitDegrees=45
  twistMotion=Free
 }
 {
  bone1Name=dyn_skirt_front_angle_l_01
  bone2Name=dyn_skirt_front_angle_l_01_base
  swing1LimitDegrees=12
  swing2LimitDegrees=12
  twistLimitDegrees=2
 }
 {
  bone1Name=dyn_skirt_front_angle_l_02
  bone2Name=dyn_skirt_front_angle_l_01
  swing1LimitDegrees=12
  swing2LimitDegrees=12
  twistLimitDegrees=2
 }
 {bone1Name=dyn_skirt_armor_l, bone2Name=dyn_skirt_armor_l_base, swing1LimitDegrees=12, swing2LimitDegrees=12, twistLimitDegrees=12}
 {
  bone1Name=dyn_skirt_armor_01_l
  bone2Name=dyn_skirt_armor_l
  swing1LimitDegrees=45
  swing1Motion=Free
  swing2LimitDegrees=45
  swing2Motion=Free
  twistLimitDegrees=45
  twistMotion=Free
 }
 {
  bone1Name=thigh_r
  bone2Name=pelvis
  swing1LimitDegrees=45
  swing1Motion=Free
  swing2LimitDegrees=45
  swing2Motion=Free
  twistLimitDegrees=45
  twistMotion=Free
 }
 {bone1Name=dyn_skirt_side_r_01, bone2Name=dyn_skirt_side_r_01_base, swing1LimitDegrees=12, swing2LimitDegrees=12, twistLimitDegrees=8}
 {bone1Name=dyn_skirt_side_r_02, bone2Name=dyn_skirt_side_r_01, swing1LimitDegrees=12, swing2LimitDegrees=12, twistLimitDegrees=8}
 {bone1Name=dyn_skirt_side_r_03, bone2Name=dyn_skirt_side_r_02, swing1LimitDegrees=12, swing2LimitDegrees=12, twistLimitDegrees=8}
 {bone1Name=dyn_skirt_armor_r, bone2Name=dyn_skirt_armor_r_base, swing1LimitDegrees=12, swing2LimitDegrees=12, twistLimitDegrees=12}
 {
  bone1Name=dyn_skirt_armor_01_r
  bone2Name=dyn_skirt_armor_r
  swing1LimitDegrees=45
  swing1Motion=Free
  swing2LimitDegrees=45
  swing2Motion=Free
  twistLimitDegrees=45
  twistMotion=Free
 }
 {
  bone1Name=dyn_skirt_front_r_02
  bone2Name=dyn_skirt_side_r_02
  swing1LimitDegrees=45
  swing1Motion=Free
  swing2LimitDegrees=45
  swing2Motion=Free
  twistLimitDegrees=45
  twistMotion=Free
 }
 {
  bone1Name=dyn_skirt_front_r_03
  bone2Name=dyn_skirt_side_r_03
  swing1LimitDegrees=45
  swing1Motion=Free
  swing2LimitDegrees=45
  swing2Motion=Free
  twistLimitDegrees=45
  twistMotion=Free
 }
 {bone1Name=dyn_skirt_side_r_04, bone2Name=dyn_skirt_side_r_03, swing1LimitDegrees=10, swing2LimitDegrees=10, twistLimitDegrees=10}
 {bone1Name=dyn_skirt_side_r_05, bone2Name=dyn_skirt_side_r_03, swing1LimitDegrees=10, swing2LimitDegrees=10, twistLimitDegrees=10}
]
