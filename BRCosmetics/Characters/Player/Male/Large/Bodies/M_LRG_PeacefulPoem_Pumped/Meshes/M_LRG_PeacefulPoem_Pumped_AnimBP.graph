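# Anim Blueprint for the PeacefulPoem Pumped large male cosmetic body, adding secondary motion and built-in pickaxe blade animation.
# The copied parent pose runs through Control Rig into AnimDynamics pec cloth, then bool-selected RigidBody setups using armor, body, crouch or frontend physics assets.
# An equip state machine is layered onto blade bones before corrective bone offsets on blades and shoulder spikes.
# The event graph computes gravity overrides per movement state.
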AnimBlueprint /BRCosmetics/Characters/Player/Male/Large/Bodies/M_LRG_PeacefulPoem_Pumped/Meshes/M_LRG_PeacefulPoem_Pumped_AnimBP
target_skeleton=&/M_LRG_PeacefulPoem_Pumped_Skeleton
graphs [
 AnimGraph {nodes [
   Local To Component(LocalPose=Use cached pose 'PostPickaxeAnim'())
   n1 = "Transform (Modify) Bone - Bone: dyn_pec_cloth_l_01"!(ComponentPose, bAlphaBoolEnabled:bool=True, Translation="0,0,0", Rotation="-5,11.04,5", Scale="1,1,1") -> (Pose) {
    alphaInputType=Bool
    rotationMode=BMM_Additive
    rotationSpace=BCS_BoneSpace
    translationMode=BMM_Additive
    translationSpace=BCS_BoneSpace
    alphaBoolBlend {blendOption=Linear}
    boneToModify {boneName=dyn_pec_cloth_l_01}
    rotation=<rot -5 11.04 5>
   }
   "Transform (Modify) Bone - Bone: dyn_pec_cloth_r_01"!(ComponentPose=@n1, bAlphaBoolEnabled:bool=True, Translation="-1.5,0,0.5", Rotation="-10,15,4", Scale="1,1,1") -> (Pose) {
    alphaInputType=Bool
    rotationMode=BMM_Additive
    rotationSpace=BCS_BoneSpace
    translationMode=BMM_Additive
    translationSpace=BCS_BoneSpace
    alphaBoolBlend {blendOption=Linear}
    boneToModify {boneName=dyn_pec_cloth_r_01}
    rotation=<rot -10 15 4>
    translation {x=-1.5, z=0.5}
   }
   n2 = "Transform (Modify) Bone - Bone: deform_pickaxe_blade_a_01_l"!(
    ComponentPose=Local To Component(LocalPose=Use cached pose 'PostPickaxeAnim'())
    bAlphaBoolEnabled:bool=True
    Translation="3,0,0"
    Rotation="5,0,0"
    Scale="0.8,1,1"
   ) -> (Pose) {
    alphaInputType=Bool
    rotationMode=BMM_Additive
    rotationSpace=BCS_BoneSpace
    scaleMode=BMM_Additive
    scaleSpace=BCS_BoneSpace
    translationMode=BMM_Additive
    translationSpace=BCS_BoneSpace
    alphaBoolBlend {blendInTime=0.15, blendOption=QuadraticInOut, blendOutTime=0.15}
    boneToModify {boneName=deform_pickaxe_blade_a_01_l}
    rotation {pitch=5}
    scale {x=0.8}
    translation {x=3}
   }
   n3 = "Transform (Modify) Bone - Bone: deform_pickaxe_blade_a_02_l"!(ComponentPose=@n2, bAlphaBoolEnabled:bool=True, Translation="0,0,0", Rotation="0,0,0", Scale="0.8,1,1") -> (Pose) {
    alphaInputType=Bool
    rotationMode=BMM_Additive
    rotationSpace=BCS_BoneSpace
    scaleMode=BMM_Additive
    scaleSpace=BCS_BoneSpace
    translationMode=BMM_Additive
    translationSpace=BCS_BoneSpace
    alphaBoolBlend {blendInTime=0.15, blendOption=QuadraticInOut, blendOutTime=0.15}
    boneToModify {boneName=deform_pickaxe_blade_a_02_l}
    scale {x=0.8}
   }
   n4 = "Transform (Modify) Bone - Bone: deform_pickaxe_blade_b_01_l"!(ComponentPose=@n3, bAlphaBoolEnabled:bool=True, Translation="3,0,0", Rotation="5,0,0", Scale="0.8,1,1") -> (Pose) {
    alphaInputType=Bool
    rotationMode=BMM_Additive
    rotationSpace=BCS_BoneSpace
    scaleMode=BMM_Additive
    scaleSpace=BCS_BoneSpace
    translationMode=BMM_Additive
    translationSpace=BCS_BoneSpace
    alphaBoolBlend {blendInTime=0.15, blendOption=QuadraticInOut, blendOutTime=0.15}
    boneToModify {boneName=deform_pickaxe_blade_b_01_l}
    rotation {pitch=5}
    scale {x=0.8}
    translation {x=3}
   }
   n5 = "Transform (Modify) Bone - Bone: deform_pickaxe_blade_b_02_l"!(ComponentPose=@n4, bAlphaBoolEnabled:bool=True, Translation="0,0,0", Rotation="0,0,0", Scale="0.8,1,1") -> (Pose) {
    alphaInputType=Bool
    rotationMode=BMM_Additive
    rotationSpace=BCS_BoneSpace
    scaleMode=BMM_Additive
    scaleSpace=BCS_BoneSpace
    translationMode=BMM_Additive
    translationSpace=BCS_BoneSpace
    alphaBoolBlend {blendInTime=0.15, blendOption=QuadraticInOut, blendOutTime=0.15}
    boneToModify {boneName=deform_pickaxe_blade_b_02_l}
    scale {x=0.8}
   }
   n6 = "Transform (Modify) Bone - Bone: deform_pickaxe_blade_a_01_r"!(ComponentPose=@n5, bAlphaBoolEnabled:bool=True, Translation="0,0,0", Rotation="5,0,0", Scale="0.85,1,1") -> (Pose) {
    alphaInputType=Bool
    rotationMode=BMM_Additive
    rotationSpace=BCS_BoneSpace
    scaleMode=BMM_Replace
    scaleSpace=BCS_BoneSpace
    alphaBoolBlend {blendInTime=0.15, blendOption=QuadraticInOut, blendOutTime=0.15}
    boneToModify {boneName=deform_pickaxe_blade_a_01_r}
    rotation {pitch=5}
    scale {x=0.85}
   }
   n7 = "Transform (Modify) Bone - Bone: deform_pickaxe_blade_a_02_r"!(ComponentPose=@n6, bAlphaBoolEnabled:bool=True, Translation="0,0,0", Rotation="0,0,0", Scale="0.85,1,1") -> (Pose) {
    alphaInputType=Bool
    rotationMode=BMM_Additive
    rotationSpace=BCS_BoneSpace
    scaleMode=BMM_Replace
    scaleSpace=BCS_BoneSpace
    alphaBoolBlend {blendInTime=0.15, blendOption=QuadraticInOut, blendOutTime=0.15}
    boneToModify {boneName=deform_pickaxe_blade_a_02_r}
    scale {x=0.85}
   }
   n8 = "Transform (Modify) Bone - Bone: deform_pickaxe_blade_b_01_r"!(ComponentPose=@n7, bAlphaBoolEnabled:bool=True, Translation="0,0,0", Rotation="5,0,0", Scale="0.85,1,1") -> (Pose) {
    alphaInputType=Bool
    rotationMode=BMM_Additive
    rotationSpace=BCS_BoneSpace
    scaleMode=BMM_Replace
    scaleSpace=BCS_BoneSpace
    alphaBoolBlend {blendInTime=0.15, blendOption=QuadraticInOut, blendOutTime=0.15}
    boneToModify {boneName=deform_pickaxe_blade_b_01_r}
    rotation {pitch=5}
    scale {x=0.85}
   }
   n9 = "Transform (Modify) Bone - Bone: deform_pickaxe_blade_b_02_r"!(ComponentPose=@n8, bAlphaBoolEnabled:bool=True, Translation="0,0,0", Rotation="0,0,0", Scale="0.85,1,1") -> (Pose) {
    alphaInputType=Bool
    rotationMode=BMM_Additive
    rotationSpace=BCS_BoneSpace
    scaleMode=BMM_Replace
    scaleSpace=BCS_BoneSpace
    alphaBoolBlend {blendInTime=0.15, blendOption=QuadraticInOut, blendOutTime=0.15}
    boneToModify {boneName=deform_pickaxe_blade_b_02_r}
    scale {x=0.85}
   }
   n10 = "Transform (Modify) Bone - Bone: deform_shoulderpad_front_spike_l"!(ComponentPose=@n9, bAlphaBoolEnabled:bool=True, Translation="0,0,0", Rotation="-20,0,0", Scale="1,1,1") -> (Pose) {
    alphaInputType=Bool
    rotationMode=BMM_Additive
    rotationSpace=BCS_BoneSpace
    alphaBoolBlend {blendInTime=0.15, blendOption=QuadraticInOut, blendOutTime=0.15}
    boneToModify {boneName=deform_shoulderpad_front_spike_l}
    rotation {pitch=-20}
   }
   n11 = "Transform (Modify) Bone - Bone: deform_shoulderpad_rear_spike_l"!(ComponentPose=@n10, bAlphaBoolEnabled:bool=True, Translation="0,0,0", Rotation="-20,0,0", Scale="1,1,1") -> (Pose) {
    alphaInputType=Bool
    rotationMode=BMM_Additive
    rotationSpace=BCS_BoneSpace
    alphaBoolBlend {blendInTime=0.15, blendOption=QuadraticInOut, blendOutTime=0.15}
    boneToModify {boneName=deform_shoulderpad_rear_spike_l}
    rotation {pitch=-20}
   }
   n12 = "Transform (Modify) Bone - Bone: deform_shoulderpad_front_spike_r"!(ComponentPose=@n11, bAlphaBoolEnabled:bool=True, Translation="0,0,0", Rotation="-20,0,0", Scale="1,1,1") -> (Pose) {
    alphaInputType=Bool
    rotationMode=BMM_Additive
    rotationSpace=BCS_BoneSpace
    alphaBoolBlend {blendInTime=0.15, blendOption=QuadraticInOut, blendOutTime=0.15}
    boneToModify {boneName=deform_shoulderpad_front_spike_r}
    rotation {pitch=-20}
   }
   n13 = "Transform (Modify) Bone - Bone: deform_shoulderpad_rear_spike_r"!(ComponentPose=@n12, bAlphaBoolEnabled:bool=True, Translation="0,0,0", Rotation="-20,0,0", Scale="1,1,1") -> (Pose) {
    alphaInputType=Bool
    rotationMode=BMM_Additive
    rotationSpace=BCS_BoneSpace
    alphaBoolBlend {blendInTime=0.15, blendOption=QuadraticInOut, blendOutTime=0.15}
    boneToModify {boneName=deform_shoulderpad_rear_spike_r}
    rotation {pitch=-20}
   }
   n14 = "Transform (Modify) Bone - Bone: deform_pickaxe_blade_a_01_l"(ComponentPose=@n13, bAlphaBoolEnabled=True, Translation="0,0,0", Rotation="5,0,0", Scale="0.9,1,1") {
    alphaInputType=Bool
    rotationMode=BMM_Additive
    rotationSpace=BCS_BoneSpace
    scaleMode=BMM_Replace
    scaleSpace=BCS_BoneSpace
    alphaBoolBlend {blendInTime=0.15, blendOption=QuadraticInOut, blendOutTime=0.15}
    boneToModify {boneName=deform_pickaxe_blade_a_01_l}
    rotation {pitch=5}
    scale {x=0.9}
   }
   n15 = "Transform (Modify) Bone - Bone: deform_pickaxe_blade_a_02_l"(ComponentPose=@n14, bAlphaBoolEnabled=True, Translation="0,0,0", Rotation="0,0,0", Scale="0.9,1,1") {
    alphaInputType=Bool
    rotationMode=BMM_Additive
    rotationSpace=BCS_BoneSpace
    scaleMode=BMM_Replace
    scaleSpace=BCS_BoneSpace
    alphaBoolBlend {blendInTime=0.15, blendOption=QuadraticInOut, blendOutTime=0.15}
    boneToModify {boneName=deform_pickaxe_blade_a_02_l}
    scale {x=0.9}
   }
   n16 = "Transform (Modify) Bone - Bone: deform_pickaxe_blade_b_01_l"(ComponentPose=@n15, bAlphaBoolEnabled=True, Translation="0,0,0", Rotation="0,0,0", Scale="0.9,1,1") {
    alphaInputType=Bool
    rotationMode=BMM_Additive
    rotationSpace=BCS_BoneSpace
    scaleMode=BMM_Replace
    scaleSpace=BCS_BoneSpace
    alphaBoolBlend {blendInTime=0.15, blendOption=QuadraticInOut, blendOutTime=0.15}
    boneToModify {boneName=deform_pickaxe_blade_b_01_l}
    scale {x=0.9}
   }
   n17 = "Transform (Modify) Bone - Bone: deform_pickaxe_blade_b_02_l"(ComponentPose=@n16, bAlphaBoolEnabled=True, Translation="0,0,0", Rotation="0,0,0", Scale="0.9,1,1") {
    alphaInputType=Bool
    rotationMode=BMM_Additive
    rotationSpace=BCS_BoneSpace
    scaleMode=BMM_Replace
    scaleSpace=BCS_BoneSpace
    alphaBoolBlend {blendInTime=0.15, blendOption=QuadraticInOut, blendOutTime=0.15}
    boneToModify {boneName=deform_pickaxe_blade_b_02_l}
    scale {x=0.9}
   }
   n18 = "Transform (Modify) Bone - Bone: deform_pickaxe_blade_a_01_r"(ComponentPose=@n17, bAlphaBoolEnabled=True, Translation="0,0,0", Rotation="5,0,0", Scale="0.9,1,1") {
    alphaInputType=Bool
    rotationMode=BMM_Additive
    rotationSpace=BCS_BoneSpace
    scaleMode=BMM_Replace
    scaleSpace=BCS_BoneSpace
    alphaBoolBlend {blendInTime=0.15, blendOption=QuadraticInOut, blendOutTime=0.15}
    boneToModify {boneName=deform_pickaxe_blade_a_01_r}
    rotation {pitch=5}
    scale {x=0.9}
   }
   n19 = "Transform (Modify) Bone - Bone: deform_pickaxe_blade_a_02_r"(ComponentPose=@n18, bAlphaBoolEnabled=True, Translation="0,0,0", Rotation="0,0,0", Scale="0.9,1,1") {
    alphaInputType=Bool
    rotationMode=BMM_Additive
    rotationSpace=BCS_BoneSpace
    scaleMode=BMM_Replace
    scaleSpace=BCS_BoneSpace
    alphaBoolBlend {blendInTime=0.15, blendOption=QuadraticInOut, blendOutTime=0.15}
    boneToModify {boneName=deform_pickaxe_blade_a_02_r}
    scale {x=0.9}
   }
   n20 = "Transform (Modify) Bone - Bone: deform_pickaxe_blade_b_01_r"(ComponentPose=@n19, bAlphaBoolEnabled=True, Translation="0,0,0", Rotation="5,0,0", Scale="0.9,1,1") {
    alphaInputType=Bool
    rotationMode=BMM_Additive
    rotationSpace=BCS_BoneSpace
    scaleMode=BMM_Replace
    scaleSpace=BCS_BoneSpace
    alphaBoolBlend {blendInTime=0.15, blendOption=QuadraticInOut, blendOutTime=0.15}
    boneToModify {boneName=deform_pickaxe_blade_b_01_r}
    rotation {pitch=5}
    scale {x=0.9}
   }
   n21 = "Transform (Modify) Bone - Bone: deform_pickaxe_blade_b_02_r"(ComponentPose=@n20, bAlphaBoolEnabled=True, Translation="0,0,0", Rotation="0,0,0", Scale="0.9,1,1") {
    alphaInputType=Bool
    rotationMode=BMM_Additive
    rotationSpace=BCS_BoneSpace
    scaleMode=BMM_Replace
    scaleSpace=BCS_BoneSpace
    alphaBoolBlend {blendInTime=0.15, blendOption=QuadraticInOut, blendOutTime=0.15}
    boneToModify {boneName=deform_pickaxe_blade_b_02_r}
    scale {x=0.9}
   }
   n22 = "Transform (Modify) Bone - Bone: deform_shoulderpad_front_spike_l"(ComponentPose=@n21, bAlphaBoolEnabled=True, Translation="0,0,0", Rotation="-25,0,0", Scale="1,1,1") {
    alphaInputType=Bool
    rotationMode=BMM_Additive
    rotationSpace=BCS_BoneSpace
    alphaBoolBlend {blendInTime=0.15, blendOption=QuadraticInOut, blendOutTime=0.15}
    boneToModify {boneName=deform_shoulderpad_front_spike_l}
    rotation {pitch=-25}
   }
   n23 = "Transform (Modify) Bone - Bone: deform_shoulderpad_rear_spike_l"(ComponentPose=@n22, bAlphaBoolEnabled=True, Translation="0,0,0", Rotation="-25,0,0", Scale="1,1,1") {
    alphaInputType=Bool
    rotationMode=BMM_Additive
    rotationSpace=BCS_BoneSpace
    alphaBoolBlend {blendInTime=0.15, blendOption=QuadraticInOut, blendOutTime=0.15}
    boneToModify {boneName=deform_shoulderpad_rear_spike_l}
    rotation {pitch=-25}
   }
   n24 = "Transform (Modify) Bone - Bone: deform_shoulderpad_front_spike_r"(ComponentPose=@n23, bAlphaBoolEnabled=True, Translation="0,0,0", Rotation="-25,0,0", Scale="1,1,1") {
    alphaInputType=Bool
    rotationMode=BMM_Additive
    rotationSpace=BCS_BoneSpace
    alphaBoolBlend {blendInTime=0.15, blendOption=QuadraticInOut, blendOutTime=0.15}
    boneToModify {boneName=deform_shoulderpad_front_spike_r}
    rotation {pitch=-25}
   }
   n25 = "Transform (Modify) Bone - Bone: deform_shoulderpad_rear_spike_r"(ComponentPose=@n24, bAlphaBoolEnabled=True, Translation="0,0,0", Rotation="-25,0,0", Scale="1,1,1") {
    alphaInputType=Bool
    rotationMode=BMM_Additive
    rotationSpace=BCS_BoneSpace
    alphaBoolBlend {blendInTime=0.15, blendOption=QuadraticInOut, blendOutTime=0.15}
    boneToModify {boneName=deform_shoulderpad_rear_spike_r}
    rotation {pitch=-25}
   }
   Output Pose(Result=Component To Local(ComponentPose=@n25)) {layerGroup=None}
   n26 = Anim Dynamics - Bone: dyn_pec_cloth_l_02(ComponentPose=Local To Component(LocalPose=Use cached pose 'DeformPose'()), Alpha=1, ExternalForce="0,0,0") {
    angularDampingOverride=1
    angularSpringConstant=2000
    bAngularSpring=true
    bOverrideAngularDamping=true
    bUseGravityOverride=true
    gravityScale=0.35
    alphaBoolBlend {blendOption=Linear}
    boundBone {boneName=dyn_pec_cloth_l_02}
    componentAppliedLinearAccClamp=<500 500 500>
    componentLinearAccScale=<0.5 0.5 0.5>
    componentLinearVelScale=<0.5 0.5 0.5>
    gravityOverride {z=-980}
    physicsBodyDefinitions [{
      boundBone {boneName=dyn_pec_cloth_l_02}
      boxExtents=<10 20 15>
      constraintSetup {angularLimitsMax=<15 0 30>, angularLimitsMin=<-15 -15 -15>, angularTarget=<1 0 0>}
      localJointOffset=<4 0 0>
     }]
    planarLimits [{drivingBone {boneName=spine_04}, planeTransform {location=<0 15 0>, rotation=<rot 0 0 90>}}]
    relativeSpaceBone {boneName=spine_05}
   }
   n27 = Anim Dynamics - Bone: dyn_pec_cloth_l_03(ComponentPose=@n26, Alpha=1, ExternalForce="0,0,0") {
    angularDampingOverride=1
    angularSpringConstant=2000
    bAngularSpring=true
    bOverrideAngularDamping=true
    bUseGravityOverride=true
    bUsePlanarLimit=false
    gravityScale=0.35
    alphaBoolBlend {blendOption=Linear}
    boundBone {boneName=dyn_pec_cloth_l_03}
    componentAppliedLinearAccClamp=<500 500 500>
    componentLinearAccScale=<0.5 0.5 0.5>
    componentLinearVelScale=<0.5 0.5 0.5>
    gravityOverride {z=-980}
    physicsBodyDefinitions [{
      boundBone {boneName=dyn_pec_cloth_l_03}
      boxExtents=<10 20 15>
      constraintSetup {angularLimitsMax=<15 15 30>, angularLimitsMin=<-15 -15 -15>, angularTarget=<1 0 0>}
      localJointOffset=<4 0 0>
     }]
    planarLimits [{drivingBone {boneName=spine_04}, planeTransform {location=<0 15 0>, rotation=<rot 0 0 90>}}]
    relativeSpaceBone {boneName=dyn_pec_cloth_l_02}
   }
   n28 = Anim Dynamics - Bone: dyn_pec_cloth_l_04(ComponentPose=@n27, Alpha=1, ExternalForce="0,0,0") {
    angularDampingOverride=1
    angularSpringConstant=2000
    bAngularSpring=true
    bOverrideAngularDamping=true
    bUseGravityOverride=true
    bUsePlanarLimit=false
    gravityScale=0.35
    alphaBoolBlend {blendOption=Linear}
    boundBone {boneName=dyn_pec_cloth_l_04}
    componentAppliedLinearAccClamp=<500 500 500>
    componentLinearAccScale=<0.5 0.5 0.5>
    componentLinearVelScale=<0.5 0.5 0.5>
    gravityOverride {z=-980}
    physicsBodyDefinitions [{
      boundBone {boneName=dyn_pec_cloth_l_04}
      boxExtents=<10 20 15>
      constraintSetup {angularLimitsMax=<15 15 30>, angularLimitsMin=<-15 -15 -15>, angularTarget=<1 0 0>}
      localJointOffset=<4 0 0>
     }]
    planarLimits [{drivingBone {boneName=spine_04}, planeTransform {location=<0 15 0>, rotation=<rot 0 0 90>}}]
    relativeSpaceBone {boneName=dyn_pec_cloth_l_03}
   }
   n29 = Anim Dynamics - Bone: dyn_pec_cloth_l_05(ComponentPose=@n28, Alpha=1, ExternalForce="0,0,0") {
    angularDampingOverride=1
    angularSpringConstant=2000
    bAngularSpring=true
    bOverrideAngularDamping=true
    bUseGravityOverride=true
    bUsePlanarLimit=false
    gravityScale=0.35
    alphaBoolBlend {blendOption=Linear}
    boundBone {boneName=dyn_pec_cloth_l_05}
    componentAppliedLinearAccClamp=<500 500 500>
    componentLinearAccScale=<0.5 0.5 0.5>
    componentLinearVelScale=<0.5 0.5 0.5>
    gravityOverride {z=-980}
    physicsBodyDefinitions [{
      boundBone {boneName=dyn_pec_cloth_l_05}
      boxExtents=<10 20 15>
      constraintSetup {angularLimitsMax=<15 15 30>, angularLimitsMin=<-15 -15 -15>, angularTarget=<1 0 0>}
      localJointOffset=<4 0 0>
     }]
    planarLimits [{drivingBone {boneName=spine_04}, planeTransform {location=<0 15 0>, rotation=<rot 0 0 90>}}]
    relativeSpaceBone {boneName=dyn_pec_cloth_l_04}
   }
   n30 = Anim Dynamics - Bone: dyn_pec_cloth_r_02(ComponentPose=@n29, Alpha=1, ExternalForce="0,0,0") {
    angularDampingOverride=1
    angularSpringConstant=2000
    bAngularSpring=true
    bOverrideAngularDamping=true
    bUseGravityOverride=true
    gravityScale=0.35
    alphaBoolBlend {blendOption=Linear}
    boundBone {boneName=dyn_pec_cloth_r_02}
    componentAppliedLinearAccClamp=<500 500 500>
    componentLinearAccScale=<0.5 0.5 0.5>
    componentLinearVelScale=<0.5 0.5 0.5>
    gravityOverride {z=-980}
    physicsBodyDefinitions [{
      boundBone {boneName=dyn_pec_cloth_r_02}
      boxExtents=<10 20 15>
      constraintSetup {angularLimitsMax=<15 0 30>, angularLimitsMin=<-15 -15 -15>, angularTarget=<1 0 0>}
      localJointOffset=<-4 0 0>
     }]
    planarLimits [{drivingBone {boneName=spine_04}, planeTransform {location=<0 15 0>, rotation=<rot 0 0 90>}}]
    relativeSpaceBone {boneName=spine_05}
   }
   n31 = Anim Dynamics - Bone: dyn_pec_cloth_r_03(ComponentPose=@n30, Alpha=1, ExternalForce="0,0,0") {
    angularDampingOverride=1
    angularSpringConstant=2000
    bAngularSpring=true
    bOverrideAngularDamping=true
    bUseGravityOverride=true
    bUsePlanarLimit=false
    gravityScale=0.35
    alphaBoolBlend {blendOption=Linear}
    boundBone {boneName=dyn_pec_cloth_r_03}
    componentAppliedLinearAccClamp=<500 500 500>
    componentLinearAccScale=<0.5 0.5 0.5>
    componentLinearVelScale=<0.5 0.5 0.5>
    gravityOverride {z=-980}
    physicsBodyDefinitions [{
      boundBone {boneName=dyn_pec_cloth_r_03}
      boxExtents=<10 20 15>
      constraintSetup {angularLimitsMax=<15 15 30>, angularLimitsMin=<-15 -15 -15>, angularTarget=<1 0 0>}
      localJointOffset=<-4 0 0>
     }]
    planarLimits [{drivingBone {boneName=spine_04}, planeTransform {location=<0 15 0>, rotation=<rot 0 0 90>}}]
    relativeSpaceBone {boneName=dyn_pec_cloth_r_02}
   }
   n32 = Anim Dynamics - Bone: dyn_pec_cloth_r_04(ComponentPose=@n31, Alpha=1, ExternalForce="0,0,0") {
    angularDampingOverride=1
    angularSpringConstant=2000
    bAngularSpring=true
    bOverrideAngularDamping=true
    bUseGravityOverride=true
    bUsePlanarLimit=false
    gravityScale=0.35
    alphaBoolBlend {blendOption=Linear}
    boundBone {boneName=dyn_pec_cloth_r_04}
    componentAppliedLinearAccClamp=<500 500 500>
    componentLinearAccScale=<0.5 0.5 0.5>
    componentLinearVelScale=<0.5 0.5 0.5>
    gravityOverride {z=-980}
    physicsBodyDefinitions [{
      boundBone {boneName=dyn_pec_cloth_r_04}
      boxExtents=<10 20 15>
      constraintSetup {angularLimitsMax=<15 15 30>, angularLimitsMin=<-15 -15 -15>, angularTarget=<1 0 0>}
      localJointOffset=<-4 0 0>
     }]
    planarLimits [{drivingBone {boneName=spine_04}, planeTransform {location=<0 15 0>, rotation=<rot 0 0 90>}}]
    relativeSpaceBone {boneName=dyn_pec_cloth_r_02}
   }
   n33 = Anim Dynamics - Bone: dyn_pec_cloth_r_05(ComponentPose=@n32, Alpha=1, ExternalForce="0,0,0") {
    angularDampingOverride=1
    angularSpringConstant=2000
    bAngularSpring=true
    bOverrideAngularDamping=true
    bUseGravityOverride=true
    bUsePlanarLimit=false
    gravityScale=0.35
    alphaBoolBlend {blendOption=Linear}
    boundBone {boneName=dyn_pec_cloth_r_05}
    componentAppliedLinearAccClamp=<500 500 500>
    componentLinearAccScale=<0.5 0.5 0.5>
    componentLinearVelScale=<0.5 0.5 0.5>
    gravityOverride {z=-980}
    physicsBodyDefinitions [{
      boundBone {boneName=dyn_pec_cloth_r_05}
      boxExtents=<10 20 15>
      constraintSetup {angularLimitsMax=<15 15 30>, angularLimitsMin=<-15 -15 -15>, angularTarget=<1 0 0>}
      localJointOffset=<-4 0 0>
     }]
    planarLimits [{drivingBone {boneName=spine_04}, planeTransform {location=<0 15 0>, rotation=<rot 0 0 90>}}]
    relativeSpaceBone {boneName=dyn_pec_cloth_r_04}
   }
   Save cached pose 'Post Base Dynamics'(Pose=Component To Local(ComponentPose=@n33)) {cacheName=Post Base Dynamics}
   n34 = RigidBody!(
    ComponentPose=Local To Component(LocalPose=Use cached pose 'Post Base Dynamics'())
    Alpha=0.99
    OverrideWorldGravity:Vector="0,0,-980"
    ExternalForce
   ) -> (Pose) {
    bOverrideWorldGravity=true
    overridePhysicsAsset=&/M_LRG_PeacefulPoem_Pumped_Armor_Physics
    simulationSpace=BaseBoneSpace
    alphaBoolBlend {blendOption=Linear}
    baseBoneRef {boneName=pelvis}
    componentAppliedLinearAccClamp=<1200 1200 1200>
    componentLinearAccScale=<0.3 0.3 0.8>
    componentLinearVelScale=<0.3 0.3 0.8>
    overrideWorldGravity {z=-980}
    simSpaceSettings {maxAngularAcceleration=100, maxAngularVelocity=100, maxLinearAcceleration=50, maxLinearVelocity=50, worldAlpha=0.5}
   }
   n35 = RigidBody(ComponentPose=@n34, Alpha=0.7, OverrideWorldGravity="0,0,-980") {
    bClampLinearTranslationLimitToRefPose=true
    bOverrideWorldGravity=true
    overridePhysicsAsset=&/M_LRG_PeacefulPoem_Pumped_Crouch_Physics
    simulationSpace=BaseBoneSpace
    alphaBoolBlend {blendOption=Linear}
    baseBoneRef {boneName=pelvis}
    componentAppliedLinearAccClamp=<1400 1400 800>
    componentLinearAccScale=<0.5 0.5 0.2>
    componentLinearVelScale=<0.5 0.5 0.2>
    overrideWorldGravity {z=-980}
    simSpaceSettings {
     maxAngularAcceleration=100
     maxAngularVelocity=100
     maxLinearAcceleration=50
     maxLinearVelocity=50
     velocityScaleZ=0.5
     worldAlpha=0.15
     externalAngularVelocity=<0.3 0.3>
     externalLinearDragV=<0.3 0.3 0.5>
     externalLinearVelocity=<0.3 0.3>
    }
   }
   n36 = Component To Local(ComponentPose=@n35)
   n37 = RigidBody(ComponentPose=Local To Component(LocalPose=Use cached pose 'Post Base Dynamics'()), Alpha=0.99, OverrideWorldGravity="0,0,-980") {
    bOverrideWorldGravity=true
    overridePhysicsAsset=&M_LRG_PeacefulPoem_Pumped_Armor_Physics
    simulationSpace=BaseBoneSpace
    alphaBoolBlend {blendOption=Linear}
    baseBoneRef {boneName=pelvis}
    componentAppliedLinearAccClamp=<1200 1200 1200>
    componentLinearAccScale=<0.3 0.3 0.8>
    componentLinearVelScale=<0.3 0.3 0.8>
    overrideWorldGravity {z=-980}
    simSpaceSettings {maxAngularAcceleration=100, maxAngularVelocity=100, maxLinearAcceleration=50, maxLinearVelocity=50, worldAlpha=0.5}
   }
   n38 = RigidBody(ComponentPose=@n37, Alpha=0.95, OverrideWorldGravity="0,0,-980") {
    bClampLinearTranslationLimitToRefPose=true
    bOverrideWorldGravity=true
    overridePhysicsAsset=&/M_LRG_PeacefulPoem_Pumped_Physics
    simulationSpace=BaseBoneSpace
    alphaBoolBlend {blendOption=Linear}
    baseBoneRef {boneName=pelvis}
    componentAppliedLinearAccClamp=<1400 1400 800>
    componentLinearAccScale=<0.5 0.5 0.2>
    componentLinearVelScale=<0.5 0.5 0.2>
    overrideWorldGravity {z=-980}
    simSpaceSettings {
     maxAngularAcceleration=100
     maxAngularVelocity=100
     maxLinearAcceleration=50
     maxLinearVelocity=50
     velocityScaleZ=0.5
     worldAlpha=0.15
     externalAngularVelocity=<0.3 0.3>
     externalLinearDragV=<0.3 0.3 0.5>
     externalLinearVelocity=<0.3 0.3>
    }
   }
   n39 = Component To Local(ComponentPose=@n38)
   n40 = RigidBody(ComponentPose=Local To Component(LocalPose=Use cached pose 'Post Base Dynamics'()), Alpha=0.99, OverrideWorldGravity="0,0,-980") {
    bOverrideWorldGravity=true
    overridePhysicsAsset=&/M_LRG_PeacefulPoem_Pumped_Armor_FrontendPhysics
    simulationSpace=BaseBoneSpace
    alphaBoolBlend {blendOption=Linear}
    baseBoneRef {boneName=pelvis}
    componentAppliedLinearAccClamp=<1200 1200 1200>
    componentLinearAccScale=<0.3 0.3 0.8>
    componentLinearVelScale=<0.3 0.3 0.8>
    overrideWorldGravity {z=-980}
    simSpaceSettings {maxAngularAcceleration=100, maxAngularVelocity=100, maxLinearAcceleration=50, maxLinearVelocity=50, worldAlpha=0.1}
   }
   n41 = RigidBody(ComponentPose=@n40, OverrideWorldGravity="0,0,-980") {
    bClampLinearTranslationLimitToRefPose=true
    bOverrideWorldGravity=true
    overridePhysicsAsset=&/M_LRG_PeacefulPoem_Pumped_FrontendPhysics
    simulationSpace=BaseBoneSpace
    alphaBoolBlend {blendOption=Linear}
    baseBoneRef {boneName=pelvis}
    componentAppliedLinearAccClamp=<1000 1000 1000>
    componentLinearAccScale=<0.3 0.3 0.2>
    componentLinearVelScale=<0.3 0.3 0.2>
    overrideWorldGravity {z=-980}
    simSpaceSettings {
     maxAngularAcceleration=100
     maxAngularVelocity=100
     maxLinearAcceleration=50
     maxLinearVelocity=50
     velocityScaleZ=0.5
     worldAlpha=0.15
     externalAngularVelocity=<0.3 0.3>
     externalLinearDragV=<0.3 0.3 0.5>
     externalLinearVelocity=<0.3 0.3>
    }
   }
   n42 = "Transform (Modify) Bone - Bone: dyn_skirt_side_l_01"(ComponentPose=@n41, Alpha=1, Translation="0,0,0", Rotation="-0.001,-7.857,-0.002", Scale="1,1,1") {
    rotationMode=BMM_Additive
    rotationSpace=BCS_BoneSpace
    scaleSpace=BCS_BoneSpace
    translationMode=BMM_Additive
    translationSpace=BCS_BoneSpace
    alphaBoolBlend {blendOption=Linear}
    boneToModify {boneName=dyn_skirt_side_l_01}
    rotation=<rot -0.001 -7.857 -0.002>
   }
   n43 = "Transform (Modify) Bone - Bone: dyn_skirt_side_l_02"(ComponentPose=@n42, Alpha=1, Translation="0,0,0", Rotation="-0.001,12.634,-0.001", Scale="1,1,1") {
    rotationMode=BMM_Additive
    rotationSpace=BCS_BoneSpace
    scaleSpace=BCS_BoneSpace
    translationMode=BMM_Additive
    translationSpace=BCS_BoneSpace
    alphaBoolBlend {blendOption=Linear}
    boneToModify {boneName=dyn_skirt_side_l_02}
    rotation=<rot -0.001 12.6341 -0.001>
   }
   n44 = "Transform (Modify) Bone - Bone: dyn_skirt_side_l_03"(ComponentPose=@n43, Alpha=1, Translation="0,0,0", Rotation="0.009,-7.662,-0.005", Scale="1,1,1") {
    rotationMode=BMM_Additive
    rotationSpace=BCS_BoneSpace
    scaleSpace=BCS_BoneSpace
    translationMode=BMM_Additive
    translationSpace=BCS_BoneSpace
    alphaBoolBlend {blendOption=Linear}
    boneToModify {boneName=dyn_skirt_side_l_03}
    rotation=<rot 0.009 -7.662 -0.005>
   }
   n45 = "Transform (Modify) Bone - Bone: dyn_skirt_side_l_04"(ComponentPose=@n44, Alpha=1, Translation="0,0,0", Rotation="0.151,-14.080,-0.010", Scale="1,1,1") {
    rotationMode=BMM_Additive
    rotationSpace=BCS_BoneSpace
    scaleSpace=BCS_BoneSpace
    translationMode=BMM_Additive
    translationSpace=BCS_BoneSpace
    alphaBoolBlend {blendOption=Linear}
    boneToModify {boneName=dyn_skirt_side_l_04}
    rotation=<rot 0.151 -14.08 -0.01>
   }
   n46 = "Transform (Modify) Bone - Bone: dyn_skirt_side_l_05"(ComponentPose=@n45, Alpha=1, Translation="0,0,0", Rotation="0.025,-7.744,0.005", Scale="1,1,1") {
    rotationMode=BMM_Additive
    rotationSpace=BCS_BoneSpace
    scaleSpace=BCS_BoneSpace
    translationMode=BMM_Additive
    translationSpace=BCS_BoneSpace
    alphaBoolBlend {blendOption=Linear}
    boneToModify {boneName=dyn_skirt_side_l_05}
    rotation=<rot 0.025 -7.744 0.005>
   }
   Save cached pose 'Post Skirt Dynamics'(Pose=Blend Poses by bool(BlendPose_0=Component To Local(ComponentPose=@n46), BlendPose_1=Blend Poses by bool(BlendPose_0=@n36, BlendPose_1=@n39) {blendType=Linear}) {blendType=Linear}) {cacheName=Post Skirt Dynamics}
   n47 = Control Rig!(
    Source=Copy Pose From Mesh() {bUseAttachedParent=true}
    Alpha
    IsParachuteOpen:bool=Get bIsParachuteOpened()
    IsBuiltInPickaxeEquipped:bool=Get PeacefulPoemPumpedPickaxe() -> (PeacefulPoemPumpedPickaxe:bool)
    DisablePickaxeElbows:bool=Get DisablePickaxeElbowDeforms() -> (DisablePickaxeElbowDeforms:bool)
   ) -> (Pose) {
    alphaBoolBlend {blendOption=Linear}
    controlRigAssetReference {blueprintRigClass=&/M_LRG_PeacefulPoem_Pumped_CtrlRig.M_LRG_PeacefulPoem_Pumped_CtrlRig_C}
   }
   n48 = "Transform (Modify) Bone - Bone: deform_shoulderpad_l_base"!(ComponentPose=Local To Component(LocalPose=@n47), bAlphaBoolEnabled:bool=True, Translation="0,0,0", Rotation="0,0,0", Scale="1,1.05,1") -> (Pose) {
    alphaInputType=Bool
    scaleMode=BMM_Additive
    scaleSpace=BCS_BoneSpace
    alphaBoolBlend {blendOption=Linear}
    boneToModify {boneName=deform_shoulderpad_l_base}
    scale {y=1.05}
   }
   n49 = "Transform (Modify) Bone - Bone: deform_shoulderpad_r_base"!(ComponentPose=@n48, bAlphaBoolEnabled:bool=True, Translation="0,0,0", Rotation="0,0,0", Scale="1,1.05,1") -> (Pose) {
    alphaInputType=Bool
    scaleMode=BMM_Additive
    scaleSpace=BCS_BoneSpace
    alphaBoolBlend {blendOption=Linear}
    boneToModify {boneName=deform_shoulderpad_r_base}
    scale {y=1.05}
   }
   Save cached pose 'DeformPose'(Pose=Component To Local(ComponentPose=@n49)) {cacheName=DeformPose}
   n50 = "Transform (Modify) Bone - Bone: dyn_skirt_side_r_01"(
    ComponentPose=Local To Component(LocalPose=Use cached pose 'Post Skirt Dynamics'())
    Alpha=1
    Translation="-0.000,-0.532,-0.027"
    Rotation="0,0,0"
    Scale="1,1,1"
   ) {
    rotationMode=BMM_Additive
    rotationSpace=BCS_BoneSpace
    scaleSpace=BCS_BoneSpace
    translationMode=BMM_Additive
    translationSpace=BCS_BoneSpace
    alphaBoolBlend {blendOption=Linear}
    boneToModify {boneName=dyn_skirt_side_r_01}
    translation {y=-0.532, z=-0.027}
   }
   n51 = "Transform (Modify) Bone - Bone: dyn_skirt_armor_r"(ComponentPose=@n50, Alpha=1, Translation="0,1,0", Rotation="0.016,-2.999,-4.126", Scale="1,1,1") {
    rotationMode=BMM_Additive
    rotationSpace=BCS_BoneSpace
    scaleSpace=BCS_BoneSpace
    translationMode=BMM_Additive
    translationSpace=BCS_BoneSpace
    alphaBoolBlend {blendOption=Linear}
    boneToModify {boneName=dyn_skirt_armor_r}
    rotation=<rot 0.016 -2.999 -4.126>
    translation {y=1}
   }
   n52 = "Transform (Modify) Bone - Bone: dyn_skirt_armor_l"(ComponentPose=@n51, Alpha=1, Translation="0,0,0", Rotation="-0.003,-3,-0.004", Scale="1,1,1") {
    rotationMode=BMM_Additive
    rotationSpace=BCS_BoneSpace
    scaleSpace=BCS_BoneSpace
    translationMode=BMM_Additive
    translationSpace=BCS_BoneSpace
    alphaBoolBlend {blendOption=Linear}
    boneToModify {boneName=dyn_skirt_armor_l}
    rotation=<rot -0.003 -3 -0.004>
   }
   AnimGraphNode_StateMachine_0 = Pickaxe Equip Anim() {is=title AnimGraphNode_StateMachine}
   n53 = Layered blend per bone(BasePose=Component To Local(ComponentPose=@n52), BlendPoses_0=Slot 'DefaultSlot'(Source=@AnimGraphNode_StateMachine_0)) {layerSetup [{branchFilters [
       {boneName=deform_pickaxe_blade_a_01_l}
       {boneName=deform_pickaxe_blade_b_01_l}
       {boneName=deform_pickaxe_blade_a_01_r}
       {boneName=deform_pickaxe_blade_b_01_r}
      ]}]}
   Save cached pose 'PostPickaxeAnim'(Pose=@n53) {cacheName=PostPickaxeAnim}
  ]}
 Pickaxe Equip Anim {
  entry=Unequipped
  path=AnimGraph.AnimGraphNode_StateMachine_0.Pickaxe Equip Anim
  ids=id
  nodes [
   AnimStateNode_0 = Equip() {is=title AnimStateNode}
   AnimStateNode_1 = Idle()
   AnimStateNode_2 = Unequipped() {is=title AnimStateNode}
   AnimStateTransitionNode_0 = AnimStateTransitionNode() {bAutomaticRuleBasedOnSequencePlayerInState=true}
   AnimStateTransitionNode_1 = AnimStateTransitionNode()
   AnimStateTransitionNode_2 = AnimStateTransitionNode()
   AnimStateTransitionNode_5 = AnimStateTransitionNode()
  ]
  links [
   {
    dst=Idle
    rule=AnimGraph.AnimGraphNode_StateMachine_0.Pickaxe Equip Anim.AnimStateTransitionNode_0.Transition
    src=Equip
    via=AnimStateTransitionNode_0
   }
   {
    dst=Unequipped
    rule=AnimGraph.AnimGraphNode_StateMachine_0.Pickaxe Equip Anim.AnimStateTransitionNode_1.Transition
    src=Idle
    via=AnimStateTransitionNode_1
   }
   {
    dst=Equip
    rule=AnimGraph.AnimGraphNode_StateMachine_0.Pickaxe Equip Anim.AnimStateTransitionNode_2.Transition
    src=Unequipped
    via=AnimStateTransitionNode_2
   }
   {
    dst=Idle
    rule=AnimGraph.AnimGraphNode_StateMachine_0.Pickaxe Equip Anim.AnimStateTransitionNode_5.Transition
    src=Unequipped
    via=AnimStateTransitionNode_5
   }
  ]
 }
 Equip {
  path=AnimGraph.AnimGraphNode_StateMachine_0.Pickaxe Equip Anim.AnimStateNode_0.Equip
  nodes [
   n1 = Sequence Player 'PeacefulPoem_Pumped_body_blade_equip'() {
    bLoopAnimation=false
    sequence=/BRCosmetics/Characters/Player/Male/Large/Bodies/M_LRG_PeacefulPoem_Pumped/Animation/PeacefulPoem_Pumped_body_blade_equip
   }
   Output Animation Pose(Result=@n1)
  ]
 }
 Idle {
  path=AnimGraph.AnimGraphNode_StateMachine_0.Pickaxe Equip Anim.AnimStateNode_1.Idle
  nodes [
   n1 = Sequence Player 'PeacefulPoem_Pumped_body_blade_equipped'() {
    bLoopAnimation=false
    sequence=/BRCosmetics/Characters/Player/Male/Large/Bodies/M_LRG_PeacefulPoem_Pumped/Animation/PeacefulPoem_Pumped_body_blade_equipped
   }
   Output Animation Pose(Result=Slot 'DefaultSlot'(Source=@n1))
  ]
 }
 Unequipped {
  path=AnimGraph.AnimGraphNode_StateMachine_0.Pickaxe Equip Anim.AnimStateNode_2.Unequipped
  nodes [
   n1 = Sequence Player 'PeacefulPoem_Pumped_body_blade_unequipped'() {
    bLoopAnimation=false
    sequence=/BRCosmetics/Characters/Player/Male/Large/Bodies/M_LRG_PeacefulPoem_Pumped/Animation/PeacefulPoem_Pumped_body_blade_unequipped
   }
   Output Animation Pose(Result=Slot 'DefaultSlot'(Source=@n1))
  ]
 }
 Transition {path=AnimGraph.AnimGraphNode_StateMachine_0.Pickaxe Equip Anim.AnimStateTransitionNode_0.Transition, nodes [Result()]}
 Transition {
  path=AnimGraph.AnimGraphNode_StateMachine_0.Pickaxe Equip Anim.AnimStateTransitionNode_1.Transition
  nodes [
   n1 = Get PeacefulPoemPumpedPickaxe()
   n2 = Get Skip Transition() -> (Skip Transition:bool)
   Result(bCanEnterTransition=AND Boolean(A=NOT Boolean(A=@n1), B=NOT Boolean(A=@n2)))
  ]
 }
 Transition {
  path=AnimGraph.AnimGraphNode_StateMachine_0.Pickaxe Equip Anim.AnimStateTransitionNode_2.Transition
  nodes [
   n1 = Get PeacefulPoemPumpedPickaxe()
   Result(bCanEnterTransition=@n1)
   AND Boolean(A=@n1, B=NOT Boolean(A=Get Skip Transition()))
  ]
 }
 Transition {
  path=AnimGraph.AnimGraphNode_StateMachine_0.Pickaxe Equip Anim.AnimStateTransitionNode_5.Transition
  nodes [Result(bCanEnterTransition=Get Skip Transition())]
 }
 EventGraph {nodes [
   Event BlueprintUpdateAnimation()
   > n1 = Sequence()
   > n2 = Gravity Override State Function DEPRECATED!(
    execute
    CharacterPartAnimInstance=Self-Reference()
    GameTimeInSeconds=GetGameTimeInSeconds()
    GravityOverrideParams_Skydive_GoingUp_28_#1:Gravity Override Setting Struct Structure=(WindFrequency_28_#2=(X=1.5,Y=7,Z=8),WindAmplitude_31_#3=(X=200,Y=-500,Z=400),WindOffset_32_#4=(X=0,Y=0,Z=0),GravityOverride_15_#5=(X=0,Y=-300,Z=1500),JointName_27_#6="root")
    GravityOverrideParams_Skydive_FloatingDown_29_#7:Gravity Override Setting Struct Structure=Get Gravity Override Params Skydive Floating Down() -> (Gravity Override Params Skydive Floating Down:Gravity Override Setting Struct Structure)
    GravityOverrideParams_Skydive_DivingDown_30_#8:Gravity Override Setting Struct Structure=Get Gravity Override Params Skydive Diving Down() -> (Gravity Override Params Skydive Diving Down:Gravity Override Setting Struct Structure)
    GravityOverrideParams_Skydive_Parachuting_31_#9:Gravity Override Setting Struct Structure=Get Gravity Override Params Skydive Parachuting() -> (Gravity Override Params Skydive Parachuting:Gravity Override Setting Struct Structure)
    GravityOverrideParams_OnGround_Standing_32_#10:Gravity Override Setting Struct Structure=(WindFrequency_28_#2=(X=0,Y=0,Z=0),WindAmplitude_31_#3=(X=0,Y=0,Z=0),WindOffset_32_#4=(X=0,Y=0,Z=0),GravityOverride_15_#5=(X=0,Y=0,Z=-980),JointName_27_#6="root")
    GravityOverrideParams_OnGround_Moving_33_#11:Gravity Override Setting Struct Structure=(WindFrequency_28_#2=(X=0,Y=0,Z=0),WindAmplitude_31_#3=(X=0,Y=0,Z=0),WindOffset_32_#4=(X=0,Y=0,Z=0),GravityOverride_15_#5=(X=0,Y=0,Z=-980),JointName_27_#6="root")
    GravityOverrideParams_EmoteOrMelee_34_#12:Gravity Override Setting Struct Structure=(WindFrequency_28_#2=(X=0,Y=0,Z=0),WindAmplitude_31_#3=(X=0,Y=0,Z=0),WindOffset_32_#4=(X=0,Y=0,Z=0),GravityOverride_15_#5=(X=0,Y=0,Z=-980),JointName_27_#6="root")
    GravityOverrideParams_Swimming_35_#13:Gravity Override Setting Struct Structure=(WindFrequency_28_#2=(X=0,Y=0,Z=0),WindAmplitude_31_#3=(X=0,Y=0,Z=0),WindOffset_32_#4=(X=0,Y=0,Z=0),GravityOverride_15_#5=(X=-800,Y=0,Z=0),JointName_27_#6="pelvis")
    GravityOverrideParams_Front_End_36_#14:Gravity Override Setting Struct Structure=(WindFrequency_28_#2=(X=0,Y=0,Z=0),WindAmplitude_31_#3=(X=0,Y=0,Z=0),WindOffset_32_#4=(X=0,Y=0,Z=0),GravityOverride_15_#5=(X=0,Y=0,Z=-980),JointName_27_#6="root")
    GravityOverrideParams_DBNO_40_#15:Gravity Override Setting Struct Structure=(WindFrequency_28_#2=(X=0,Y=0,Z=0),WindAmplitude_31_#3=(X=0,Y=0,Z=0),WindOffset_32_#4=(X=0,Y=0,Z=0),GravityOverride_15_#5=(X=-980,Y=400,Z=0),JointName_27_#6="pelvis")
    GravityOverrideParams_DBNO_Carried_41_#16:Gravity Override Setting Struct Structure=(WindFrequency_28_#2=(X=0,Y=0,Z=0),WindAmplitude_31_#3=(X=0,Y=0,Z=0),WindOffset_32_#4=(X=0,Y=0,Z=0),GravityOverride_15_#5=(X=-980,Y=500,Z=-500),JointName_27_#6="pelvis")
    GravityOverrideParams_OnGround_Falling_45_#17:Gravity Override Setting Struct Structure=(WindFrequency_28_#2=(X=0,Y=0,Z=0),WindAmplitude_31_#3=(X=0,Y=0,Z=0),WindOffset_32_#4=(X=0,Y=0,Z=0),GravityOverride_15_#5=(X=0,Y=0,Z=-980),JointName_27_#6="root")
    GravityOverrideParams_Ziplining_46_#18:Gravity Override Setting Struct Structure=(WindFrequency_28_#2=(X=0,Y=0,Z=0),WindAmplitude_31_#3=(X=0,Y=0,Z=0),WindOffset_32_#4=(X=0,Y=0,Z=0),GravityOverride_15_#5=(X=0,Y=0,Z=-980),JointName_27_#6="root")
    GravityOverrideParams_Skydive_Surfing_49_#19:Gravity Override Setting Struct Structure=(WindFrequency_28_#2=(X=1.5,Y=7,Z=8),WindAmplitude_31_#3=(X=200,Y=-500,Z=400),WindOffset_32_#4=(X=0,Y=0,Z=0),GravityOverride_15_#5=(X=500,Y=-300,Z=-980),JointName_27_#6="root")
    GravityOverrideParams_Riding_Standing_54_#20:Gravity Override Setting Struct Structure=(WindFrequency_28_#2=(X=0,Y=0,Z=0),WindAmplitude_31_#3=(X=0,Y=0,Z=0),WindOffset_32_#4=(X=0,Y=0,Z=0),GravityOverride_15_#5=(X=0,Y=-350,Z=-980),JointName_27_#6="root")
    GravityOverrideParams_Riding_Moving_55_#21:Gravity Override Setting Struct Structure=(WindFrequency_28_#2=(X=0,Y=0,Z=0),WindAmplitude_31_#3=(X=0,Y=0,Z=0),WindOffset_32_#4=(X=0,Y=0,Z=0),GravityOverride_15_#5=(X=0,Y=-450,Z=-450),JointName_27_#6="root")
    GravityOverrideParams_Riding_Moving_Targeting_58_#22:Gravity Override Setting Struct Structure=(WindFrequency_28_#2=(X=0,Y=0,Z=0),WindAmplitude_31_#3=(X=0,Y=0,Z=0),WindOffset_32_#4=(X=0,Y=0,Z=0),GravityOverride_15_#5=(X=0,Y=-200,Z=-980),JointName_27_#6="root")
    GravityOverrideParams_Tactical_Sprint_57_#23:Gravity Override Setting Struct Structure=(WindFrequency_28_#2=(X=0,Y=0,Z=0),WindAmplitude_31_#3=(X=0,Y=0,Z=0),WindOffset_32_#4=(X=0,Y=0,Z=0),GravityOverride_15_#5=(X=0,Y=-450,Z=-500),JointName_27_#6="root")
    GravityOverrideParams_Rail_Grinding_61_#24:Gravity Override Setting Struct Structure=Get Gravity Override Params Rail Grinding() -> (Gravity Override Params Rail Grinding:Gravity Override Setting Struct Structure)
    GravityOverrideParams_DBNO_Tumbling_67_#25:Gravity Override Setting Struct Structure=(WindFrequency_28_#2=(X=0,Y=0,Z=0),WindAmplitude_31_#3=(X=0,Y=0,Z=0),WindOffset_32_#4=(X=0,Y=0,Z=0),GravityOverride_15_#5=(X=-2000,Y=0,Z=0),JointName_27_#6="pelvis")
    GravityOverrideParams_DBNO_RollFwd_68_#26:Gravity Override Setting Struct Structure=(WindFrequency_28_#2=(X=0,Y=0,Z=0),WindAmplitude_31_#3=(X=0,Y=0,Z=0),WindOffset_32_#4=(X=0,Y=0,Z=0),GravityOverride_15_#5=(X=-2000,Y=0,Z=0),JointName_27_#6="pelvis")
    GravityOverrideParams_DBNO_RollBwd_69_#27:Gravity Override Setting Struct Structure=(WindFrequency_28_#2=(X=0,Y=0,Z=0),WindAmplitude_31_#3=(X=0,Y=0,Z=0),WindOffset_32_#4=(X=0,Y=0,Z=0),GravityOverride_15_#5=(X=-4000,Y=0,Z=0),JointName_27_#6="pelvis")
    GravityOverrideParams_DBNO_Impulse_70_#28:Gravity Override Setting Struct Structure=(WindFrequency_28_#2=(X=0,Y=0,Z=0),WindAmplitude_31_#3=(X=0,Y=0,Z=0),WindOffset_32_#4=(X=0,Y=0,Z=0),GravityOverride_15_#5=(X=-2000,Y=0,Z=0),JointName_27_#6="pelvis")
    RBAN Params
    DebugDraw
    DebugJointName
   ) -> (then, GravityOverride, CurrentStateOut, GravityOverrideSettingsOut, RBAN SettingsOut)
   > Set Gravity Override(Gravity Override=@n2)
   n3 = Branch(execute=@n1.then_1, Condition=Get bIsInFrontEnd())
   n4 = GetAnimInstance(self=GetOwnerSkeletalMeshComponent())
   > n5 = Branch(Condition="To Bool (Double)"!(Value:double=GetCurveValue(self=@n4, CurveName=PeacefulPoemPumpedFrontendIdle)) -> (ReturnValue:bool))
   > Set DisablePickaxeElbowDeforms(DisablePickaxeElbowDeforms:bool=true) -> (Output_Get:bool)
   Set DisablePickaxeElbowDeforms(execute=[@n5.else, @n3.else], DisablePickaxeElbowDeforms=false)
   n6 = Event OnCharacterAndPickaxeCustomizationLoaded!() -> (
    OutputDelegate
    then
    CustomBaseMesh:Skeletal Mesh Component Object Reference
    CustomParts:Array of Custom Character Part Object References
    CustomMeshParts:Array of Skeletal Mesh Component Budgeted Object References
   ) {is=title K2Node_Event}
   n7 = Branch(execute=[@n6.then, @n14.then, @n16.then, @n10.then_1], Condition=Get bIsSkyDiving())
   n8 = Set PeacefulPoemPumpedPickaxe(
    execute=@n7.else
    PeacefulPoemPumpedPickaxe:bool=PickaxeItemDefinitionHasAnyMetaTag(Tags=Make Array("[0]"=(TagName="Cosmetics.MetaTags.PeacefulPoemPumped")), bRequiresEquipped=true)
   ) -> (Output_Get:bool)
   n9 = Set hero montage(
    execute=[@n15.then_0, @n8.then]
    hero montage:bool=Montage_IsActive!(
     self:Anim Instance Object Reference=GetOwnerFortPlayerAnimInstance!(self:Custom Character Part Anim Instance Object Reference) -> (ReturnValue:Fort Player Anim Instance Object Reference)
     Montage:Anim Montage Object Reference=/BRCosmetics/Animation/Game/MainPlayer/Emotes/PeacefulPoem_Pumped/Hero/Emote_PeacefulPoem_Pumped_Equip_Hero_M
    ) -> (ReturnValue:bool)
   ) -> (Output_Get:bool)
   > Set Skip Transition(Skip Transition:bool=@n9) -> (Output_Get:bool)
   Event OnWeaponChanged()
   > n10 = Sequence()
   > n11 = Cast To FortPlayerAnimInstance(Object=GetAnimInstance(self=GetOwnerSkeletalMeshComponent()))
   > n12 = Switch on EFortWeaponCoreAnimation!(execute, Selection:EFortWeaponCoreAnimation Enum=Get WeaponCoreAnim(self=@n11)) -> (
    Melee:Exec
    Pistol:Exec
    Shotgun:Exec
    PaperBlueprint:Exec
    Rifle:Exec
    MeleeOneHand:Exec
    MachinePistol:Exec
    RocketLauncher:Exec
    GrenadeLauncher:Exec
    GoingCommando:Exec
    AssaultRifle:Exec
    TacticalShotgun:Exec
    SniperRifle:Exec
    TrapPlacement:Exec
    ShoulderLauncher:Exec
    AbilityDecoTool:Exec
    Crossbow:Exec
    C4:Exec
    RemoteControl:Exec
    DualWield:Exec
    AR_BullPup:Exec
    AR_ForwardGrip:Exec
    MedPackPaddles:Exec
    SMG_P90:Exec
    AR_DrumGun:Exec
    Consumable_Small:Exec
    Consumable_Large:Exec
    Balloon:Exec
    MountedTurret:Exec
    CreativeTool:Exec
    ExplosiveBow:Exec
    AshtonIndigo:Exec
    AshtonChicago:Exec
    MeleeDualWield:Exec
    Unarmed:Exec
   )
   Set IsHoldingWeapon(
    execute=[
     @n12.AR_BullPup
     @n12.AR_ForwardGrip
     @n12.AshtonChicago
     @n12.AshtonIndigo
     @n12.AssaultRifle
     @n12.Crossbow
     @n12.ExplosiveBow
     @n12.GrenadeLauncher
     @n12.MachinePistol
     @n12.Rifle
     @n12.RocketLauncher
     @n12.SMG_P90
     @n12.Shotgun
     @n12.ShoulderLauncher
     @n12.SniperRifle
     @n12.TacticalShotgun
    ]
    IsHoldingWeapon:bool=true
   ) -> (Output_Get:bool)
   Set IsHoldingWeapon(
    execute=[
     @n12.AR_DrumGun
     @n12.AbilityDecoTool
     @n12.Balloon
     @n12.C4
     @n12.Consumable_Large
     @n12.Consumable_Small
     @n12.CreativeTool
     @n12.GoingCommando
     @n12.MedPackPaddles
     @n12.Melee
     @n12.MeleeDualWield
     @n12.MeleeOneHand
     @n12.MountedTurret
     @n12.PaperBlueprint
     @n12.RemoteControl
     @n12.TrapPlacement
     @n12.Unarmed
    ]
    IsHoldingWeapon=false
   )
   > Set IsHoldingPistol(IsHoldingPistol:bool=false) -> (Output_Get:bool)
   Set IsHoldingPistol(execute=[@n12.DualWield, @n12.Pistol], IsHoldingPistol=true)
   n13 = Event OnBeginSkydiving()
   Set PeacefulPoemPumpedPickaxe(execute=[@n13.then, @n15.then_1], PeacefulPoemPumpedPickaxe=false)
   > Set RetractBlades(RetractBlades:bool=false) -> (Output_Get:bool)
   > Set RetractBladesPistol(RetractBladesPistol:bool=false) -> (Output_Get:bool)
   n14 = Event OnEndSkydiving()
   Event OnBeginPlayingEmote() {is=title K2Node_Event}
   > Delay(Duration=0.2)
   > n15 = Sequence()
   n16 = Event OnEndPlayingEmote() {is=title K2Node_Event}
   Event OnCharacterCustomizationLoaded()
   > Set IsWearingBuiltInCape(IsWearingBuiltInCape:bool=BackpackItemDefinitionHasAnyMetaTag!(Tags:Array of Gameplay Tag Structures=Make Array("[0]"=(TagName="Cosmetics.MetaTags.PeacefulPoemPumped")), bRequiresEquipped:bool=false) -> (ReturnValue:bool)) -> (Output_Get:bool)
   Set RetractBlades(RetractBlades=false)
   > Set RetractBladesPistol(RetractBladesPistol=false)
   Set RetractBlades(RetractBlades=true)
   > Set RetractBladesPistol(RetractBladesPistol=false)
   Set RetractBlades(RetractBlades=true)
   > Set RetractBladesPistol(RetractBladesPistol=true)
   Get Gravity Override Params on Ground Moving() -> (Gravity Override Params on Ground Moving:Gravity Override Setting Struct Structure)
  ]}
]
variables [
 DisablePickaxeElbowDeforms {type=boolean}
 Gravity Override {type=Vector}
 Gravity Override Params on Ground Moving {
  type=GravityOverrideSettingStruct
  default {jointName=root, gravityOverride=<0 0 -980>, windAmplitude=<0 -500 0>, windFrequency=<0 5 0>, windOffset=<0 -200 0>}
 }
 Gravity Override Params Rail Grinding {
  type=GravityOverrideSettingStruct
  default {jointName=root, gravityOverride=<0 0 -980>, windAmplitude=<0 0 0>, windFrequency=<0 0 0>, windOffset=<0 0 0>}
 }
 Gravity Override Params Skydive Diving Down {
  type=GravityOverrideSettingStruct
  default {jointName=pelvis, gravityOverride=<-1500 0 0>, windAmplitude=<50 150 50>, windFrequency=<1.5 6 10>, windOffset=<0 0 0>}
 }
 Gravity Override Params Skydive Floating Down {
  type=GravityOverrideSettingStruct
  default {jointName=root, gravityOverride=<0 -500 200>, windAmplitude=<200 -250 200>, windFrequency=<1.5 3 8>, windOffset=<0 -300 200>}
 }
 Gravity Override Params Skydive Parachuting {
  type=GravityOverrideSettingStruct
  default {jointName=root, gravityOverride=<0 -800 -980>, windAmplitude=<200 -500 400>, windFrequency=<1.5 20 8>, windOffset=<0 -900 -900>}
 }
 hero montage {type=boolean}
 IsHoldingPistol {type=boolean}
 IsHoldingWeapon {type=boolean}
 IsWearingBuiltInCape {type=boolean}
 PeacefulPoemPumpedPickaxe {type=boolean}
 PumpedPickaxe_IsEquipped {type=number}
 RetractBlades {type=boolean}
 RetractBladesPistol {type=boolean}
 Skip Transition {type=boolean}
]
