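# Physics asset of cape colliders for the large male PeacefulPoem Pumped character body, giving a simplified collision volume of the upper body and rear skirt.
# Capsules cover the pelvis, spine, neck, clavicles and rear dyn_skirt bones, with a single sphere on the head.
# Arm, leg and foot bodies hold no shapes, and no constraints exist, so the asset defines collision shapes without ragdoll joints.
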PhysicsAsset /BRCosmetics/Characters/Player/Male/Large/Bodies/M_LRG_PeacefulPoem_Pumped/Meshes/M_LRG_PeacefulPoem_Pumped_CapeColliders
bodies {
 calf_l []
 calf_r []
 clavicle_l [{shapeType=Capsule, length=9.0192, radius=9.8871, center=<9.4266 -1.1806 -2.4977>, rotation=<rot 69.3297 1.4476 176.9462>}]
 clavicle_r [{shapeType=Capsule, length=9.0192, radius=9.8871, center=<-9.4266 1.1806 2.4977>, rotation=<rot 69.3297 1.4476 -3.0538>}]
 dyn_skirt_rear_l_02 [{
   shapeType=Capsule
   length=10.6913
   radius=8.5682
   shapeName=dyn_skirt_rear_l_02_capsule
   center=<-0.8411 -4.2641 1.2074>
   rotation=<rot -10.3289 88.621 -88.8532>
  }]
 dyn_skirt_rear_l_03 [{
   shapeType=Capsule
   length=15.5573
   radius=7.5259
   shapeName=dyn_skirt_rear_l_03_capsule
   center=<-7.0377 -0.2697 -0.3895>
   rotation=<rot 10.4155 -94.4527 -94.3265>
  }]
 dyn_skirt_rear_r_02 [{
   shapeType=Capsule
   length=10.6913
   radius=8.5682
   shapeName=dyn_skirt_rear_l_02_capsule
   center=<0.6551 -3.4616 0.2011>
   rotation=<rot -10.3289 88.621 91.1468>
  }]
 dyn_skirt_rear_r_03 [{
   shapeType=Capsule
   length=19.951
   radius=5.4404
   shapeName=dyn_skirt_rear_r_03_capsule
   center=<-3.3357 -0.0609 0.3101>
   rotation=<rot 0 -92.9264 -87.8119>
  }]
 foot_l []
 foot_r []
 head [{shapeType=Sphere, radius=17.3125, center=<11.028 -0.174 0>}]
 lowerarm_l []
 lowerarm_r []
 neck_02 [{shapeType=Capsule, length=13.4375, radius=7.3175, center=<-0.2642 -0.3408 0>, rotation=<rot 90 -171.1193 -175.3978>}]
 pelvis [{shapeType=Capsule, length=16.0518, radius=20.695, center=<-4.1467 0.1881 0>}]
 spine_01 [{shapeType=Capsule, length=7.9438, radius=13.5586, center=<-1.7965 0.1884 0>}]
 spine_02 [
  {shapeType=Capsule, length=10.5764, radius=15.2572, center=<-1.711 -0.1851 0.0001>}
  {shapeType=Capsule, length=29.3956, radius=11.5807, center=<2.5024 0.4992 3.9825>, rotation=<rot 90 0.0049 0.0049>}
  {shapeType=Capsule, length=29.875, radius=11.1013, center=<2.4704 0.4991 -3.9825>, rotation=<rot 90 -0.0396 -0.0396>}
 ]
 spine_03 [{shapeType=Capsule, length=9.9294, radius=15.7343, center=<0.1445 0.2396 0.0001>}]
 spine_04 [{shapeType=Capsule, length=12.2216, radius=18.695, center=<1.5706 0.7765 0.0002>}]
 spine_05 [
  {shapeType=Capsule, length=6.4601, radius=21.3825, center=<-3.4016 1.8109 0>}
  {shapeType=Capsule, length=13.9518, radius=13.9425, center=<-1.5629 0.4452 0>}
 ]
 thigh_l []
 thigh_r []
 upperarm_l []
 upperarm_r []
}
constraints []
