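# Control Rig for the face accessory of the PeacefulPoem Pumped large male body, correcting hood and neck scarf bones as the head moves.
# Parent constraints first blend the three hood base bones between weighted head, neck and torso bones.
# Five spherical pose readers on the head, relative to spine_05, then weight corrective ModifyTransforms on hood and scarf bones.
# A head-rotation slerp and a curve remap are left outside the execution chain.
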ControlRigBlueprint /BRCosmetics/Characters/Player/Male/Large/Bodies/M_LRG_PeacefulPoem_Pumped/Meshes/Parts/M_LRG_PeacefulPoem_Pumped_FaceAcc_CtrlRig
graphs [
 RigVMFunctionLibrary {links [], nodes []}
 RigVMModel {nodes [
   RigUnit_BeginExecution = &RigUnit_BeginExecution()
   &RigUnit_ParentConstraint(
    AdvancedSettings=(RotationOrderForFilter=XZY)
    Child=(Type=Bone,Name="dyn_hood_l_base")
    Parents=((Item=(Type=Bone,Name="head"),Weight=0.5),(Item=(Type=Bone,Name="neck_02"),Weight=0),(Item=(Type=Bone,Name="spine_05"),Weight=0.5))
    Weight=1
    ExecutePin=@RigUnit_BeginExecution.ExecutePin
   )
   > &RigUnit_ParentConstraint(
    AdvancedSettings=(RotationOrderForFilter=XZY)
    Child=(Type=Bone,Name="dyn_hood_rear_base")
    Parents="((Item=(Type=Bone,Name=\"spine_04\"),Weight=0.0032),(Item=(Type=Bone,Name=\"spine_03\"),Weight=0.0143),(Item=(Type=Bone,Name=\"clavicle_l\"),Weight=0.0188),(Item=(Type=Bone,Name=\"clavicle_r\"),Weight=0.0188),(Item=(Type=Bone,Name=\"neck_01\"),Weight=0.0211),(Item=( ...(truncated)"
    Weight=1
   )
   > &RigUnit_ParentConstraint(
    AdvancedSettings=(RotationOrderForFilter=XZY)
    Child=(Type=Bone,Name="dyn_hood_r_base")
    Parents=((Item=(Type=Bone,Name="head"),Weight=0.5),(Item=(Type=Bone,Name="neck_02"),Weight=0),(Item=(Type=Bone,Name="spine_05"),Weight=0.5))
    Weight=1
   )
   > n1 = &RigUnit_SphericalPoseReader(
    ActiveRegionScaleFactors=(PositiveWidth=3,NegativeWidth=1,PositiveHeight=0.5,NegativeHeight=0.5)
    ActiveRegionSize=0.1
    Debug=(bDrawDebug=true,bDraw2D=False,bDrawLocalAxes=False,DebugScale=25,DebugSegments=20,DebugThickness=0.25)
    DriverAxis=(X=1,Y=0,Z=0)
    DriverItem=(Type=Bone,Name="head")
    FalloffRegionScaleFactors=(PositiveWidth=1,NegativeWidth=1,PositiveHeight=1,NegativeHeight=1)
    FalloffSize=0.25
    FlipHeightScaling=False
    FlipWidthScaling=False
    OptionalParentItem=(Type=Bone,Name="spine_05")
    RotationOffset=(X=90,Y=0,Z=-170)
   )
   > &RigUnit_ModifyTransforms(
    ItemToModify="((Item=(Type=Bone,Name=\"dyn_hood_rear_base\"),Transform=(Rotation=(X=0,Y=0,Z=-0.6045,W=0.7966),Translation=(X=-0.6832,Y=-3.0341,Z=-0.0828),Scale3D=(X=1,Y=1,Z=1))),(Item=(Type=Bone,Name=\"dyn_neckscarf_rear_base\"),Transform=(Rotation=(X=0,Y=0,Z=-0.3727,W=0.92 ...(truncated)"
    WeightMinimum=0
    Weight=@n1.OutputParam
   )
   > n2 = &RigUnit_SphericalPoseReader(
    ActiveRegionScaleFactors=(PositiveWidth=3,NegativeWidth=1,PositiveHeight=0.5,NegativeHeight=0.5)
    ActiveRegionSize=0.1
    Debug=(bDrawDebug=true,bDraw2D=False,bDrawLocalAxes=False,DebugScale=25,DebugSegments=20,DebugThickness=0.25)
    DriverAxis=(X=1,Y=0,Z=0)
    DriverItem=(Type=Bone,Name="head")
    FalloffRegionScaleFactors=(PositiveWidth=1,NegativeWidth=1,PositiveHeight=1,NegativeHeight=1)
    FalloffSize=0.25
    FlipHeightScaling=False
    FlipWidthScaling=False
    OptionalParentItem=(Type=Bone,Name="spine_05")
    RotationOffset=(X=70,Y=0,Z=-160)
   )
   > &RigUnit_ModifyTransforms(
    ItemToModify=((Item=(Type=Bone,Name="dyn_neckscarf_rear_r_base"),Transform=(Rotation=(X=0.1946,Y=0.0301,Z=0.6144,W=0.764),Translation=(X=0,Y=0,Z=0),Scale3D=(X=1,Y=1,Z=1))))
    WeightMinimum=0
    Weight=@n2.OutputParam
   )
   > n3 = &RigUnit_SphericalPoseReader(
    ActiveRegionScaleFactors=(PositiveWidth=3,NegativeWidth=1,PositiveHeight=0.5,NegativeHeight=0.5)
    ActiveRegionSize=0.1
    Debug=(bDrawDebug=false,bDraw2D=False,bDrawLocalAxes=False,DebugScale=25,DebugSegments=20,DebugThickness=0.25)
    DriverAxis=(X=1,Y=0,Z=0)
    DriverItem=(Type=Bone,Name="head")
    FalloffRegionScaleFactors=(PositiveWidth=1,NegativeWidth=1,PositiveHeight=1,NegativeHeight=1)
    FalloffSize=0.25
    FlipHeightScaling=False
    FlipWidthScaling=False
    OptionalParentItem=(Type=Bone,Name="spine_05")
    RotationOffset=(X=110,Y=0,Z=-160)
   )
   > &RigUnit_ModifyTransforms(
    ItemToModify=((Item=(Type=Bone,Name="dyn_neckscarf_rear_l_base"),Transform=(Rotation=(X=0.2027,Y=0.0444,Z=0.4933,W=0.8447),Translation=(X=0,Y=0,Z=0),Scale3D=(X=1,Y=1,Z=1))))
    WeightMinimum=0
    Weight=@n3.OutputParam
   )
   > n4 = &RigUnit_SphericalPoseReader(
    ActiveRegionScaleFactors=(PositiveWidth=1,NegativeWidth=1,PositiveHeight=2,NegativeHeight=1)
    ActiveRegionSize=0.1
    Debug=(bDrawDebug=false,bDraw2D=False,bDrawLocalAxes=False,DebugScale=25,DebugSegments=20,DebugThickness=0.25)
    DriverAxis=(X=0,Y=1,Z=0)
    DriverItem=(Type=Bone,Name="head")
    FalloffRegionScaleFactors=(PositiveWidth=1,NegativeWidth=1,PositiveHeight=1,NegativeHeight=1)
    FalloffSize=0.45
    FlipHeightScaling=False
    FlipWidthScaling=False
    OptionalParentItem=(Type=Bone,Name="spine_05")
    RotationOffset=(X=180,Y=0,Z=0)
   )
   > &RigUnit_ModifyTransforms(
    ItemToModify="((Item=(Type=Bone,Name=\"dyn_hood_rear\"),Transform=(Rotation=(X=0,Y=0,Z=0,W=1),Translation=(X=0,Y=0,Z=0.6328),Scale3D=(X=1,Y=1,Z=1))),(Item=(Type=Bone,Name=\"dyn_neckscarf_rear_base\"),Transform=(Rotation=(X=0,Y=0,Z=0,W=1),Translation=(X=0,Y=0,Z=0),Scale3D=(X ...(truncated)"
    WeightMinimum=0
    Weight=@n4.OutputParam
   )
   > n5 = &RigUnit_SphericalPoseReader(
    ActiveRegionScaleFactors=(PositiveWidth=1,NegativeWidth=1,PositiveHeight=1,NegativeHeight=2)
    ActiveRegionSize=0.1
    Debug=(bDrawDebug=false,bDraw2D=False,bDrawLocalAxes=False,DebugScale=25,DebugSegments=20,DebugThickness=0.25)
    DriverAxis=(X=0,Y=1,Z=0)
    DriverItem=(Type=Bone,Name="head")
    FalloffRegionScaleFactors=(PositiveWidth=1,NegativeWidth=1,PositiveHeight=1,NegativeHeight=1)
    FalloffSize=0.45
    FlipHeightScaling=False
    FlipWidthScaling=False
    OptionalParentItem=(Type=Bone,Name="spine_05")
    RotationOffset=(X=0,Y=0,Z=0)
   )
   > &RigUnit_ModifyTransforms(
    ItemToModify="((Item=(Type=Bone,Name=\"dyn_hood_rear\"),Transform=(Rotation=(X=0,Y=0,Z=0,W=1),Translation=(X=0,Y=0,Z=-2.978),Scale3D=(X=1,Y=1,Z=1))),(Item=(Type=Bone,Name=\"dyn_hood_rear_base\"),Transform=(Rotation=(X=0,Y=0,Z=0,W=1),Translation=(X=0,Y=0,Z=0),Scale3D=(X=1,Y= ...(truncated)"
    WeightMinimum=0
    Weight=@n5.OutputParam
   )
   RigUnit_GetTransform = &RigUnit_GetTransform(Item=(Type=Bone,Name="head"), Space=LocalSpace, bInitial=False)
   Interpolate = /Script/RigVM.RigVMFunction_MathQuaternionSlerp(B=(X=0,Y=0,Z=0,W=1), T=0.7, A=@RigUnit_GetTransform.Transform.Rotation) {is=struct RigVMUnitNode, template="Interpolate::Execute(in A,in B,in T,out Result)"}
   RigUnit_SetRotation = &RigUnit_SetRotation(ExecutePin=(), Item=(Type=Bone,Name="head"), Weight=1, bInitial=False, Value=@Interpolate.Result)
   RigVMFunction_AnimEvalRichCurve = &RigVMFunction_AnimEvalRichCurve(
    Curve=(EditorCurveData=(Keys=((InterpMode=RCIM_Cubic),(InterpMode=RCIM_Cubic,Time=0.5579,Value=0.225,ArriveTangent=1,LeaveTangent=1),(Time=1,Value=1))))
    SourceMinimum=0
    TargetMinimum=0
    Value=@n1.OutputParam
   )
  ]}
]
hierarchy [
 root {loc=<0 0 0>, rot=<0 0 0 1>}
 pelvis {parent=root, loc=<0 -0.7321 92.4218>, rot=<0.0127 0.707 -0.0127 -0.707>}
 spine_01 {parent=pelvis, loc=<5.6002 -0.5128 0>, rot=<0 0 0.034 -0.9994>}
 spine_02 {parent=spine_01, loc=<8.0753 0.0139 0>, rot=<0 0 0.0238 -0.9997>}
 spine_03 {parent=spine_02, loc=<7.0393 0 0>, rot=<0 0 0.0149 -0.9999>}
 spine_04 {parent=spine_03, loc=<8.1024 0 0>, rot=<0 0 -0.0132 -0.9999>}
 spine_05 {parent=spine_04, loc=<10.0146 0 0>, rot=<0 0 -0.0015 -1>}
 neck_01 {parent=spine_05, loc=<10.815 1.9196 0>, rot=<0 0 -0.1864 -0.9825>}
 neck_02 {parent=neck_01, loc=<4.7347 0.3045 0>, rot=<0 0 0.0363 -0.9993>}
 head {parent=neck_02, loc=<5.7632 0.5552 0>, rot=<0 0 0.115 -0.9934>}
 clavicle_l {parent=spine_05, loc=<2.5856 -0.2866 -4.3777>, rot=<0.0465 -0.7289 -0.009 -0.6829>}
 clavicle_r {parent=spine_05, loc=<2.5855 -0.2866 4.3777>, rot=<0.7289 0.0465 -0.6829 0.009>}
 dyn_hood_l_base {parent=spine_05, loc=<23.7915 2.8112 -9.0208>, rot=<0 0 -0.1033 -0.9947>}
 dyn_hood_l {parent=dyn_hood_l_base, loc=<0 0 0>, rot=<0 0 0.0633 -0.998>}
 dyn_hood_rear_base {parent=spine_05, loc=<25.1844 -4.6233 0>, rot=<0 0 -0.0631 -0.998>}
 dyn_hood_rear {parent=dyn_hood_rear_base, loc=<0 0 0>, rot=<0 0 0.023 -0.9997>}
 dyn_hood_r_base {parent=spine_05, loc=<23.7912 2.8112 9.0208>, rot=<0 0 0.9947 -0.1033>}
 dyn_hood_r {parent=dyn_hood_r_base, loc=<0 0 0>, rot=<0 0 0.0633 -0.998>}
 dyn_neckscarf_rear_base {parent=spine_05, loc=<13.9345 -9.8495 0>, rot=<0 0 -0.0631 -0.998>}
 dyn_neckscarf_rear {parent=dyn_neckscarf_rear_base, loc=<0 0 0>, rot=<0 0 0 1>}
 dyn_neckscarf_l_base {parent=spine_05, loc=<9.0818 11.6642 -7.7904>, rot=<0.4633 0.0293 -0.0559 -0.8839>}
 dyn_neckscarf_l {parent=dyn_neckscarf_l_base, loc=<0 0 0>, rot=<0 0 0.023 -0.9997>}
 dyn_neckscarf_r_base {parent=spine_05, loc=<9.4411 9.9788 10.2497>, rot=<-0.5388 -0.1152 -0.1055 -0.8278>}
 dyn_neckscarf_r {parent=dyn_neckscarf_r_base, loc=<0 0 0>, rot=<0 0 0 1>}
 dyn_neckscarf_rear_l_base {parent=spine_05, loc=<13.0249 -3.4285 -11.6702>, rot=<0.8196 0.0518 -0.036 -0.5695>}
 dyn_neckscarf_rear_l {parent=dyn_neckscarf_rear_l_base, loc=<0 0 0>, rot=<0 0 -0.2775 -0.9607>}
 dyn_neckscarf_rear_r_base {parent=spine_05, loc=<13.2182 -4.9527 10.5357>, rot=<0.0518 -0.8196 0.5695 -0.036>}
 dyn_neckscarf_rear_r {parent=dyn_neckscarf_rear_r_base, loc=<0 0 0>, rot=<0 0 0.023 -0.9997>}
]
