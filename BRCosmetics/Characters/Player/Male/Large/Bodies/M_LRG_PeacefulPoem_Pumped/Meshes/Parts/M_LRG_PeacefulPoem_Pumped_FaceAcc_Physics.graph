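# Physics asset for the face accessory part of the large male PeacefulPoem Pumped player body, covering its hood and neck scarf bones.
# Each dynamic hood and scarf bone is a capsule constrained with tight swing and twist limits to a tiny capsule on its matching base bone.
# The base bones and a head capsule link to a large spine_05 capsule through constraints with free angular motion.
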PhysicsAsset /BRCosmetics/Characters/Player/Male/Large/Bodies/M_LRG_PeacefulPoem_Pumped/Meshes/Parts/M_LRG_PeacefulPoem_Pumped_FaceAcc_Physics
bodies {
 dyn_hood_l [{
   shapeType=Capsule
   length=7.9375
   radius=3.1613
   shapeName=dyn_hood_l_capsule
   center=<-3.3859 2.421 -2.4555>
   rotation=<rot 0 -176.605 -83.6092>
  }]
 dyn_hood_l_base [{shapeType=Capsule, length=0.505, radius=0.505, shapeName=dyn_hood_l_base_capsule, rotation=<rot 0 -90 0>}]
 dyn_hood_r [{
   shapeType=Capsule
   length=7.9375
   radius=3.1613
   shapeName=dyn_hood_l_capsule
   center=<3.3859 -2.421 2.4555>
   rotation=<rot 0 -176.605 96.3908>
  }]
 dyn_hood_r_base [{shapeType=Capsule, length=0.505, radius=0.505, shapeName=dyn_hood_r_base_capsule, rotation=<rot 0 -90 0>}]
 dyn_hood_rear [{shapeType=Capsule, length=11.375, radius=2.8438, shapeName=dyn_hood_rear_capsule, center=<-4.0959 -3.43 0>}]
 dyn_hood_rear_base [{shapeType=Capsule, length=0.505, radius=0.505, shapeName=dyn_hood_rear_base_capsule, rotation=<rot 0 -90 0>}]
 dyn_neckscarf_l [{
   shapeType=Capsule
   length=6.4062
   radius=5.5
   shapeName=dyn_neckscarf_l_capsule
   center=<-0.8086 1.6207 -0.155>
   rotation=<rot -15.2123 -87.0126 -11.2488>
  }]
 dyn_neckscarf_l_base [{shapeType=Capsule, length=0.505, radius=0.505, shapeName=dyn_neckscarf_l_base_capsule, rotation=<rot 0 -90 0>}]
 dyn_neckscarf_r [{
   shapeType=Capsule
   length=12.2812
   radius=4.8438
   shapeName=dyn_neckscarf_r_capsule
   center=<0.8065 2.5204 -0.9416>
   rotation=<rot 0.0001 0 7.9969>
  }]
 dyn_neckscarf_r_base [{shapeType=Capsule, length=0.505, radius=0.505, shapeName=dyn_neckscarf_r_base_capsule, rotation=<rot 0 -90 0>}]
 dyn_neckscarf_rear [{
   shapeType=Capsule
   length=4.4918
   radius=3.785
   shapeName=dyn_neckscarf_rear_capsule
   center=<4.4003 -2.1243 -0.1734>
   rotation=<rot -3.0075 -90 -0.0283>
  }]
 dyn_neckscarf_rear_base [{shapeType=Capsule, length=0.505, radius=0.505, shapeName=dyn_neckscarf_rear_base_capsule}]
 dyn_neckscarf_rear_l [{
   shapeType=Capsule
   length=9.3438
   radius=4.8438
   shapeName=dyn_neckscarf_rear_r_capsule
   center=<0.0853 1.5831 -0.243>
   rotation=<rot 3.6305 -15.363 -167.0214>
  }]
 dyn_neckscarf_rear_l_base [{shapeType=Capsule, length=0.505, radius=0.505, shapeName=dyn_neckscarf_rear_l_base_capsule, rotation=<rot 0 -90 0>}]
 dyn_neckscarf_rear_r [{
   shapeType=Capsule
   length=9.3438
   radius=4.8438
   shapeName=dyn_neckscarf_rear_r_capsule
   center=<1.3929 -2.4877 0.6988>
   rotation=<rot -0.0001 0 27.2412>
  }]
 dyn_neckscarf_rear_r_base [{shapeType=Capsule, length=0.505, radius=0.505, shapeName=dyn_neckscarf_rear_r_base_capsule, rotation=<rot 0 -90 0>}]
 head [{shapeType=Capsule, length=11.4726, radius=7.8619, rotation=<rot 90 0 0>}]
 spine_05 [{
   shapeType=Capsule
   length=16.6739
   radius=15.1473
   shapeName=spine_05_capsule
   center=<12.0278 3.9053 0.5652>
   rotation=<rot 0 22.8227 -86.7593>
  }]
}
constraints [
 {bone1Name=dyn_hood_l, bone2Name=dyn_hood_l_base, swing1LimitDegrees=15, swing2LimitDegrees=15, twistLimitDegrees=5}
 {bone1Name=dyn_hood_rear, bone2Name=dyn_hood_rear_base, swing1LimitDegrees=15, swing2LimitDegrees=15, twistLimitDegrees=15}
 {
  bone1Name=head
  bone2Name=spine_05
  swing1LimitDegrees=45
  swing1Motion=Free
  swing2LimitDegrees=45
  swing2Motion=Free
  twistLimitDegrees=45
  twistMotion=Free
 }
 {
  bone1Name=dyn_hood_l_base
  bone2Name=spine_05
  swing1LimitDegrees=45
  swing1Motion=Free
  swing2LimitDegrees=45
  swing2Motion=Free
  twistLimitDegrees=45
  twistMotion=Free
 }
 {
  bone1Name=dyn_hood_rear_base
  bone2Name=spine_05
  swing1LimitDegrees=45
  swing1Motion=Free
  swing2LimitDegrees=45
  swing2Motion=Free
  twistLimitDegrees=45
  twistMotion=Free
 }
 {
  bone1Name=dyn_hood_r_base
  bone2Name=spine_05
  swing1LimitDegrees=45
  swing1Motion=Free
  swing2LimitDegrees=45
  swing2Motion=Free
  twistLimitDegrees=45
  twistMotion=Free
 }
 {bone1Name=dyn_hood_r, bone2Name=dyn_hood_r_base, swing1LimitDegrees=15, swing2LimitDegrees=15, twistLimitDegrees=5}
 {bone1Name=dyn_neckscarf_rear, bone2Name=dyn_neckscarf_rear_base, swing1LimitDegrees=12, swing2LimitDegrees=8, twistLimitDegrees=8}
 {
  bone1Name=dyn_neckscarf_rear_base
  bone2Name=spine_05
  swing1LimitDegrees=45
  swing1Motion=Free
  swing2LimitDegrees=45
  swing2Motion=Free
  twistLimitDegrees=45
  twistMotion=Free
 }
 {
  bone1Name=dyn_neckscarf_l_base
  bone2Name=spine_05
  swing1LimitDegrees=45
  swing1Motion=Free
  swing2LimitDegrees=45
  swing2Motion=Free
  twistLimitDegrees=45
  twistMotion=Free
 }
 {bone1Name=dyn_neckscarf_l, bone2Name=dyn_neckscarf_l_base, swing1LimitDegrees=12, swing2LimitDegrees=8, twistLimitDegrees=8}
 {
  bone1Name=dyn_neckscarf_r_base
  bone2Name=spine_05
  swing1LimitDegrees=45
  swing1Motion=Free
  swing2LimitDegrees=45
  swing2Motion=Free
  twistLimitDegrees=45
  twistMotion=Free
 }
 {bone1Name=dyn_neckscarf_r, bone2Name=dyn_neckscarf_r_base, swing1LimitDegrees=12, swing2LimitDegrees=8, twistLimitDegrees=8}
 {
  bone1Name=dyn_neckscarf_rear_l_base
  bone2Name=spine_05
  swing1LimitDegrees=45
  swing1Motion=Free
  swing2LimitDegrees=45
  swing2Motion=Free
  twistLimitDegrees=45
  twistMotion=Free
 }
 {
  bone1Name=dyn_neckscarf_rear_r_base
  bone2Name=spine_05
  swing1LimitDegrees=45
  swing1Motion=Free
  swing2LimitDegrees=45
  swing2Motion=Free
  twistLimitDegrees=45
  twistMotion=Free
 }
 {bone1Name=dyn_neckscarf_rear_r, bone2Name=dyn_neckscarf_rear_r_base, swing1LimitDegrees=12, swing2LimitDegrees=8, twistLimitDegrees=8}
 {bone1Name=dyn_neckscarf_rear_l, bone2Name=dyn_neckscarf_rear_l_base, swing1LimitDegrees=12, swing2LimitDegrees=8, twistLimitDegrees=8}
]
